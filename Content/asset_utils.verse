using { /Verse.org/Assets }
using { /UnrealEngine.com/Temporary/Diagnostics }

my_debug_draw := class(debug_draw_channel) {}

SlashHitVFX := module:
    NS<public> := module {}

Pack_Effects_1 := module:
    Air_Magic<public> := module:
        VFX_Niagara<public> := module {}
    Fire_Magic<public> := module:
        VFX_Niagara<public> := module {}

Electric_VFX := module:
    Niagara<public> := module:
        Electric<public> := module {}
        Laser<public> := module {}

MagicSpells := module:
    Electric<public> := module:
        FX<public> := module {}

SkillsVFX := module:
    Particles<public> := module {}

DarkMagicFX := module:
    Niagara<public> := module:
        Projectile<public> := module {}

# _VFX_ToUEFN := module {}

Original := module:
    UI<public> := module:
        Streak<public> := module {}