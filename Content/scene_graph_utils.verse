using { /Verse.org/SpatialMath }
using { /UnrealEngine.com/Temporary/SpatialMath }

UnrealVector3 := (/UnrealEngine.com/Temporary/SpatialMath:)vector3
UnrealRotation := (/UnrealEngine.com/Temporary/SpatialMath:)rotation
VerseVector3 := (/Verse.org/SpatialMath:)vector3
VerseRotation := (/Verse.org/SpatialMath:)rotation
VerseTransform := (/Verse.org/SpatialMath:)transform