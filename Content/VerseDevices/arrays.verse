using { /Verse.org/Random }

Array<public> := module:

    # Makes an `array` with only unique elements
    (Input:[]t where t:subtype(comparable)).Unique<public>()<transacts>:[]t =
        var UniqueArray : []t = array{} 
        for (Value : Input):
            if (UniqueArray.Find[Value] > -1) {}
            else:
                set UniqueArray = UniqueArray.Append(Value)
        return UniqueArray

    # Makes an `array` by returning items in a random order 
    (Input:[]t where t:type).Shuffle<public>()<transacts>:[]t =
        Shuffle(Input)

    # Makes an `array` by returning items in a random order 
    # Fails if `Input` is empty
    (Input:[]t where t:type).RandomElement<public>()<decides><transacts>:t =
        Index : int := GetRandomInt(0, Input.Length - 1)
        Input[Index]

    # Makes an `array` by running a `Mapper` function on every item
    (Input:[]t where t:type).Map<public>(Mapper: type{_(:t)<transacts> : t})<transacts>:[]t =
        for (Value : Input, NewValue := Mapper(Value)). NewValue

    # Makes an `array` duplicate of the input
    (Input:[]t where t:type).Copy<public>()<transacts>:[]t =
        for (Value : Input). Value

    # Makes an `array` by filtering out items that `FilterBy` function succeeds
    (Input:[]t where t:type).Filter<public>(FilterBy: type{_(:t)<decides><transacts> : void})<transacts>:[]t =
        for (Value : Input, FilterBy[Value]). Value

    # Succeeds if all items in an `array` make the `FilterBy` function succeed
    (Input:[]t where t:type).IsEveryElement<public>(FilterBy: type{_(:t)<decides><transacts> : void})<decides><transacts> : void =
        FilteredArray := Input.Filter(FilterBy)
        FilteredArray.Length = Input.Length

    # Succeeds if any items in an `array` make the `FilterBy` function succeed
    (Input:[]t where t:type).IsAnyElement<public>(FilterBy: type{_(:t)<decides><transacts> : void})<decides><transacts> : void =
        FilteredArray := Input.Filter(FilterBy)
        FilteredArray.Length > 0

     # Reduces an `array` by running a function on an accumulator
    (Input:[]t where t:type).Reduce<public>(ReductionFunction: type{_(:k, :t)<transacts> : k}, Accumulator: k where k:type)<transacts>:k =
        if (Input.Length > 0, Return := Input.Slice[1].Reduce(ReductionFunction, ReductionFunction(Accumulator, Input[0]))):
            return Return
        else: 
            return Accumulator

    # Sums an `array` of ints
    (Input:[]int).SumInts<public>()<transacts> : int = 
        return Input.Reduce(Reduce.SumInts, 0)
    
    # Sums an `array` of floats
    (Input:[]float).SumFloats<public>()<transacts> : float = 
        return Input.Reduce(Reduce.SumFloats, 0.0)

    # Makes an `array` with the `Element` added to the end
    (Input:[]t where t:type).Append<public>(Element:t)<computes>:[]t = 
        return Input + array{Element}
    
    (Input:[]t where t:subtype(comparable)).RemoveValue<public>(Element:t)<transacts>:[]t =
        var NewArray : []t = array{} 
        for (Value : Input):
            if (Value = Element) {}
            else:
                set NewArray = NewArray.Append(Value)
        return NewArray

    # to see if the array contains the element:
    (Input:[]t where t:subtype(comparable)).ContainsValue<public>(Element:t)<transacts>:logic =
        for (Value : Input):
            if (Value = Element):
                return true
        return false

    # Makes an `array` with the `Element` added to the beginning
    (Input:[]t where t:type).Prepend<public>(Element:t)<computes>:[]t = 
        return array{Element} + Input

    # Makes an `array` with the items in reverse order
    (Input:[]t where t:type).Reverse<public>()<computes>:[]t=
        for (Index -> _Unused : Input, Element := Input[Input.Length - Index - 1]). Element

    # Makes an `array` of ints sorted in order of the `SortOrder` enum
    (Input:[]int).SortInts<public>(SortOrder : Sort.orderBy)<transacts> : []int = 
        return Input.Sort(SortOrder, Array.Sort.CompareInts)

    # Makes an `array` of ints sorted in order of the `SortOrder` enum
    (Input:[]float).SortFloats<public>(SortOrder : Sort.orderBy)<transacts> : []float = 
        return Input.Sort(SortOrder, Array.Sort.CompareFloats)

    # Makes an `array` of items sorted in order of the `SortOrder` enum with a `Comparer` function
    # Usage: SortedItems := Sort(Items, Array.Sort.orderBy.Asc, Array.Sort.CompareFloats)
    (Input:[]t where t:type).Sort<public>(SortOrder : Sort.orderBy, Comparer: type{_(:t, :t)<transacts> : int})<transacts> : []t = 
        if (Input.Length > 1, Pivot := Input[Floor(Input.Length/2)]):
            Left := for(Item : Input, Comparer(Item, Pivot) < 0) do Item 
            Middle := for(Item : Input, Comparer(Item, Pivot) = 0) do Item
            Right := for(Item : Input, Comparer(Item, Pivot) > 0) do Item
            if(SortOrder = Sort.orderBy.Asc):
                Left.Sort(SortOrder, Comparer) + Middle + Right.Sort(SortOrder, Comparer)
            else: 
                Right.Sort(SortOrder, Comparer) + Middle + Left.Sort(SortOrder, Comparer)
        else:    
            Input 

    # Returns the item that is the largest by the `Comparer` function
    # Fails if `arary` is empty
    (Input:[]t where t:type).MaxBy<public>(Comparer: type{_(:t, :t)<transacts> : int})<transacts><decides> : t = 
        Input.Sort(Sort.orderBy.Desc, Comparer)[0]
    
    # Returns the item that is the smallest by the `Comparer` function
    # Fails if `arary` is empty
    (Input:[]t where t:type).MinBy<public>(Comparer: type{_(:t, :t)<transacts> : int})<transacts><decides> : t = 
        Input.Sort(Sort.orderBy.Asc, Comparer)[0]

    # Submodule for Reduce functions   
    Reduce<public> := module:
        SumInts<public>(Accumulator: int, Input: int)<computes>: int =
            return Accumulator + Input
        SumFloats<public>(Accumulator: float, Input: float)<computes>: float =
            return Accumulator + Input

    # Submodule for Sort functions   
    Sort<public> := module:
        orderBy<public> := enum{Asc, Desc}

        CompareInts<public>(A : int, B : int)<transacts>: int = 
            if(A < B) then -1
            else if(A > B) then 1
            else 0

        CompareFloats<public>(A : float, B : float)<transacts>: int = 
            if(A < B) then -1
            else if(A > B) then 1
            else 0