using { /Fortnite.com/Devices }
using { /Verse.org/Simulation }

birthday_device := class(creative_device):

    EVENTS : []date_data = array {
        date_data {Year := 2025, Month := 8, Day := 1, Hour := 15, Minute := 30},
        date_data {Year := 2025, Month := 8, Day := 3, Hour := 21, Minute := 30},
        date_data {Year := 2025, Month := 8, Day := 5, Hour := 22, Minute := 30},
        date_data {Year := 2025, Month := 8, Day := 7, Hour := 23, Minute := 30},
        date_data {Year := 2025, Month := 8, Day := 8, Hour := 15, Minute := 30},
        date_data {Year := 2025, Month := 8, Day := 9, Hour := 7, Minute := 30},
        date_data {Year := 2025, Month := 8, Day := 10, Hour := 15, Minute := 30},
        date_data {Year := 2025, Month := 8, Day := 20, Hour := 14, Minute := 00},
        date_data {Year := 2025, Month := 8, Day := 20, Hour := 18, Minute := 16},
        date_data {Year := 2025, Month := 9, Day := 16, Hour := 16, Minute := 30},
        date_data {Year := 2025, Month := 9, Day := 18, Hour := 13, Minute := 00},
        date_data {Year := 2025, Month := 10, Day := 15, Hour := 10, Minute := 01},
        date_data {Year := 2025, Month := 10, Day := 23, Hour := 18, Minute := 50},
        date_data {Year := 2025, Month := 10, Day := 28, Hour := 12, Minute := 30}
    }

    @editable
    Cinema : cinematic_sequence_device = cinematic_sequence_device {}

    OnBegin<override>()<suspends>:void=
        Sleep(0.0)

        for (Event: EVENTS):
            Print(Event)

        spawn:
            TrackTime()

    TrackTime()<suspends>:void=
        loop:
            CurrentEpochTime := GetSecondsSinceEpoch()
         
            # Print("Current epoch time is {CurrentEpochTime}")

            TodaysDate : date_data = date_data{}.ParseEpoch(CurrentEpochTime) #Get the current date
            Print(TodaysDate)
            for (Event: EVENTS):                
                ShouldTrigger := SameMinute(TodaysDate,Event)
                if (ShouldTrigger = true):
                    Cinema.Play()
            Sleep(60.0)