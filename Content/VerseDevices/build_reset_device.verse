using { /Fortnite.com/Devices }
using { /Verse.org/Simulation }
using { /Verse.org/Simulation/Tags }

ExplosiveForResetTag := class(tag) {}

build_reset_device := class(creative_device):

    var Explosives : []explosive_device = array {}

    @editable
    ResetBuildsButton : button_device = button_device {}
    @editable
    CoinRemoverDevice : coin_remover_device = coin_remover_device{}

    OnBegin<override>()<suspends>:void=
        Sleep(0.0)

        AllExplosives := FindCreativeObjectsWithTag(ExplosiveForResetTag {})

        for (ExplosiveItem : AllExplosives, ExplosiveDevice := explosive_device[ExplosiveItem]):
            set Explosives = Explosives + array {ExplosiveDevice}

        ResetBuildsButton.InteractedWithEvent.Subscribe(OnBuildReset)

    OnBuildReset(Agent:agent):void=
        CanActivate := CoinRemoverDevice.ActivateWithCost(Agent, 50)
        if (CanActivate = true):
            spawn:
                ResetBuilds(Agent)
        
    ResetBuilds(Agent:agent)<suspends>:void=
        for (Explosive : Explosives):
            Explosive.Reset()
            Explosive.Explode(Agent)