using { /Fortnite.com/Devices }
using { /Verse.org/Simulation }
using { /Verse.org/Simulation/Tags }

CoinRemoverTag := class(tag) {}

coin_remover_device := class(creative_device):

    var AllRemovers : []conditional_button_device = array {}
    var NextId : int = 0
    var RemoversCount : int = 1   

    @editable
    ErrorHudMessage : hud_message_device = hud_message_device {}

    OnBegin<override>()<suspends>:void=
        Sleep(0.0)

        AllCB := FindCreativeObjectsWithTag(CoinRemoverTag {})

        set RemoversCount = 0
        for (CB_Item : AllCB, CB := conditional_button_device[CB_Item]):
            set AllRemovers = AllRemovers + array { CB }
            set RemoversCount += 1

        Print("Coin removers count = {RemoversCount}")

    GetFreeCoinRemover():conditional_button_device=
        Id := NextId
        set NextId += 1
        if (Mod[NextId, RemoversCount] = 0):
            set NextId = 0
        if (CB := AllRemovers[Id]):
            return CB
        return conditional_button_device {}

    ActivateWithCost(Agent:agent, Cost:int):logic=
        GoldRemover := GetFreeCoinRemover()
        GoldRemover.SetItemCountRequired(0, Cost)
        if (GoldRemover.HasAllItems[Agent]):
            GoldRemover.Activate(Agent)
            return true
        ErrorHudMessage.Show(Agent, S2M("Missing coins"))
        return false

    S2M<localizes>(S:string) : message = "{S}"