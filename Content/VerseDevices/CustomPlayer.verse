using { /Fortnite.com/Devices }
using { /Verse.org/Simulation }
using { /UnrealEngine.com/Temporary/Diagnostics }
using { /Fortnite.com/Characters }
using { Array }

NUM_FINISHERS : int = 135
NUM_PARRIES : int = 7

CustomPlayer := class():

    MyAgentObj:agent

    var CurrentFinisher<private>:int = 0

    var AllUnlockedFinishers<private>:[int]int = map {0 => 1}
    var AllBadges<private>:[int]int = map {}

    # ----- Admin variables -----
    var InvisibleOn : logic = false
    var InvulnerableOn : logic = false
    var WasAdmin : logic = false

    # ----- Streak variables -----
    var CurrentStreak<private> : int = 0
    var MaxStreak<private>:int = 0

    # ----- Selected variables -----
    var CustomOrder<private> : []int = array {}
    var OrderId<private> : int = 0

    var PlasmaColor<private>: int = 0 #[0, 7]
    var SelectedPlasmaColor<private> : int = -1
    NUM_PLASMAS_COLORS<private>:int = 8

    # ----- Invulnerability on finishers -----
    var InvulnerableToken<private>: int = 0

    # ----- Parry variables -----
    var ParryCount<private> : int = 0
    var IsParrying<private> : logic = false
    PARRY_TIME : float = 0.6

    InitFromProfileData(SavedData: player_profile_data):void=
        set CurrentFinisher = SavedData.CurrentFinisherEquip_
        if (SavedData.WasAdmin_ = false):
            set AllUnlockedFinishers = SavedData.AllUnlockedFinishers_
        set MaxStreak = SavedData.StreakMax_
        set CustomOrder = SavedData.CustomOrder_
        set SelectedPlasmaColor = SavedData.PlasmaColorId_
        set AllBadges = SavedData.AllBadges_

    ChangeFinisher():void=
        Next : int = GetNextFinisher()
        set CurrentFinisher = Next
        ChangeFinisherEquip(MyAgentObj, Next)

    GetNextFinisher():int=
        if (CustomOrder.Length > 0):
            return GetNextFinisherCustomOrder()
        return GetNextFinisherOrdered()

    GetNextFinisherOrdered():int=
        for (i := CurrentFinisher+1..NUM_FINISHERS-1):
            if (Unlocked := AllUnlockedFinishers[i]):
                if (Unlocked > 0):
                    return i
        for (i := 0..CurrentFinisher):
            if (Unlocked := AllUnlockedFinishers[i]):
                if (Unlocked > 0):
                    return i
        return CurrentFinisher

    GetNextFinisherCustomOrder():int=
        Id := OrderId
        set OrderId += 1
        if (Mod[OrderId, CustomOrder.Length] = 0):
            set OrderId = 0
        if (Next := CustomOrder[Id]):
            return Next-1  # This array is 1-based
        return CurrentFinisher

    ChangeFinisher(Id:int):logic=
        Unlocked := IsFinisherUnlocked(Id)
        if (Unlocked = false):
            return false
        else:
            set CurrentFinisher = Id
            ChangeFinisherEquip(MyAgentObj, Id)
            set OrderId = 0
            return true

    GetFinisherId():int=
        return CurrentFinisher

    IsFinisherUnlocked(Id:int):logic=
        if(AllUnlockedFinishers[Id] > 0):
            return true
        return false

    GetFinisherCount(Id:int):int=
        if(Unlocked := AllUnlockedFinishers[Id]):
            return Unlocked
        return 0

    GetLastUsedDailySpin():float=
        Data := GetPersistenceData(MyAgentObj)
        return Data.LastUsedDailySpin

    SetLastUsedDailySpin(Time:float):void=
        SetLastUsedDailySpin(MyAgentObj, Time)

    # Return true if the finisher was unlocked previously
    UnlockedFinisher(Id:int):logic=
        Unlocked := IsFinisherUnlocked(Id)
        if (Unlocked = false):
            if (set AllUnlockedFinishers[Id] = 1):
                StoreUnlockFinisher(MyAgentObj, AllUnlockedFinishers)
        return Unlocked

    GetBadge(Id:int):void=
        if (set AllBadges[Id] = 1):
            StoreBadges(MyAgentObj, AllBadges)

    # ----- Admin powers -----
    SwitchInvisible():void=
        if (FC := MyAgentObj.GetFortCharacter[]):
            if (InvisibleOn = false):
                FC.Hide()
                set InvisibleOn = true
            else:
                FC.Show()
                set InvisibleOn = false

    SwitchInvulnerable():void=
        if (FC := MyAgentObj.GetFortCharacter[]):
            if (InvulnerableOn = false):
                FC.SetVulnerability(false)
                set InvulnerableOn = true
            else:
                FC.SetVulnerability(true)
                set InvulnerableOn = false

    UnlockAll():void=
        for (Id := 0..NUM_FINISHERS-1):
            if (set AllUnlockedFinishers[Id] = 1):
                # do nothing

        StoreUnlockFinisher(MyAgentObj, AllUnlockedFinishers)

    LoadFake():void=
        UnlockAll()
        set MaxStreak = 130
        MaxStreakUpdate(MyAgentObj, MaxStreak)
        SetFinishersDone(MyAgentObj, 56165)
        # set 

    SetAdmin():void=
        set WasAdmin = true
        StoreWasAdmin(MyAgentObj)

    ClearData():void=
        for (Id := 0..NUM_FINISHERS-1):
            if (set AllUnlockedFinishers[Id] = 0):
        StoreUnlockFinisher(MyAgentObj, AllUnlockedFinishers)
        MaxStreakUpdate(MyAgentObj, 0)
        SetFinishersDone(MyAgentObj, 0)

    # ----- Streak functions -----
    GetStreakCount():int=
        return CurrentStreak
    ClearStreak():void=
        set CurrentStreak = 0

    IncreaseStreak():int=
        set CurrentStreak += 1
        if (CurrentStreak > MaxStreak):
            set MaxStreak = CurrentStreak
            MaxStreakUpdate(MyAgentObj, CurrentStreak)
        return CurrentStreak

    # ----- Streak functions -----
    GetInvulnerableToken():int=
        return InvulnerableToken
    ReduceInvulnerable():void=
        set InvulnerableToken = Max(0, InvulnerableToken - 1)
    AddInvulnerable():void=
        set InvulnerableToken += 1
            
    # ----- Selected functions -----
    SetSelectedFinishers(Selected_: []int):void=
        Selected := GetOnlyUnlocked(Selected_)
        set CustomOrder = Selected
        set OrderId = 0
        UpdateCustomOrder(MyAgentObj, Selected)  # Store as 1-based array.
        ChangeFinisher()

    GetOnlyUnlocked(Selected: []int):[]int=
        var OnlyUnlocked : []int = array {}
        for (S : Selected):
            Unlocked := IsFinisherUnlocked(S-1)
            if (Unlocked = true):
                set OnlyUnlocked += array {S}
        return OnlyUnlocked

    GetSelected():[]int=
        return CustomOrder

    # ----- Plasma Color functions -----
    GetPlasmaColor():int=
        if (SelectedPlasmaColor >= 0):
            return SelectedPlasmaColor + 1
        X := PlasmaColor
        set PlasmaColor = Mod[PlasmaColor+1, NUM_PLASMAS_COLORS] or 0
        return X + 1 # Frames are [1, 8] 

    ChoosePlasmaColor(ColorId: int):void=
        NormalizedColorId := Mod[ColorId, NUM_PLASMAS_COLORS] or 0
        set SelectedPlasmaColor = NormalizedColorId
        SetSelectedPlasmaColor(MyAgentObj, NormalizedColorId)
    ClearPlasmaColor():void=
        set SelectedPlasmaColor = -1
        SetSelectedPlasmaColor(MyAgentObj, -1)

    # ----- Parry functions -----
    RechargeParry():void=
        set ParryCount = Min(2, ParryCount + 1)

    UseParry():logic=
        if (ParryCount > 0 and IsParrying = false):
            set ParryCount = Max(0, ParryCount - 1)
            set IsParrying = true
            spawn:
                RemoveParry()
            return true
        return false
    RemoveParry()<suspends>:void=
        Sleep(PARRY_TIME)
        set IsParrying = false
    GetIsParrying():logic=
        return IsParrying
    GetParryCount():int=
        return ParryCount
