
using { /Fortnite.com/Devices }
using { /Fortnite.com/Characters }
using { /Fortnite.com/Game }
using { /Verse.org/Simulation }
using { /UnrealEngine.com/Temporary/Diagnostics }
using { /UnrealEngine.com/Temporary/SpatialMath }
using { /Verse.org/Random }
using { /UnrealEngine.com/Temporary/UI }
using { /Verse.org/Simulation/Tags }
using { Array }

FixedCameraTag := class(tag) {}

custom_player_manager := class(creative_device):

    var PlayersMap:[player]CustomPlayer = map{}

    var Hour_:int = 0
    var Min_:int = 0

    PlayerSpawnedEvent<public> : event(agent) = event(agent){}
    PlayerLeaveEvent<public> : event(agent) = event(agent){}

    DEFAULT_ROTATION : rotation = rotation {}

    var Cameras: []gameplay_camera_fixed_point_device = array {}

    @editable
    var PlayerSpawners:[]player_spawner_device = array{}

    @editable
    FinisherManager:finisher_manager = finisher_manager {}

    # ----- Pickaxe damage -----
    @editable
    WeaponHolding : weapon_holding_device = weapon_holding_device {}

    # ----- Finisher Store -----
    FinisherUI : finisher_ui_device = finisher_ui_device {}

    @editable
    UnlockFinisherButtons : []button_device = array {}
    
    @editable
    EquipFinisherButtons : []button_device = array {}

    @editable
    FailMessage : hud_message_device = hud_message_device {}

    @editable
    EquipVfx : visual_effect_powerup_device = visual_effect_powerup_device {}
    @editable
    UltimateVfx : visual_effect_powerup_device = visual_effect_powerup_device {}
    @editable
    UnlockFinisherVfx : visual_effect_powerup_device = visual_effect_powerup_device {}
    @editable
    ParryVfx : visual_effect_powerup_device = visual_effect_powerup_device {}

    @editable
    GoldReturner : item_granter_device = item_granter_device {}

    @editable
    PPs: []post_process_device = array {}

    # ----- Accolades -----
    @editable
    FinisherUnlockedAcc : accolades_device = accolades_device{}

    @editable
    StreakAcc : accolades_device = accolades_device{}

    @editable
    TutorialPopUp : popup_dialog_device = popup_dialog_device {}

    # ----- Admin Devices -----
    @editable
    AdminMenu : popup_dialog_device = popup_dialog_device {}
    @editable
    LargeGoldGranter : item_granter_device = item_granter_device {}
    @editable
    TeleporterAdmin : teleporter_device = teleporter_device {}
    @editable
    AdminTrigger : trigger_device = trigger_device {}

    # ----- Test -----
    @editable
    ChangeFinisher : input_trigger_device = input_trigger_device {}

    @editable
    RunParryInput : input_trigger_device = input_trigger_device {}
    
    @editable
    RunFinisherInput : input_trigger_device = input_trigger_device {}

    @editable
    TestFinishers : input_trigger_device = input_trigger_device {}

    @editable
    UnlockAllInput : input_trigger_device = input_trigger_device {}

    @editable
    RunNukeInput : input_trigger_device = input_trigger_device {}

    # ----- Streak devices -----
    @editable 
    ElimManager : elimination_manager_device = elimination_manager_device {}

    # ----- Guard devices -----
    @editable
    GuardSpawner : guard_spawner_device = guard_spawner_device {}

    @editable
    FakeStatsAnalytics : analytics_device = analytics_device {}

    @editable
    ClearDataTrigger : trigger_device = trigger_device {}

    @editable
    UnlockGunFinisherButtons : []button_device = array {}

    @editable
    SusanoCamera : gameplay_camera_orbit_device = gameplay_camera_orbit_device {}

    @editable
    SusanoInputs : []input_trigger_device = array {}

    @editable
    NormalClass : class_and_team_selector_device = class_and_team_selector_device {}

    @editable
    CoinRemoverDevice : coin_remover_device = coin_remover_device{}

    @editable
    ErrorHudMessage : hud_message_device = hud_message_device {}

    OnBegin<override>()<suspends>:void=
        Sleep(0.0)

        for (PS : PlayerSpawners):
            PS.SpawnedEvent.Subscribe(OnPlayerSpawned)

        GetPlayspace().PlayerRemovedEvent().Subscribe(OnPlayerRemoved)

        spawn:
            EnsureRegisterAllPlayers()      
                
        ChangeFinisher.PressedEvent.Subscribe(OnChangeFinisher)
        RunParryInput.PressedEvent.Subscribe(OnParryClicked)
        RunFinisherInput.PressedEvent.Subscribe(OnTestPressed)
        TestFinishers.PressedEvent.Subscribe(OnTestAllPressed)
        UnlockAllInput.PressedEvent.Subscribe(OnUnlockAllPressed)

        spawn:
            RegisterSecondary()
        spawn:
            TrackTime()

    RegisterSecondary()<suspends>:void=
        Sleep(0.1)
        # Store
        for (Index -> Button : UnlockFinisherButtons):
            Button.InteractedWithEvent.SubscribeAgent(UnlockFinisher, (Index, false))
        for(Index -> Button : EquipFinisherButtons):
            Button.InteractedWithEvent.SubscribeAgent(EquipFinisher, Index)
        for(Index -> Button : UnlockGunFinisherButtons):
            Button.InteractedWithEvent.SubscribeAgent(UnlockFinisher, (Index+106, true))

        AdminMenu.RespondingButtonEvent.Subscribe(OnMenuItemPressed)
        ElimManager.EliminatedEvent.Subscribe(OnPlayerEliminated)
        AdminTrigger.TriggeredEvent.Subscribe(OnAdminGranted)

        GuardSpawner.DamagedEvent.Subscribe(OnGuardDamaged)
        GuardSpawner.SpawnedEvent.Subscribe(OnGuardSpawned)

        AllItems := FindCreativeObjectsWithTag(FixedCameraTag {})
        set Cameras = for (Item : AllItems, Camera := gameplay_camera_fixed_point_device[Item]) . Camera
        Print("Cameras = {Cameras.Length}")

    TrackTime()<suspends>:void=
        MINUTES_TO_WAIT : int = 1
        loop:
            Sleep(60.0 * MINUTES_TO_WAIT)
            set Min_ += MINUTES_TO_WAIT
            if (Min_ >= 60):
                set Min_ = 0
                set Hour_ += 1
            for (Player -> CP : PlayersMap):
                spawn:
                    UpdateServerLifeUI(Player, Hour_, Min_)

    EnsureRegisterAllPlayers()<suspends>:void=
        Sleep(1.0)
        RegisterAllPlayers()

    RegisterAllPlayers():void=
        Players := GetPlayspace().GetPlayers()
        for (Player : Players):
            OnPlayerSpawned(Player)

    OnTestPressed(Agent:agent):void=
        if:
            Player := player[Agent],
            CP := PlayersMap[Player]
        then:
            RunFinisher(Agent, Agent, Player, CP, true)

    OnUnlockAllPressed(Agent:agent):void=
        if:
            Player := player[Agent],
            CP := PlayersMap[Player]
        then:
            CP.UnlockAll()

    OnTestAllPressed(Agent:agent):void=
        if:
            Player := player[Agent],
            CP := PlayersMap[Player]
        then:
            FinisherId := CP.GetFinisherId()
            ParryId := GetParryId()
            spawn:
                FinisherManager.TestFinishers(Agent, FinisherId, CP)
                # FinisherManager.TestParries(Agent, ParryId, CP)
              
    OnPlayerSpawned(Agent:agent): void=
        spawn:
            RemoveSusanoDevices(Agent)        
        CheckAndRegister(Agent)

    CheckAndRegister(Agent:agent):void=
        if:
            PlayerObj := player[Agent]
        then:
            if (PlayerExists := PlayersMap[PlayerObj]):
                # do nothing
            else:
                spawn:
                    ForceRegisterPlayer(Agent)

    RemoveSusanoDevices(Agent:agent)<suspends>:void=
        Sleep(1.0)
        NormalClass.ChangeClass(Agent)
        SusanoCamera.RemoveFrom(Agent)
        for (Input:SusanoInputs):
            Input.Unregister(Agent)
        AddUI(Agent)

        for(Camera:Cameras):
            Camera.RemoveFrom(Agent)
        for(PP:PPs):
            PP.BlendOut(Agent)

    CheckRegister(Agent:agent)<suspends>:void=
        if:
            PlayerObj := player[Agent]
        then:
            if (PlayerExists := PlayersMap[PlayerObj]):
                # do nothing
            else:
                spawn:
                    ForceRegisterPlayer(Agent)

    ForceRegisterPlayer(Agent:agent)<suspends>:void=
        if:
            PlayerObj := player[Agent]
        then:
            NewCustomPlayer:CustomPlayer = CustomPlayer{MyAgentObj:=Agent}
            SavedData := LoadPersistenceData(Agent)
            NewCustomPlayer.InitFromProfileData(SavedData)
            if:
                set PlayersMap[PlayerObj] = NewCustomPlayer
            if (FC := Agent.GetFortCharacter[]):
                FC.DamagedEvent().Subscribe(OnDamaged)
                FC.DamagedShieldEvent().Subscribe(OnDamaged)
            spawn:
                RegisterUI(PlayerObj, NewCustomPlayer)
            spawn:
                NotifyGuardChangedPlayer()
            TutorialPopUp.Show(Agent)
            spawn:
                TrackTimedAbility(Agent)

    RegisterUI(PlayerObj : player, NewCustomPlayer : CustomPlayer)<suspends>:void=
        Sleep(1.0)
        PlayerSpawnedEvent.Signal(PlayerObj)
        FinisherUI.Init(PlayerObj, Hour_, Min_)
        RefreshFinisherUI(PlayerObj, NewCustomPlayer)

    RemoveUI(Agent:agent):void=
        if (Player := player[Agent]):
            FinisherUI.RemoveFromPlayer(Player)

    AddUI(Agent:agent):void=
        CP_O := GetCustomPlayer(Agent)
        if (Player := player[Agent], CP := CP_O?):
            FinisherUI.Init(Player, Hour_, Min_)
            RefreshFinisherUI(Player, CP)

    OnPlayerRemoved(Player:player):void =
        if (OldCustomPlayer:CustomPlayer = PlayersMap[Player]):
            var TempAllPlayersMap:[player]CustomPlayer = map{}
            for (Key -> Value: PlayersMap, Key <> Player):
                set TempAllPlayersMap = ConcatenateMaps(TempAllPlayersMap, map{Key => Value})
            set PlayersMap = TempAllPlayersMap
            
            PlayerLeaveEvent.Signal(Player)
            spawn:
                NotifyGuardChangedPlayer()
            FinisherUI.OnPlayerLeave(Player)

    OnDamaged(Result: damage_result):void=
        PICKAXE_DAMAGE := 20.0
        if:
            DamagedPlayer := fort_character[Result.Target],
            TargetAgent := DamagedPlayer.GetAgent[],
            Ins := Result.Instigator?,
            InstigatorAgent := Ins.GetInstigatorAgent[],
            Player := player[InstigatorAgent],
            CP := PlayersMap[Player]
        then:
            if (Result.Amount = PICKAXE_DAMAGE):
                IsHoldingWeapon := WeaponHolding.IsHoldingWeapon(InstigatorAgent)
                if (IsHoldingWeapon = true):
                    return
                if (TargetPlayer := player[TargetAgent], TargetCP := PlayersMap[TargetPlayer]):
                    IsParrying := TargetCP.GetIsParrying()
                    if (IsParrying = true):
                        RunParry(InstigatorAgent, TargetAgent, CP, TargetCP)
                        return

                RunFinisher(InstigatorAgent, TargetAgent, Player, CP, true)
     
    RunFinisher(InstigatorAgent: agent, TargetAgent:agent, InstigatorPlayer:player, CP: CustomPlayer, IsPlayer:logic):void=
        FinisherId := CP.GetFinisherId()
        spawn:
            FinisherManager.RunFinisher(InstigatorAgent, TargetAgent, FinisherId, CP)
        AddToFinishersDone(InstigatorAgent)
        if (IsPlayer = true):
            CurStreak := CP.IncreaseStreak()
            FinisherUI.UpdateStreak(InstigatorPlayer, CP.GetStreakCount())
            spawn:
                GrantStreakXP(CurStreak, InstigatorAgent)
            UnlockedFinisherId := UnlockByStreak(CurStreak)
            if (UnlockedFinisherId <> 0):
                Unlocked := CP.UnlockedFinisher(UnlockedFinisherId)
                if (Unlocked = false):
                    FinisherUnlockedAcc.Award(InstigatorAgent)
                    EquipFinisher(InstigatorAgent, UnlockedFinisherId)
                    GoldReturner.GrantItemIndex(InstigatorAgent, 0)

            if (CurStreak > 0 and Mod[CurStreak,30] = 0):
                RunNukeInput.Register(InstigatorAgent)

        CP.RechargeParry()
        spawn:
            UpdateParryUI(InstigatorAgent)

    UnlockByStreak(CurStreak:int):int=
        if (CurStreak=5):
            return 120
        if (CurStreak=10):
            return 121
        if (CurStreak=15):
            return 122
        if (CurStreak=20):
            return 123
        return 0

    GrantStreakXP(CurStreak:int, Agent:agent)<suspends>:void=
        if (CurStreak > 0 and Mod[CurStreak, 3] = 0):
            StreakAcc.Award(Agent)

    GrantNuke(Agent: agent):void=
        RunNukeInput.Register(Agent)

    RunParry(InstigatorAgent: agent, TargetAgent:agent, CP: CustomPlayer, TargetCP: CustomPlayer):void=
        ParryId := GetParryId()
        spawn:
            FinisherManager.RunParry(InstigatorAgent, TargetAgent, ParryId, CP, TargetCP)

    GetParryId():int=
        return GetRandomInt(0, NUM_PARRIES - 1)

    OnChangeFinisher(Agent:agent):void=
        if (Player := player[Agent], CP := PlayersMap[Player]):
            CP.ChangeFinisher()
            RefreshFinisherUI(Player, CP)


    RefreshFinisherUI(Player: player, CP: CustomPlayer):void=
        FinisherId := CP.GetFinisherId()
        Count_ := CP.GetFinisherCount(FinisherId)
        FinisherUI.UpdateLevel(Player, finisher_stats {Level := FinisherId + 1, Count := Count_})

    PlayerNameToMessage<localizes>(Agent:agent):message="{Agent}"

    # Also change the function in finisher_ui_device
    # {param} FinisherId backend finisher number (same as persistence) or public finisher number - 1. 
    IsUltimateFinisher(FinisherId:int):logic=
        case(FinisherId):
            29 => true
            38 => true
            39 => true
            52 => true
            53 => true
            67 => true
            74 => true
            75 => true
            87 => true
            88 => true
            95 => true
            96 => true
            98 => true
            99 => true
            105 => true
            123 => true
            _ => false

    # ----- Finisher Store -----
    UnlockFinisher(Agent:agent, Data:tuple(int, logic)):void=
        Id:= Data(0)
        NoCost := Data(1)
        if (Player := player[Agent], CP := PlayersMap[Player]):
            WasUnlocked := CP.IsFinisherUnlocked(Id+1)
            if (WasUnlocked = true):
                EquipFinisher(Agent, Id+1)
                return
            if (NoCost = true):
                CP.UnlockedFinisher(Id+1)
                EquipFinisher(Agent, Id+1)
                return
            GoldRemover := CoinRemoverDevice.GetFreeCoinRemover()
            Price := GetFinisherPrice(Id+1)
            GoldRemover.SetItemCountRequired(0, Price)
            if (GoldRemover.HasAllItems[Player]):
                GoldRemover.Activate(Player)
                FinisherUnlockedAcc.Award(Agent)
                UnlockFinisherVfx.Pickup(Agent)
                CP.UnlockedFinisher(Id+1)
                EquipFinisher(Agent, Id+1)
            else:
                ErrorHudMessage.Show(Agent, S2M("Missing coins"))

    S2M<localizes>(S:string) : message = "{S}"

    # Also change the function in finisher_ui_device
    # {param} FinisherId backend finisher number (same as persistence) or public finisher number - 1. 
    GetFinisherPrice(FinisherId:int):int=
        IsUltimate := IsUltimateFinisher(FinisherId)
        if (IsUltimate = true):
            return 300
        return 100

    EquipFinisher(Agent:agent, Id: int):void=
        if (Player := player[Agent], CP := PlayersMap[Player]):
            Changed := CP.ChangeFinisher(Id)
            RefreshFinisherUI(Player, CP)
            if (Changed = false):
                FailMessage.Show(Agent)
            else:
                IsUltimate := IsUltimateFinisher(Id)
                if (IsUltimate = true):
                    UltimateVfx.Pickup(Agent)
                else:
                    EquipVfx.Pickup(Agent)

    # ----- Admin powers -----
    OnMenuItemPressed(Value:tuple(agent, int)):void=
        Instigator := Value(0)
        ButtonPressed := Value(1)
        if (Player := player[Instigator], CP := PlayersMap[Player]):
            case(ButtonPressed):
                5 =>
                    CP.SwitchInvisible()
                1 =>
                    CP.SwitchInvulnerable()
                2 =>
                    TeleportToNearPlayer(Instigator)
                3 =>
                    CP.UnlockAll()
                4 =>
                    LargeGoldGranter.GrantItem(Instigator)
                _ =>
                    return

    TeleportToNearPlayer(Agent:agent):void=
        var Alive : []agent = array {}
        for (Key -> Value : PlayersMap):
            if (Key.IsActive[] and Key <> Agent):
                set Alive = Alive + array {Key}
        
        if (FC := Agent.GetFortCharacter[]):
            Target_O := GetNearest(Alive, FC)
            if (Target := Target_O?):
                TeleportPlayerTo(Agent, Target)
    
    GetNearest(Alive:[]agent, FC:fort_character):?fort_character=
        var ClosestDistance : float = 10000000.0
        var ClosestAgent : ?fort_character = false
        for (Target : Alive):
            if (TargetFC := Target.GetFortCharacter[]):
                D := Distance(FC.GetTransform().Translation, TargetFC.GetTransform().Translation)
                if (D < ClosestDistance):
                    set ClosestDistance = D
                    set ClosestAgent = option {TargetFC}
        return ClosestAgent

    TeleportPlayerTo(Agent:agent, TargetFC:fort_character):void=
        PlayerTransform := TargetFC.GetTransform()
        Forward := PlayerTransform.Rotation.GetLocalForward()
        NewLoc := PlayerTransform.Translation + Forward * -120.0
        if (TeleporterAdmin.TeleportTo[NewLoc, DEFAULT_ROTATION]):
            TeleporterAdmin.Teleport(Agent)

    OnAdminGranted(AgentO : ?agent):void=
        if (Agent := AgentO?, Player := player[Agent], CP := PlayersMap[Player]):
            CP.SetAdmin()

    ClearData(Agent:agent):void=
        if (Player := player[Agent], CP := PlayersMap[Player]):
            CP.ClearData()
            ClearDataTrigger.Trigger(Agent)

    # ----- Daily Spin -----
    SetDailySpin(Agent:agent, Time:float):void=
        if (Player := player[Agent], CP := PlayersMap[Player]):
            CP.SetLastUsedDailySpin(Time)
        
    GetDailySpin(Agent:agent):float=
        if (Player := player[Agent], CP := PlayersMap[Player]):
            return CP.GetLastUsedDailySpin()
        return 0.0

    # ----- Finishers Selector -----
    StoreSelectedFinishers(Player:player, Selected: []int):void=
        if (CP := PlayersMap[Player]):
            CP.SetSelectedFinishers(Selected)
            UnlockFinisherVfx.Pickup(Player)

    # ----- Streak functions -----
    OnPlayerEliminated(Agent:agent):void=
        if (Player := player[Agent], CP := PlayersMap[Player]):
            CP.ClearStreak()
            FinisherUI.UpdateStreak(Player, CP.GetStreakCount())

    # ----- Plasma Colors functions -----
    ChooseCustomColor(Player:player, ColorId : int):void=
        if (CP := PlayersMap[Player]):
            if (ColorId >= 0):
                CP.ChoosePlasmaColor(ColorId)
            else:
                CP.ClearPlasmaColor()

    # ----- Fake stats functions -----
    LoadFakeStats(Agent:agent):void=
        if (Player := player[Agent]):
            if (CP := PlayersMap[Player]):
                # Exists
                CP.LoadFake()
                FakeStatsAnalytics.Submit(Agent)
            else:
                # Not exists

    UnlockAllFinishers(Agent:agent):void=
        if (Player := player[Agent], CP := PlayersMap[Player]):
            CP.UnlockAll()

    # ----- Parry functions -----
    TrackTimedAbility(Agent:agent)<suspends>:void=
        PARRY_TIME_TO_GRANT := 9.0
        loop:
            Sleep(PARRY_TIME_TO_GRANT)
            CP_O := GetCustomPlayer(Agent)
            if (CP := CP_O?):
                CP.RechargeParry()
                spawn:
                    UpdateParryUI(Agent)
            else:
                break

    OnParryClicked(Agent:agent):void=
        if (Player := player[Agent], CP := PlayersMap[Player]):
            Usable := CP.UseParry()
            if (Usable = true):
                ParryVfx.Pickup(Agent)
                spawn:
                    UpdateParryUI(Agent)

    UpdateParryUI(Agent:agent)<suspends>:void=
        if (Player := player[Agent], CP := PlayersMap[Player]):
            Count := CP.GetParryCount()
            FinisherUI.UpdateParryCount(Player, Count)

    UpdateServerLifeUI(Player:player, H:int, M:int)<suspends>:void=
        FinisherUI.UpdateServerTime(Player, H, M)

    # ----- Guard spanwer functions -----
    OnGuardSpawned(Agent:agent):void=
        return

    OnGuardDamaged(Result: device_ai_interaction_result):void=
        PICKAXE_DAMAGE := 20.0
        if:
            Target := Result.Target?,
            InstigatorAgent := Result.Source?,
            Player := player[InstigatorAgent],
            CP := PlayersMap[Player]
        then:
            GuardHit := GuardHitWithPickaxe(Target)
            if (GuardHit = true):
                RunFinisher(InstigatorAgent, Target, Player, CP, false)
                # RunParry(InstigatorAgent, Target, CP, CustomPlayer {MyAgentObj := Target})


    GuardHitWithPickaxe(Agent:agent):logic=
        if:
            TargetFC := Agent.GetFortCharacter[]
        then:
            H := TargetFC.GetHealth()
            if (TargetFC.GetHealth() < 80.0 or TargetFC.GetHealth() > 80.05):
                return false
            return true
        return false

    NotifyGuardChangedPlayer()<suspends>:void=
        NumPlayers := PlayersMap.Length
        if (NumPlayers > 1):
            GuardSpawner.Disable()
        else:
            GuardSpawner.Enable()

    # ----- Getters -----
    GetCustomPlayer(Agent : agent):?CustomPlayer=
        if (Player := player[Agent], CP := PlayersMap[Player]):
            return option {CP}
        return false

    GetNumPlayers():int=
        return PlayersMap.Length

    GetAllPlayers():[]CustomPlayer=
        var All : []CustomPlayer = array {}
        for (Key -> Value : PlayersMap):
            set All = All.Append(Value)
        return All

    GetAllPlayersAlive():[]player=
        var Ret : []player = array {}
        for (Key -> Value: PlayersMap):
            if (Key.IsActive[]):
                set Ret += array {Key}
        return Ret
