date_data := class:
    # Default values for the date (January 1, 1970, 00:00:00)
    Year : int = 1970
    Month : int = 1
    Day : int = 1
    Hour : int = 0
    Minute : int = 0
    Second : int = 0

    SECONDS_IN_A_DAY : int = 86400
    SECONDS_IN_AN_HOUR : int = 3600
    SECONDS_IN_A_MINUTE : int = 60

    # Function that initializes the date using epoch seconds
    ParseEpoch(EpochSeconds : float):date_data= 

        var DateData : date_data = date_data{}

        if:
            IntEpochSeconds : int = Round[EpochSeconds]
            EpochDays := IntEpochSeconds / SECONDS_IN_A_DAY
            DaysSinceEpoch : int = Floor(EpochDays)
        then:
            RemainingSeconds := IntEpochSeconds - (DaysSinceEpoch * SECONDS_IN_A_DAY)
            DateVal : tuple(int,int,int) = Self.ConvertDaysToDate(DaysSinceEpoch)
            TimeVal : tuple(int,int,int) = Self.ConvertSecondsToTime(RemainingSeconds)
            set DateData = date_data{
                Year := DateVal(0), Month := DateVal(1), Day := DateVal(2),
                Hour := TimeVal(0), Minute := TimeVal(1), Second := TimeVal(2)
            }

        return DateData


    # Method to convert seconds left to time
    ConvertSecondsToTime(EpochSeconds : int):tuple(int,int,int)=
        var Time : tuple(int,int,int) = (0,0,0)
        if:
            Hours : int = Floor(EpochSeconds / 3600)
            Minutes : int = Floor((EpochSeconds - (Hours * 3600)) / 60)
            Seconds : int = EpochSeconds - (Hours * 3600) - (Minutes * 60)
        then:
            set Time = (Hours, Minutes, Seconds)
        
        return Time

    # Method to convert days since epoch to year, month, and day
    ConvertDaysToDate(DaysSinceEpoch : int) : tuple(int, int, int) =
        var CurrentYear : int = 1970
        var DaysRemaining : int = DaysSinceEpoch

        loop:
            DayCount := Self.DaysInYear(CurrentYear)
            if(DaysRemaining >= DayCount):
                set DaysRemaining -= Self.DaysInYear(CurrentYear)
                set CurrentYear += 1
            else:
                break

        var DaysInMonths : []int = Self.GetDaysInMonths(CurrentYear)
        var CurrentMonth : int = 1

        loop:
            if(DaysRemaining >= DaysInMonths[CurrentMonth - 1]):
                if:
                    set DaysRemaining -= DaysInMonths[CurrentMonth - 1]
                set CurrentMonth += 1
            else:
                break

        return (CurrentYear, CurrentMonth, DaysRemaining + 1)

    # Method to determine the number of days in a year
    DaysInYear(YearVal : int)<transacts>: int =
        return if (IsLeapYear(YearVal)?) then 366 else 365

    # Method to determine if a year is a leap year
    IsLeapYear(YearVal : int)<transacts>: logic =
        return if((Mod[YearVal,4] = 0) and (Mod[YearVal,100] <> 0) or (Mod[YearVal,400] = 0)) then true else false

    # Method to get the number of days in each month for a given year
    GetDaysInMonths(YearVal : int)<transacts>: []int =
        return if (Self.IsLeapYear(YearVal)?) then array{31, 29, 31, 30, 31, 30, 31, 31, 30, 31, 30, 31} else array{31, 28, 31, 30, 31, 30, 31, 31, 30, 31, 30, 31}

    ToEpochSeconds()<transacts>:float=
        var EpochDays:int = 0
        var YearRes : int = 1970
        loop:
            if (YearRes < Self.Year):
                set EpochDays += Self.DaysInYear(YearRes)
                set YearRes += 1
            else:
                break
        
        var DaysInMonths:[]int = Self.GetDaysInMonths(Self.Year)
        var MonthRes :int = 1
        loop:
            if (MonthRes < Self.Month, NewDay := DaysInMonths[MonthRes - 1]):
                set EpochDays += NewDay
                set MonthRes += 1
            else:
                break
        
        set EpochDays += Self.Day - 1
        
        var EpochSeconds : float = EpochDays * Self.SECONDS_IN_A_DAY * 1.0
        set EpochSeconds += Self.Hour * Self.SECONDS_IN_AN_HOUR * 1.0
        set EpochSeconds += Self.Minute * Self.SECONDS_IN_A_MINUTE * 1.0
        set EpochSeconds += Self.Second * 1.0
        
        return EpochSeconds

    # Add a specified number of days to the current date
    AddDays(Days:int):date_data=
        var NewEpochSeconds:float = Self.ToEpochSeconds() + (Days * Self.SECONDS_IN_A_DAY * 1.0)
        return Self.ParseEpoch(NewEpochSeconds)

    # Add a specified number of hours to the current date
    AddHours(Hours:int):date_data=
        var NewEpochSeconds:float = Self.ToEpochSeconds() + (Hours * Self.SECONDS_IN_AN_HOUR * 1.0)
        return Self.ParseEpoch(NewEpochSeconds)

    # Calculate the difference in days between two dates
    DaysBetween(OtherDate:date_data):?int=
        var DiffSeconds:float = Abs(Self.ToEpochSeconds() - OtherDate.ToEpochSeconds())
        return option{Floor[DiffSeconds / (Self.SECONDS_IN_A_DAY * 1.0)]}

    # Check if the current date is before another date
    IsBefore(OtherDate:date_data)<decides><transacts>:void=
        Self.ToEpochSeconds() < OtherDate.ToEpochSeconds()

    # Check if the current date is after another date
    IsAfter(OtherDate:date_data)<decides><transacts>:void=
        Self.ToEpochSeconds() > OtherDate.ToEpochSeconds()


    # Get the day of the week (0 = Sunday, 1 = Monday, ..., 6 = Saturday)
    DayOfWeek()<decides><transacts>:int=
        var DaysFromEpoch:int = Floor[Self.ToEpochSeconds() / (Self.SECONDS_IN_A_DAY * 1.0)]
        return Mod[(DaysFromEpoch + 4), 7]  # January 1, 1970 was a Thursday (4)

    # Get the day of the year (1-366)
    DayOfYear():int=
        var DaysInMonths:[]int = Self.GetDaysInMonths(Self.Year)
        var DayCount:int = Self.Day
        var MonthRes:int = 1
        loop:
            if (MonthRes < Self.Month, NewDay := DaysInMonths[MonthRes - 1]):
                set DayCount += NewDay
                set MonthRes += 1
            else:
                break
        return DayCount

    # Check if the current year is a leap year
    IsCurrentYearLeap()<decides><transacts>:void=
        Self.IsLeapYear(Self.Year)?


# Utility functions

# Convert a date to a string
ToString(Date : date_data):string=
    return "{Date.Year}-{Date.Month}-{Date.Day} {Date.Hour}:{Date.Minute}:{Date.Second}"

ToPrettyTime(Date: date_data): string =
    return "{Date.Day}/{Date.Month} at {Date.Hour}:{Date.Minute}"

# Print the date
Print(Date : date_data):void=
    Print(ToString(Date))

SameMinute(A: date_data, B: date_data):logic=
    if (A.Year <> B.Year):
        return false
    if (A.Month <> B.Month):
        return false
    if (A.Day <> B.Day):
        return false
    if (A.Hour <> B.Hour):
        return false
    if (A.Minute <> B.Minute):
        return false
    return true