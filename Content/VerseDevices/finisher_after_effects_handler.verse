
using { /Fortnite.com/Devices }
using { /Fortnite.com/Game }
using { /Verse.org/Simulation }
using { /Fortnite.com/Characters }
using { /UnrealEngine.com/Temporary/Diagnostics }
using { /UnrealEngine.com/Temporary/SpatialMath }
using { /Verse.org/Random }
using { /UnrealEngine.com/Assets }
using { /Verse.org/Colors }
using { Array }

finisher_effects := class <concrete>():

    @editable var Teleporter : teleporter_device = teleporter_device {}
    @editable var Explosive : explosive_device = explosive_device {}
    @editable var InfiniteDomainAsset : creative_prop_asset = DefaultCreativePropAsset
    @editable var InfiniteCamera1 : cinematic_sequence_device = cinematic_sequence_device {}
    @editable var InfiniteCamera2 : cinematic_sequence_device = cinematic_sequence_device {}

    @editable var SukunaDomainAsset : creative_prop_asset = DefaultCreativePropAsset
    @editable var DomainExplosive : explosive_device = explosive_device {}
    @editable var GiantCamera1 : cinematic_sequence_device = cinematic_sequence_device {}
    @editable var GiantCamera2 : cinematic_sequence_device = cinematic_sequence_device {}
    @editable var GiantExplosive : explosive_device = explosive_device {}

    @editable var Luffy1Camera : gameplay_camera_fixed_point_device = gameplay_camera_fixed_point_device {}
    @editable var Luffy5CameraAtk : gameplay_camera_fixed_point_device = gameplay_camera_fixed_point_device {}
    @editable var Luffy5CameraVic : gameplay_camera_fixed_point_device = gameplay_camera_fixed_point_device {}
    @editable var Luffy4RedHawkSound : audio_player_device = audio_player_device {}
    @editable var FugaCamera : gameplay_camera_fixed_point_device = gameplay_camera_fixed_point_device {}

    @editable var Toji1Camera : gameplay_camera_fixed_point_device = gameplay_camera_fixed_point_device {}
    @editable var ThinkCamera : gameplay_camera_fixed_point_device = gameplay_camera_fixed_point_device {}

finisher_data := struct:
    Cinema<public>: cinematic_sequence_device
    Sound<public>: audio_player_device
    Duration<public>:float
    PlayerTransform<public> : transform
    Instigator<public> : agent
    Target<public> : agent
    MoverTransform<public> : transform
    DevicesId<public> : int
    PlasmaColorCinema<public> : cinematic_sequence_device
    PlasmaColorId<public> : int
    TestOffset<public> : vector3

finisher_after_effects_handler := class (creative_device):

    DebugDraw:debug_draw = debug_draw{Channel := my_debug_draw}

    DEFAULT_ROTATION : rotation = rotation {}
    HIDE_LOCATION_SOUND<private>: vector3 = vector3 {Z := -900.0}
    FinishersWithEffect : [int]int = map { 
        13 => 1,
        14 => 1,
        15 => 1,
        16 => 1,
        17 => 1,
        18 => 1,
        19 => 1,
        20 => 1,
        21 => 1,
        22 => 1,
        23 => 1,
        24 => 1,
        25 => 1,
        26 => 1,
        28 => 1,
        29 => 1,
        35 => 1,
        37 => 1,
        38 => 1,
        39 => 1,
        40 => 1,
        41 => 1,
        42 => 1,
        43 => 1,
        44 => 1,
        45 => 1,
        47 => 1,
        48 => 1,
        49 => 1,
        50 => 1,
        51 => 1,
        52 => 1,
        53 => 1,
        54 => 1,
        55 => 1,
        57 => 1,
        59 => 1,
        60 => 1,
        63 => 1,
        64 => 1,
        65 => 1,
        66 => 1,
        67 => 1,
        68 => 1,
        74 => 1,
        75 => 1,
        78 => 1,
        79 => 1,
        80 => 1,
        81 => 1,
        82 => 1,
        84 => 1,
        85 => 1,
        87 => 1,
        88 => 1,
        89 => 1,
        90 => 1,
        91 => 1,
        92 => 1,
        93 => 1,
        94 => 1,
        95 => 1,
        96 => 1,
        97 => 1,
        98 => 1,
        99 => 1,
        102 => 1,
        103 => 1,
        104 => 1,
        105 => 1,
        106 => 1,
        107 => 1,
        108 => 1,
        109 => 1,
        110 => 1,
        111 => 1,
        112 => 1,
        113 => 1,
        114 => 1,
        115 => 1,
        116 => 1,
        117 => 1,
        118 => 1,
        119 => 1,
        120 => 1,
        121 => 1,
        122 => 1,
        123 => 1,
        124 => 1,
        125 => 1,
        127 => 1,
        128 => 1,
        129 => 1,
        130 => 1,
        131 => 1,
        132 => 1,
        133 => 1,
        134 => 1
    }

    # ----- Device after effects -----
    @editable
    EffectDevices : []finisher_effects = array {}

    @editable
    PP_Spidersense : post_process_device = post_process_device {}
    @editable
    PP_BlackFlash3 : post_process_device = post_process_device {}

    @editable
    SharinganHud: hud_message_device = hud_message_device {}
    @editable
    SharinganItachiHud: hud_message_device = hud_message_device {}
    @editable
    SharinganKakashiHud: hud_message_device = hud_message_device {}
    @editable
    SharinganSasukeHud: hud_message_device = hud_message_device {}
    @editable
    SharinganShusuiHud: hud_message_device = hud_message_device {}
    
    SelectDevices(Id: int):?finisher_effects =
        if :
            Next := EffectDevices[Id]
        then:
            return option {Next}
        return false

    HasEffect(FinisherId:int):logic=
        if (FinishersWithEffect[FinisherId] = 1):
            return true
        return false

    HandleFinisher(FinisherId:int, Data:finisher_data)<suspends>:void=
        Devices_O := SelectDevices(Data.DevicesId)
        case(FinisherId):
            13 =>
                ActivateNinja1(Devices_O, Data)
            14 =>
                ActivateNinja2(Devices_O, Data)
            15 =>
                ActivateNinja3(Devices_O, Data)
            16 =>
                ActivateNinja4(Devices_O, Data)
            17 =>
                ActivateNinja5(Devices_O, Data)
            18 =>
                ActivateNinja6(Devices_O, Data)
            19 =>
                ActivateHero1(Devices_O, Data)
            20 =>
                ActivateHero2(Devices_O, Data)
            21 =>
                ActivateHero3(Devices_O, Data)
            22 =>
                ActivateHero4(Devices_O, Data)
            23 =>
                ActivateHero5(Devices_O, Data)
            24 =>
                ActivateHero6(Devices_O, Data)
            25 =>
                ActivateKi1(Devices_O, Data)
            26 =>
                ActivateSamurai1(Devices_O, Data)
            28 =>
                ActivateKi2(Devices_O, Data)
            29 =>
                ActivateInfiniteVoid(Devices_O, Data)
            35 =>
                ActivateWWE1(Devices_O, Data)
            37 =>
                ActivateSamurai3(Devices_O, Data)
            38 =>
                ActivateSukunaDomain(Devices_O, Data)
            39 =>
                ActivateGenkidama(Devices_O, Data)
            40 =>
                ActivateWater1(Devices_O, Data)
            41 =>
                ActivateFire1(Devices_O, Data)
            42 =>
                ActivateEarth1(Devices_O, Data)
            43 =>
                ActivateAir1(Devices_O, Data)
            44 =>
                ActivateWolverine1(Devices_O, Data)
            45 =>
                ActivateWolverine2(Devices_O, Data)
            47 =>
                ActivateWolverine4(Devices_O, Data)
            48 =>
                ActivateKi3(Devices_O, Data)
            49 =>
                ActivateKi4(Devices_O, Data)
            50 =>
                ActivateKi5(Devices_O, Data)
            51 =>
                ActivateKi6(Devices_O, Data)
            52 =>
                ActivateGiant1(Devices_O, Data)
            53 =>
                ActivateGiant1(Devices_O, Data)
            54 =>
                ActivateSpiderman1(Devices_O, Data)
            55 =>
                ActivateSpiderman2(Devices_O, Data)
            57 =>
                ActivateSpiderman4(Devices_O, Data)
            59 =>
                ActivateSW1(Devices_O, Data)
            60 =>
                ActivateSW2(Devices_O, Data)
            63 => 
                ActivateLuffy1(Devices_O, Data)
            64 => 
                ActivateLuffy2(Devices_O, Data)
            65 =>
                ActivateLuffy3(Devices_O, Data)
            66 =>
                ActivateLuffy4(Devices_O, Data)
            67 =>
                ActivateLuffy5(Devices_O, Data)
            68 =>
                ActivateHollowPurple(Devices_O, Data)
            74 => 
                ActivateFuga(Devices_O, Data)
            75 =>
                ActivateToji1(Devices_O, Data)
            78 =>
                ActivateFlash(Devices_O, Data)
            79 =>
                ActivateDemonSlayer1(Devices_O, Data)
            80 =>
                ActivateDemonSlayer2(Devices_O, Data)
            81 =>
                ActivateDemonSlayer3(Devices_O, Data)
            82 =>
                ActivateDemonSlayer4(Devices_O, Data)
            84 => 
                ActivateMK2(Devices_O, Data)
            85 =>
                ActivateMK3(Devices_O, Data)
            87 =>
                ActivateInvincibleTrain(Devices_O, Data)
            88 =>
                ActivateInvincibleThink(Devices_O, Data)
            89 => 
                ActivateZoro1(Devices_O, Data)
            90 =>
                ActivateZoro2(Devices_O, Data)
            91 =>
                ActivateSanji1(Devices_O, Data)
            92 =>
                ActivateSanji2(Devices_O, Data)
            93 =>
                ActivateRockLee(Devices_O, Data)
            94 =>
                ActivateSharingan(Devices_O, Data)
            95 =>
                ActivateRasenshuriken(Devices_O, Data)
            96 =>
                ActivateBijuu(Devices_O, Data)
            97 => 
                ActivateAoT1(Devices_O, Data)
            98 => 
                ActivateAoT2(Devices_O, Data)
            99 =>
                ActivateAoT3(Devices_O, Data)
            102 =>
                ActivateDeku1(Devices_O, Data)
            103 =>
                ActivateTodoroki1(Devices_O, Data)
            104 =>
                ActivateTodoroki2(Devices_O, Data)
            105 =>
                ActivateBakugo1(Devices_O, Data)
            106 =>
                ActivateBakugo2(Devices_O, Data)
            107 =>
                ActivateGuns1(Devices_O, Data)
            108 =>
                ActivateGuns2(Devices_O, Data)
            109 =>
                ActivateGuns3(Devices_O, Data)
            110 =>
                ActivateGuns4(Devices_O, Data)
            111 =>
                ActivateGuns5(Devices_O, Data)
            112 =>
                ActivateGuns6(Devices_O, Data)
            113 =>
                ActivateGuns7(Devices_O, Data)
            114 =>
                ActivateGuns8(Devices_O, Data)
            115 =>
                ActivateGuns9(Devices_O, Data)
            116 =>
                ActivateHalo1(Devices_O, Data)
            117 =>
                ActivateHalo2(Devices_O, Data)
            118 =>
                ActivateHalo3(Devices_O, Data)
            119 =>
                ActivateHalo4(Devices_O, Data)
            120 =>
                ActivateCap1(Devices_O, Data)
            121 =>
                ActivateThor1(Devices_O, Data)
            122 =>
                ActivateCap2(Devices_O, Data)
            123 =>
                ActivateThor2(Devices_O, Data)
            124 =>
                ActivateBoxing1(Devices_O, Data)
            125 =>
                ActivateBoxing2(Devices_O, Data)
            127 =>
                ActivateBoxing4(Devices_O, Data)
            128 =>
                ActivateSamurai4(Devices_O, Data)
            129 =>
                ActivateSamurai5(Devices_O, Data)
            130 =>
                ActivateSamurai6(Devices_O, Data)
            131 =>
                ActivateItadori1(Devices_O, Data)
            132 =>
                ActivateItadori2(Devices_O, Data)
            133 =>
                ActivateItadori3(Devices_O, Data)
            134 =>
                ActivateItadori4(Devices_O, Data)
            _ =>
                return

    # ----- Each finisher implemantation -----
    ActivateNinja1(Devices_O : ?finisher_effects, Data:finisher_data)<suspends>:void=
        Forward := Data.PlayerTransform.Rotation.GetLocalForward()
        NewLoc := Data.PlayerTransform.Translation + Forward * 901.0
        if (Devices := Devices_O?):
            Data.Cinema.Play()
            spawn:
                PlaySound(Data.Sound, Data.MoverTransform, Data.Duration)
            spawn:
                Ninja1Vfx(Data.PlayerTransform, Data.TestOffset)
            spawn:
                Ninja1DestructionPath(Data, Devices.Explosive)
            Sleep(Data.Duration)
            IsValid := IsValidLocation(NewLoc)
            if (IsValid = true):
                if (Devices.Teleporter.TeleportTo[NewLoc, DEFAULT_ROTATION]):
                    Sleep(0.0)
                    Devices.Teleporter.Teleport(Data.Instigator)

    Ninja1Vfx(InstigatorT : transform, TestOffset:vector3)<suspends>:void=
        Sleep(0.52)
        # Spawns a VFX particle system at the given Translation
        MyVFX:cancelable = SpawnParticleSystem(SlashHitVFX.NS.NS_Yellow_Flash_asset, InstigatorT.Translation + TestOffset)
        Sleep(0.5)
        MyVFX.Cancel()

    Ninja1DestructionPath(Data:finisher_data, Explosive : explosive_device)<suspends>:void=
        Sleep(0.20)
        Forward := Data.PlayerTransform.Rotation.GetLocalForward()
        OFFSET := vector3 {Z:=220.0}
        for (i := 1..3):
            NewLoc := Data.PlayerTransform.Translation + Forward * (300.0 * i) + OFFSET
            # DebugDraw.DrawSphere(NewLoc, ?DrawDurationPolicy := debug_draw_duration_policy.FiniteDuration, ?Duration := 10.0, ?Radius := 70.0, ?Color := NamedColors.Red)
            Explosive.Reset()
            if (Explosive.TeleportTo[NewLoc, DEFAULT_ROTATION]):
                Explosive.Explode(Data.Instigator)
            Sleep(0.15)

    ActivateNinja2(Devices_O : ?finisher_effects, Data:finisher_data)<suspends>:void=
        Forward := Data.PlayerTransform.Rotation.GetLocalForward()
        NewLoc := Data.PlayerTransform.Translation + Forward * 901.0
        if (Devices := Devices_O?):
            Data.Cinema.Play()
            spawn:
                PlaySound(Data.Sound, Data.MoverTransform, Data.Duration)
            spawn:
                Ninja2Vfx(Data.MoverTransform, Data.TestOffset)
            spawn:
                Ninja1DestructionPath(Data, Devices.Explosive)
            Sleep(Data.Duration)
            IsValid := IsValidLocation(NewLoc)
            if (IsValid = true):
                if (Devices.Teleporter.TeleportTo[NewLoc, DEFAULT_ROTATION]):
                    Sleep(0.0)
                    Devices.Teleporter.Teleport(Data.Target)
    
    Ninja2Vfx(InstigatorT : transform, TestOffset:vector3)<suspends>:void=
        Sleep(0.1)
        # Spawns a VFX particle system at the given Translation
        MyVFX:cancelable = SpawnParticleSystem(Pack_Effects_1.Air_Magic.VFX_Niagara.NS_Air_Magic_Shield_asset, InstigatorT.Translation + TestOffset)
        Sleep(1.5)
        MyVFX.Cancel()

    ActivateNinja3(Devices_O : ?finisher_effects, Data:finisher_data)<suspends>:void=
        if (Devices := Devices_O?):
            Data.Cinema.Play()
            spawn:
                PlaySound(Data.Sound, Data.MoverTransform, Data.Duration)
            spawn:
                Ninja3Vfx(Data.PlayerTransform, Data.TestOffset)
            # spawn:
            #     Ninja1DestructionPath(Data, Devices.Explosive)
            Sleep(Data.Duration)

    Ninja3Vfx(InstigatorT : transform, TestOffset:vector3)<suspends>:void=
        Forward := InstigatorT.Rotation.GetLocalForward()
        NewLoc := InstigatorT.Translation + vector3 {Z := 32.0} + Forward * 70.0
        # DebugDraw.DrawSphere(NewLoc, ?DrawDurationPolicy := debug_draw_duration_policy.FiniteDuration, ?Duration := 10.0, ?Radius := 70.0, ?Color := NamedColors.Red)
        Sleep(2.83)
        # Spawns a VFX particle system at the given Translation
        MyVFX:cancelable = SpawnParticleSystem(Asset := Pack_Effects_1.Fire_Magic.VFX_Niagara.NS_Fire_Magic_Flamethrower_asset, Position := NewLoc + TestOffset, ?Rotation := InstigatorT.Rotation, ?StartDelay := 0.0)
        Sleep(0.8)
        MyVFX.Cancel()

    ActivateNinja4(Devices_O : ?finisher_effects, Data:finisher_data)<suspends>:void=
        if (Devices := Devices_O?):
            Data.Cinema.Play()
            spawn:
                PlaySound(Data.Sound, Data.MoverTransform, Data.Duration)
            spawn:
                Ninja4Vfx(Data.PlayerTransform, Data.TestOffset)
            # spawn:
            #     Ninja1DestructionPath(Data, Devices.Explosive)
            Sleep(Data.Duration)

    Ninja4Vfx(MoverT : transform, TestOffset:vector3)<suspends>:void=
        Forward := MoverT.Rotation.GetLocalForward()
        NewLoc := MoverT.Translation + vector3 {Z := 60.0} + Forward * 80.0
        # DebugDraw.DrawSphere(NewLoc, ?DrawDurationPolicy := debug_draw_duration_policy.FiniteDuration, ?Duration := 10.0, ?Radius := 70.0, ?Color := NamedColors.Red)
        Sleep(1.3)
        # Spawns a VFX particle system at the given Translation
        # MagicSpells.Electric.FX.NS_DartingSpark
        MyVFX:cancelable = SpawnParticleSystem(Asset := Electric_VFX.Niagara.Electric.NS_Electric_Orb_asset, Position := NewLoc + TestOffset, ?Rotation := MoverT.Rotation, ?StartDelay := 0.0)
        Sleep(2.2)
        MyVFX.Cancel()

    ActivateNinja5(Devices_O : ?finisher_effects, Data:finisher_data)<suspends>:void=
        if (Devices := Devices_O?):
            Data.Cinema.Play()
            spawn:
                PlaySound(Data.Sound, Data.MoverTransform, Data.Duration)
            spawn:
                Ninja5Vfx(Data.MoverTransform, Data.TestOffset)
            # spawn:
            #     Ninja1DestructionPath(Data, Devices.Explosive)
            Sleep(Data.Duration)

    Ninja5Vfx(MoverT : transform, TestOffset:vector3)<suspends>:void=
        Forward := MoverT.Rotation.GetLocalForward()
        NewLoc := MoverT.Translation + Forward * 78.0
        Sleep(0.3)
        MyVFX:cancelable = SpawnParticleSystem(Asset := SkillsVFX.Particles.NS_Skill_Portal_asset, Position := NewLoc + TestOffset, ?Rotation := MoverT.Rotation, ?StartDelay := 0.0)
        Sleep(3.0)
        MyVFX.Cancel()

    ActivateNinja6(Devices_O : ?finisher_effects, Data:finisher_data)<suspends>:void=
        Forward := Data.PlayerTransform.Rotation.GetLocalForward()
        NewLoc := Data.PlayerTransform.Translation + Forward * -700.0
        if (Devices := Devices_O?):
            Data.Cinema.Play()
            spawn:
                PlaySound(Data.Sound, Data.MoverTransform, Data.Duration)
            spawn:
                Ninja6Vfx(Data.MoverTransform, Data.TestOffset)
            spawn:
                Ninja6DestructionPath(Data, Devices.Explosive)
            Sleep(0.9)
            IsValid := IsValidLocation(NewLoc)
            if (IsValid = true):
                if (Devices.Teleporter.TeleportTo[NewLoc, DEFAULT_ROTATION]):
                    Sleep(0.0)
                    Devices.Teleporter.Teleport(Data.Instigator)

            Sleep(Data.Duration - 0.9)

    Ninja6DestructionPath(Data:finisher_data, Explosive : explosive_device)<suspends>:void=
        Forward := Data.PlayerTransform.Rotation.GetLocalForward()
        OFFSET := vector3 {Z:=220.0}
        for (i := 1..3):
            NewLoc := Data.PlayerTransform.Translation + Forward * (-300.0 * i) + OFFSET
            # DebugDraw.DrawSphere(NewLoc, ?DrawDurationPolicy := debug_draw_duration_policy.FiniteDuration, ?Duration := 10.0, ?Radius := 70.0, ?Color := NamedColors.Red)
            Explosive.Reset()
            if (Explosive.TeleportTo[NewLoc, DEFAULT_ROTATION]):
                Explosive.Explode(Data.Instigator)
            Sleep(0.13)
        Sleep(0.2)
        for (i := 1..3):
            NewLoc := Data.PlayerTransform.Translation + Forward * (300.0 * i) + OFFSET
            # DebugDraw.DrawSphere(NewLoc, ?DrawDurationPolicy := debug_draw_duration_policy.FiniteDuration, ?Duration := 10.0, ?Radius := 70.0, ?Color := NamedColors.Red)
            Explosive.Reset()
            if (Explosive.TeleportTo[NewLoc, DEFAULT_ROTATION]):
                Explosive.Explode(Data.Instigator)
            Sleep(0.13)

    

    Ninja6Vfx(MoverT : transform, TestOffset:vector3)<suspends>:void=
        Forward := MoverT.Rotation.GetLocalForward()
        NewLoc := MoverT.Translation + vector3 {Z := 74.0} + Forward * 126.0
        Sleep(0.5)
        MyVFX:cancelable = SpawnParticleSystem(Asset := DarkMagicFX.Niagara.Projectile.NS_Explosion_Cast_DarkMagic_copy_asset, Position := NewLoc + TestOffset, ?Rotation := MoverT.Rotation, ?StartDelay := 0.0)
        Sleep(1.0)
        MyVFX.Cancel()

    ActivateHero1(Devices_O : ?finisher_effects, Data:finisher_data)<suspends>:void=
        Forward := Data.PlayerTransform.Rotation.GetLocalForward()
        NewLoc := Data.PlayerTransform.Translation + Forward * 1250.0
        if (Devices := Devices_O?):
            Data.Cinema.Play()
            spawn:
                PlaySound(Data.Sound, Data.MoverTransform, Data.Duration)
            spawn:
                Hero1Vfx(Data.MoverTransform, Data.TestOffset)
            spawn:
                HeroDestructionPath(Data, Devices.Explosive, 0.5, 3, 400.0)
            Sleep(Data.Duration)
            IsValid := IsValidLocation(NewLoc)
            if (IsValid = true):
                if (Devices.Teleporter.TeleportTo[NewLoc, DEFAULT_ROTATION]):
                    Sleep(0.0)
                    Devices.Teleporter.Teleport(Data.Target)

    Hero1Vfx(MoverT : transform, TestOffset:vector3)<suspends>:void=
        Forward := MoverT.Rotation.GetLocalForward()
        NewLoc := MoverT.Translation + vector3 {Z := 137.0} + Forward * 85.0
        Sleep(0.5) # 0.65
        FirstVFx:cancelable = SpawnParticleSystem(_VFX_ToUEFN.NS_Iron_start_asset, NewLoc + TestOffset, ?Rotation := MoverT.Rotation, ?StartDelay := 0.0)
        MyVFX:cancelable = SpawnParticleSystem(_VFX_ToUEFN.NS_iron_laser_asset, NewLoc + TestOffset, ?Rotation := MoverT.Rotation, ?StartDelay := 0.0)
        Sleep(0.2)
        FirstVFx.Cancel()
        MyVFX.Cancel()

    HeroDestructionPath(Data:finisher_data, Explosive : explosive_device, WaitTime : float, NumExplosions : int, DistanceBetween: float)<suspends>:void=
        Sleep(WaitTime)
        Forward := Data.PlayerTransform.Rotation.GetLocalForward()
        OFFSET := vector3 {Z:=220.0}
        for (i := 1..NumExplosions):
            NewLoc := Data.PlayerTransform.Translation + Forward * (DistanceBetween * i) + OFFSET
            # DebugDraw.DrawSphere(NewLoc, ?DrawDurationPolicy := debug_draw_duration_policy.FiniteDuration, ?Duration := 10.0, ?Radius := 70.0, ?Color := NamedColors.Red)
            Explosive.Reset()
            if (Explosive.TeleportTo[NewLoc, DEFAULT_ROTATION]):
                Explosive.Explode(Data.Instigator)
            Sleep(0.15)

    ActivateHero2(Devices_O : ?finisher_effects, Data:finisher_data)<suspends>:void=
        Forward := Data.PlayerTransform.Rotation.GetLocalForward()
        NewLoc := Data.PlayerTransform.Translation + Forward * 525.0
        if (Devices := Devices_O?):
            Data.Cinema.Play()
            spawn:
                PlaySound(Data.Sound, Data.MoverTransform, Data.Duration)
            spawn:
                Hero2Vfx(Data.MoverTransform, Data.TestOffset)
            spawn:
                HeroDestructionPath(Data, Devices.Explosive, 0.63, 3, 300.0)
            Sleep(Data.Duration)
            IsValid := IsValidLocation(NewLoc)
            if (IsValid = true):
                if (Devices.Teleporter.TeleportTo[NewLoc, DEFAULT_ROTATION]):
                    Sleep(0.0)
                    Devices.Teleporter.Teleport(Data.Target)

    Hero2Vfx(MoverT : transform, TestOffset:vector3)<suspends>:void=
        Forward := MoverT.Rotation.GetLocalForward()
        NewLoc := MoverT.Translation + vector3 {Z := 165.0} + Forward * 31.0
        NewRot := MoverT.Rotation.ApplyPitch(DegreesToRadians(-10.0))
        Sleep(0.63)
        MyVFX:cancelable = SpawnParticleSystem(_VFX_ToUEFN.NS_cyclops_laser_asset, NewLoc + TestOffset, ?Rotation := NewRot, ?StartDelay := 0.0)
        Sleep(1.0)
        MyVFX.Cancel()

    ActivateHero3(Devices_O : ?finisher_effects, Data:finisher_data)<suspends>:void=
        if (Devices := Devices_O?):
            Data.Cinema.Play()
            spawn:
                PlaySound(Data.Sound, Data.MoverTransform, Data.Duration)
            Sleep(Data.Duration)

    ActivateHero4(Devices_O : ?finisher_effects, Data:finisher_data)<suspends>:void=
        if (Devices := Devices_O?):
            Data.Cinema.Play()
            spawn:
                PlaySound(Data.Sound, Data.MoverTransform, Data.Duration)
            spawn:
                Hero4Vfx(Data.MoverTransform, Data.TestOffset)
            Sleep(Data.Duration)

    Hero4Vfx(MoverT : transform, TestOffset:vector3)<suspends>:void=
        Forward := MoverT.Rotation.GetLocalForward()
        NewLoc := MoverT.Translation + vector3 {Z := 100.0} + Forward * 156.0
        NewRot := MoverT.Rotation.ApplyYaw(DegreesToRadians(90.0))
        Sleep(0.3)
        MyVFX:cancelable = SpawnParticleSystem(_VFX_ToUEFN.NS_Mag_start_asset, NewLoc + TestOffset, ?Rotation := NewRot, ?StartDelay := 0.0)
        Sleep(3.86)
        MyVFX.Cancel()
        SecondVFX:cancelable = SpawnParticleSystem(_VFX_ToUEFN.NS_Mag_impact_asset, NewLoc + TestOffset, ?Rotation := MoverT.Rotation, ?StartDelay := 0.0)
        Sleep(1.0)
        SecondVFX.Cancel()
        
    ActivateHero5(Devices_O : ?finisher_effects, Data:finisher_data)<suspends>:void=
        Forward := Data.PlayerTransform.Rotation.GetLocalForward()
        NewLoc := Data.PlayerTransform.Translation + Forward * 901.0
        IsValid := IsValidLocation(NewLoc)
        if (Devices := Devices_O?):
            Data.Cinema.Play()
            spawn:
                PlaySound(Data.Sound, Data.MoverTransform, Data.Duration)
            spawn:
                Hero5Vfx(Data.MoverTransform, Data.TestOffset)
            spawn:
                HeroDestructionPath(Data, Devices.Explosive, 0.8, 3, 300.0)
            Sleep(Data.Duration)
            if (IsValid = true):
                if (Devices.Teleporter.TeleportTo[NewLoc, DEFAULT_ROTATION]):
                    Sleep(0.0)
                    Devices.Teleporter.Teleport(Data.Target)

    Hero5Vfx(PlayerT : transform, TestOffset:vector3)<suspends>:void=
        MyVFX:cancelable = SpawnParticleSystem(_VFX_ToUEFN.NS_doomFire_explosion_asset, PlayerT.Translation, ?Rotation := PlayerT.Rotation, ?StartDelay := 0.25)
        Sleep(3.0)
        MyVFX.Cancel()

    ActivateHero6(Devices_O : ?finisher_effects, Data:finisher_data)<suspends>:void=
        if (Devices := Devices_O?):
            Data.Cinema.Play()
            spawn:
                PlaySound(Data.Sound, Data.MoverTransform, Data.Duration)
            spawn:
                Hero6Vfx(Data.MoverTransform, Data.TestOffset)
            Sleep(Data.Duration)

    Hero6Vfx(MoverT : transform, TestOffset:vector3)<suspends>:void=
        MyVFX:cancelable = SpawnParticleSystem(_VFX_ToUEFN.NS_Tele_asset, MoverT.Translation, ?Rotation := MoverT.Rotation, ?StartDelay := 0.0)
        Sleep(4.10)
        MyVFX.Cancel()

    ActivateKi1(Devices_O : ?finisher_effects, Data:finisher_data)<suspends>:void=
        Forward := Data.PlayerTransform.Rotation.GetLocalForward()
        NewLoc := Data.PlayerTransform.Translation + Forward * 901.0
        IsValid := IsValidLocation(NewLoc)
        if (Devices := Devices_O?):
            Data.Cinema.Play()
            spawn:
                PlaySound(Data.Sound, Data.MoverTransform, Data.Duration)
            spawn:
                HeroDestructionPath(Data, Devices.Explosive, 1.2, 3, 300.0)
            Sleep(Data.Duration)
            if (IsValid = true):
                if (Devices.Teleporter.TeleportTo[NewLoc, DEFAULT_ROTATION]):
                    Sleep(0.0)
                    Devices.Teleporter.Teleport(Data.Target)

    ActivateSamurai1(Devices_O : ?finisher_effects, Data:finisher_data)<suspends>:void=
        Forward := Data.PlayerTransform.Rotation.GetLocalForward()
        NewLocTarget := Data.PlayerTransform.Translation + Forward * 400.0
        IsValid := IsValidLocation(NewLocTarget)
        
        if (Devices := Devices_O?):
            Data.Cinema.Play()
            spawn:
                PlaySound(Data.Sound, Data.MoverTransform, Data.Duration)
            spawn:
                HeroDestructionPath(Data, Devices.Explosive, 1.2, 2, 300.0)
            Sleep(Data.Duration)

    ActivateWater1(Devices_O : ?finisher_effects, Data:finisher_data)<suspends>:void=
        if (Devices := Devices_O?):
            Data.Cinema.Play()
            spawn:
                PlaySound(Data.Sound, Data.MoverTransform, Data.Duration)
            spawn:
                Water1Vfx(Data.MoverTransform, Data.TestOffset)
            Sleep(Data.Duration)

    Water1Vfx(MoverT : transform, TestOffset:vector3)<suspends>:void=
        Forward := MoverT.Rotation.GetLocalForward()
        Loc := MoverT.Translation + Forward * 250.0
        Sleep(0.16)
        StartVFX:cancelable = SpawnParticleSystem(_VFX_ToUEFN.NS_Ocean_Bubble_Start_asset, Loc, ?Rotation := MoverT.Rotation, ?StartDelay := 0.0)
        Sleep(1.17)
        EndVFX:cancelable = SpawnParticleSystem(_VFX_ToUEFN.NS_Ocean_Bubble_End_asset, Loc, ?Rotation := MoverT.Rotation, ?StartDelay := 0.0)
        StartVFX.Cancel()
        Sleep(0.73)
        ExplosionVFX:cancelable = SpawnParticleSystem(_VFX_ToUEFN.NS_Ocean_Bubble_Explosion_asset, Loc, ?Rotation := MoverT.Rotation, ?StartDelay := 0.0)
        Sleep(1.4)
        EndVFX.Cancel()
        ExplosionVFX.Cancel()

    ActivateFire1(Devices_O : ?finisher_effects, Data:finisher_data)<suspends>:void=
        if (Devices := Devices_O?):
            Data.Cinema.Play()
            spawn:
                PlaySound(Data.Sound, Data.MoverTransform, Data.Duration)
            spawn:
                Fire1Vfx(Data.MoverTransform, Data.TestOffset)
            Sleep(Data.Duration)

    Fire1Vfx(MoverT : transform, TestOffset:vector3)<suspends>:void=
        Forward := MoverT.Rotation.GetLocalForward()
        Loc := MoverT.Translation 
        Sleep(0.27)
        StartVFX:cancelable = SpawnParticleSystem(_VFX_ToUEFN.NS_AOE_BlazeBlast_asset, Loc, ?Rotation := MoverT.Rotation, ?StartDelay := 0.0)
        Sleep(2.0)
        StartVFX.Cancel()

    ActivateEarth1(Devices_O : ?finisher_effects, Data:finisher_data)<suspends>:void=
        if (Devices := Devices_O?):
            Data.Cinema.Play()
            spawn:
                PlaySound(Data.Sound, Data.MoverTransform, Data.Duration)
            spawn:
                Earth1Vfx(Data.MoverTransform, Data.TestOffset)
            Sleep(Data.Duration)

    Earth1Vfx(MoverT : transform, TestOffset:vector3)<suspends>:void=
        Forward := MoverT.Rotation.GetLocalForward()
        Loc := MoverT.Translation + Forward * 156.0
        # Sleep(0.27)
        StartVFX:cancelable = SpawnParticleSystem(_VFX_ToUEFN.NS_Earth_Magic_Spike4_asset, Loc, ?Rotation := MoverT.Rotation, ?StartDelay := 0.0)
        Sleep(3.0)
        StartVFX.Cancel()

    ActivateAir1(Devices_O : ?finisher_effects, Data:finisher_data)<suspends>:void=
        Forward := Data.PlayerTransform.Rotation.GetLocalForward()
        NewLoc := Data.PlayerTransform.Translation + Forward * 760.0
        IsValid := IsValidLocation(NewLoc)
        if (Devices := Devices_O?):
            Data.Cinema.Play()
            spawn:
                PlaySound(Data.Sound, Data.MoverTransform, Data.Duration)
            spawn:
                Air1Vfx(Data.MoverTransform, Data.TestOffset)
            spawn:
                HeroDestructionPath(Data, Devices.Explosive, 1.0, 3, 250.0)
            Sleep(Data.Duration)
            if (IsValid = true):
                if (Devices.Teleporter.TeleportTo[NewLoc, DEFAULT_ROTATION]):
                    Sleep(0.0)
                    Devices.Teleporter.Teleport(Data.Target)

    Air1Vfx(MoverT : transform, TestOffset:vector3)<suspends>:void=
        Forward := MoverT.Rotation.GetLocalForward()
        Loc := MoverT.Translation
        Sleep(0.33)
        StartVFX:cancelable = SpawnParticleSystem(Pack_Effects_1.Air_Magic.VFX_Niagara.NS_Air_Magic_show_asset, Loc, ?Rotation := MoverT.Rotation, ?StartDelay := 0.0)
        Sleep(1.77)
        StartVFX.Cancel()

    ActivateKi2(Devices_O : ?finisher_effects, Data:finisher_data)<suspends>:void=
        Forward := Data.PlayerTransform.Rotation.GetLocalForward()
        NewLoc := Data.PlayerTransform.Translation + Forward * 901.0
        IsValid := IsValidLocation(NewLoc)
        if (Devices := Devices_O?):
            Data.Cinema.Play()
            spawn:
                PlaySound(Data.Sound, Data.MoverTransform, Data.Duration)
            spawn:
                HeroDestructionPath(Data, Devices.Explosive, 1.2, 3, 300.0)
            Sleep(Data.Duration)
            if (IsValid = true):
                if (Devices.Teleporter.TeleportTo[NewLoc, DEFAULT_ROTATION]):
                    Sleep(0.0)
                    Devices.Teleporter.Teleport(Data.Target)

    ActivateWWE1(Devices_O : ?finisher_effects, Data:finisher_data)<suspends>:void=
        Forward := Data.PlayerTransform.Rotation.GetLocalForward()
        NewLoc := Data.PlayerTransform.Translation + Forward * 279.0
        IsValid := IsValidLocation(NewLoc)
        if (Devices := Devices_O?):
            Data.Cinema.Play()
            spawn:
                PlaySound(Data.Sound, Data.MoverTransform, Data.Duration)
            spawn:
                HeroDestructionPath(Data, Devices.Explosive, 2.0, 2, 150.0)
            Sleep(Data.Duration)
            if (IsValid = true):
                if (Devices.Teleporter.TeleportTo[NewLoc, DEFAULT_ROTATION]):
                    Sleep(0.0)
                    Devices.Teleporter.Teleport(Data.Instigator)

    ActivateWolverine1(Devices_O : ?finisher_effects, Data:finisher_data)<suspends>:void=
        Forward := Data.PlayerTransform.Rotation.GetLocalForward()
        NewLoc := Data.PlayerTransform.Translation + Forward * 618.0
        IsValid := IsValidLocation(NewLoc)
        if (Devices := Devices_O?):
            Data.Cinema.Play()
            spawn:
                PlaySound(Data.Sound, Data.MoverTransform, Data.Duration)
            spawn:
                HeroDestructionPath(Data, Devices.Explosive, 0.5, 3, 200.0)
            Sleep(Data.Duration)
            if (IsValid = true):
                if (Devices.Teleporter.TeleportTo[NewLoc, DEFAULT_ROTATION]):
                    Sleep(0.0)
                    Devices.Teleporter.Teleport(Data.Target)

    ActivateWolverine2(Devices_O : ?finisher_effects, Data:finisher_data)<suspends>:void=
        Forward := Data.PlayerTransform.Rotation.GetLocalForward()
        NewLoc := Data.PlayerTransform.Translation + Forward * 455.0
        IsValid := IsValidLocation(NewLoc)
        if (Devices := Devices_O?):
            Data.Cinema.Play()
            spawn:
                PlaySound(Data.Sound, Data.MoverTransform, Data.Duration)
            spawn:
                HeroDestructionPath(Data, Devices.Explosive, 0.5, 3, 150.0)
            Sleep(Data.Duration)
            if (IsValid = true):
                if (Devices.Teleporter.TeleportTo[NewLoc, DEFAULT_ROTATION]):
                    Sleep(0.0)
                    Devices.Teleporter.Teleport(Data.Target)

    ActivateWolverine4(Devices_O : ?finisher_effects, Data:finisher_data)<suspends>:void=
        Forward := Data.PlayerTransform.Rotation.GetLocalForward()
        NewLoc := Data.PlayerTransform.Translation + Forward * 570.0
        IsValid := IsValidLocation(NewLoc)
        if (Devices := Devices_O?):
            Data.Cinema.Play()
            spawn:
                PlaySound(Data.Sound, Data.MoverTransform, Data.Duration)
            spawn:
                HeroDestructionPath(Data, Devices.Explosive, 0.5, 3, 190.0)
            Sleep(Data.Duration)
            if (IsValid = true):
                if (Devices.Teleporter.TeleportTo[NewLoc, DEFAULT_ROTATION]):
                    Sleep(0.0)
                    Devices.Teleporter.Teleport(Data.Target)

    ActivateKi3(Devices_O : ?finisher_effects, Data:finisher_data)<suspends>:void=
        KiMoveEnemy(Devices_O, Data, 1030.0, 1.3, 5)

    ActivateKi4(Devices_O : ?finisher_effects, Data:finisher_data)<suspends>:void=
        KiMoveEnemy(Devices_O, Data, 596.0, 0.0, 3)

    ActivateKi5(Devices_O : ?finisher_effects, Data:finisher_data)<suspends>:void=
        KiMoveEnemy(Devices_O, Data, 1030.0, 2.9, 5)

    ActivateKi6(Devices_O : ?finisher_effects, Data:finisher_data)<suspends>:void=
        KiMoveEnemy(Devices_O, Data, 1030.0, 2.4, 5)

    KiMoveEnemy(Devices_O : ?finisher_effects, Data:finisher_data, ForwardDistance:float, WaitTime:float, NumExplosions: int)<suspends>:void=
        Forward := Data.PlayerTransform.Rotation.GetLocalForward()
        NewLoc := Data.PlayerTransform.Translation + Forward * ForwardDistance
        IsValid := IsValidLocation(NewLoc)
        if (Devices := Devices_O?):
            Data.Cinema.Play()
            spawn:
                PlaySound(Data.Sound, Data.MoverTransform, Data.Duration)
            spawn:
                HeroDestructionPath(Data, Devices.Explosive, WaitTime, NumExplosions, 200.0)
            Sleep(Data.Duration)
            if (IsValid = true):
                if (Devices.Teleporter.TeleportTo[NewLoc, DEFAULT_ROTATION]):
                    Sleep(0.0)
                    Devices.Teleporter.Teleport(Data.Target)

    ActivateInfiniteVoid(Devices_O : ?finisher_effects, Data:finisher_data)<suspends>:void=
        if (Devices := Devices_O?):
            Data.Cinema.Play()
            spawn:
                PlayCamera(Devices.InfiniteCamera1, Data.Instigator)
            spawn:
                PlayCamera(Devices.InfiniteCamera2, Data.Target)
            spawn:
                PlaySound(Data.Sound, Data.MoverTransform, Data.Duration)
            spawn:
                InfiniteVfx(Data.MoverTransform, Data.TestOffset)
            spawn:
                CreateInfiniteDomain(Data.MoverTransform, Devices.InfiniteDomainAsset)
            spawn:
                HeroDestructionPath(Data, Devices.Explosive, 0.4, 3, 300.0)
            Sleep(Data.Duration)
            
    InfiniteVfx(MoverT : transform, TestOffset:vector3)<suspends>:void=
        Loc := MoverT.Translation + vector3 {Z := 200.0}
        Sleep(0.2)
        MyVFX:cancelable = SpawnParticleSystem(_VFX_ToUEFN.NS_WarpTunnel_asset, Loc , ?Rotation := MoverT.Rotation, ?StartDelay := 0.0)
        Sleep(2.8)
        MyVFX.Cancel()

    CreateInfiniteDomain(MoverT : transform, InfiniteDomainAsset:creative_prop_asset)<suspends>:void=
        SmallScale := vector3 {X:=0.001, Y:=0.001, Z:=0.001}
        # BigScale := vector3 {X:=0.25, Y:=0.25, Z:=0.25}
        BigScale := vector3 {X:=1.0, Y:=1.0, Z:=1.0}
        Duration := 0.5
        SpawnTransform := transform {Translation:= MoverT.Translation, Rotation:= MoverT.Rotation, Scale := SmallScale}
        FinalTransform := transform {Translation:= MoverT.Translation, Rotation:= MoverT.Rotation, Scale := BigScale}

        Sleep(2.8)

        Prop := SpawnProp(InfiniteDomainAsset, SpawnTransform)
        var CreatedProps:[]creative_prop = array {}
        if (NewCreatedProp:= Prop(0)?):
            set CreatedProps += array {NewCreatedProp}
            NewCreatedProp.MoveTo(FinalTransform, Duration)
        Sleep(8.2)
        for (Prop_: CreatedProps):
            Prop_.Dispose()

    ActivateSukunaDomain(Devices_O : ?finisher_effects, Data:finisher_data)<suspends>:void=
        if (Devices := Devices_O?):
            Data.Cinema.Play()
            spawn:
                PlayCamera(Devices.InfiniteCamera1, Data.Instigator)
            spawn:
                PlayCamera(Devices.InfiniteCamera2, Data.Target)
            spawn:
                PlaySound(Data.Sound, Data.MoverTransform, Data.Duration)
            spawn:
                SukunaSlashesVFX(Data.MoverTransform, Data.TestOffset)
            spawn:
                CreateInfiniteDomain(Data.MoverTransform, Devices.SukunaDomainAsset)
            spawn:
                DomainDestruction(Data.Instigator, Data.MoverTransform.Translation, Devices.DomainExplosive, 3.3, 8)
            Sleep(Data.Duration)

    SukunaSlashesVFX(MoverT : transform, TestOffset:vector3)<suspends>:void=
        Loc := MoverT.Translation + vector3 {Z := 200.0}
        Sleep(3.3)
        MyVFX:cancelable = SpawnParticleSystem(_VFX_ToUEFN.NS_Cut_Sl_08_red_asset, Loc , ?Rotation := MoverT.Rotation, ?StartDelay := 0.0)
        Sleep(5.0)
        MyVFX.Cancel()

    DomainDestruction(Instigator:agent, Loc: vector3, Explosive : explosive_device, WaitTime : float, NumExplosions:int)<suspends>:void=
        Sleep(WaitTime)
        if (Explosive.TeleportTo[Loc, DEFAULT_ROTATION]):
            for (i := 1..NumExplosions):
                Explosive.Reset()
                Explosive.Explode(Instigator)
                Sleep(0.5)

    ActivateGenkidama(Devices_O : ?finisher_effects, Data:finisher_data)<suspends>:void=
        Forward := Data.PlayerTransform.Rotation.GetLocalForward()
        NewLoc := Data.PlayerTransform.Translation + Forward * 800.0
        ExplosionLoc := Data.PlayerTransform.Translation + Forward * 1300.0
        IsValid := IsValidLocation(NewLoc)
        if (Devices := Devices_O?):
            Data.Cinema.Play()
            spawn:
                PlaySound(Data.Sound, Data.MoverTransform, Data.Duration)
            spawn:
                DomainDestruction(Data.Instigator, ExplosionLoc, Devices.DomainExplosive, 5.0, 2)
            spawn:
                HeroDestructionPath(Data, Devices.Explosive, 2.0, 4, 300.0)
            Sleep(Data.Duration)
            if (IsValid = true):
                if (Devices.Teleporter.TeleportTo[NewLoc, DEFAULT_ROTATION]):
                    Sleep(0.0)
                    Devices.Teleporter.Teleport(Data.Target)
    
    ActivateSamurai3(Devices_O : ?finisher_effects, Data:finisher_data)<suspends>:void=       
        if (Devices := Devices_O?):
            SelectPlasmaColor(Data.PlasmaColorCinema, Data.PlasmaColorId)
            Data.Cinema.Play()
            spawn:
                PlaySound(Data.Sound, Data.MoverTransform, Data.Duration)
            Sleep(Data.Duration)

    ActivateGiant1(Devices_O : ?finisher_effects, Data:finisher_data)<suspends>:void=
        if (Devices := Devices_O?):
            Data.Cinema.Play()
            spawn:
                PlayCamera(Devices.GiantCamera1, Data.Instigator)
            spawn:
                PlayCamera(Devices.GiantCamera2, Data.Target)
            spawn:
                PlaySound(Data.Sound, Data.MoverTransform, Data.Duration)
            ExplosionLocation := Data.MoverTransform.Translation + vector3 {Z := 995.0}
            spawn:
                DomainDestruction(Data.Instigator, ExplosionLocation, Devices.GiantExplosive, 1.0, 10)
                
            Sleep(Data.Duration)

    GiantDestructionPath(Data:finisher_data, Explosive : explosive_device, WaitTime : float, NumExplosions : int, DistanceBetween: float)<suspends>:void=
        Sleep(WaitTime)
        Up := vector3 {Z := 1.0}
        OFFSET := vector3 {Z:=220.0}
        for (i := 1..NumExplosions):
            NewLoc := Data.PlayerTransform.Translation + Up * (DistanceBetween * i) + OFFSET
            # DebugDraw.DrawSphere(NewLoc, ?DrawDurationPolicy := debug_draw_duration_policy.FiniteDuration, ?Duration := 10.0, ?Radius := 70.0, ?Color := NamedColors.Red)
            Explosive.Reset()
            if (Explosive.TeleportTo[NewLoc, DEFAULT_ROTATION]):
                Explosive.Explode(Data.Instigator)
            Sleep(0.15)

    ActivateSpiderman1(Devices_O : ?finisher_effects, Data:finisher_data)<suspends>:void=
        Forward := Data.PlayerTransform.Rotation.GetLocalForward()
        NewLoc := Data.PlayerTransform.Translation + Forward * 665.0
        NewLocTarget :=  Data.PlayerTransform.Translation + Forward * 880.0
        if (Devices := Devices_O?):
            Data.Cinema.Play()
            spawn:
                PlaySound(Data.Sound, Data.MoverTransform, Data.Duration)
            spawn:
                HeroDestructionPath(Data, Devices.Explosive, 0.5, 4, 300.0)
            Sleep(Data.Duration)
            IsValid := IsValidLocation(NewLoc)
            if (IsValid = true):
                if (Devices.Teleporter.TeleportTo[NewLoc, DEFAULT_ROTATION]):
                    Sleep(0.0)
                    Devices.Teleporter.Teleport(Data.Instigator)
            TeleportTarget(Data.Target, Devices.Teleporter, NewLocTarget)

    ActivateSW1(Devices_O : ?finisher_effects, Data:finisher_data)<suspends>:void=
        Forward := Data.PlayerTransform.Rotation.GetLocalForward()
        NewLoc := Data.PlayerTransform.Translation + Forward * 485.0
        NewLocTarget :=  Data.PlayerTransform.Translation + Forward * 1048.0
        if (Devices := Devices_O?):
            Data.Cinema.Play()
            spawn:
                PlaySound(Data.Sound, Data.MoverTransform, Data.Duration)
            spawn:
                HeroDestructionPath(Data, Devices.Explosive, 1.15, 6, 200.0)
            Sleep(Data.Duration)
            IsValid := IsValidLocation(NewLoc)
            if (IsValid = true):
                if (Devices.Teleporter.TeleportTo[NewLoc, DEFAULT_ROTATION]):
                    Sleep(0.0)
                    Devices.Teleporter.Teleport(Data.Instigator)
            TeleportTarget(Data.Target, Devices.Teleporter, NewLocTarget)

    ActivateSW2(Devices_O : ?finisher_effects, Data:finisher_data)<suspends>:void=
        Forward := Data.PlayerTransform.Rotation.GetLocalForward()
        Right := Data.PlayerTransform.Rotation.GetLocalRight()
        NewLoc :=  Data.PlayerTransform.Translation + Forward * 550.0 + Right * 182.0
        if (Devices := Devices_O?):
            Data.Cinema.Play()
            spawn:
                PlaySound(Data.Sound, Data.MoverTransform, Data.Duration)
            spawn:
                HeroDestructionPath(Data, Devices.Explosive, 1.2, 3, 200.0)
            Sleep(Data.Duration)
            IsValid := IsValidLocation(NewLoc)
            if (IsValid = true):
                if (Devices.Teleporter.TeleportTo[NewLoc, DEFAULT_ROTATION]):
                    Sleep(0.0)
                    Devices.Teleporter.Teleport(Data.Target)

    TeleportTarget(Target: agent, Teleporter: teleporter_device, Loc: vector3)<suspends>:void=
        Sleep(0.1)
        IsValidT := IsValidLocation(Loc)
        if (IsValidT = true):
            if (Teleporter.TeleportTo[Loc, DEFAULT_ROTATION]):
                Sleep(0.0)
                Teleporter.Teleport(Target)

    ActivateSpiderman2(Devices_O : ?finisher_effects, Data:finisher_data)<suspends>:void=
        if (Devices := Devices_O?):
            Data.Cinema.Play()
            spawn:
                PlaySound(Data.Sound, Data.MoverTransform, Data.Duration)
            spawn:
                ActivateSpiderSense(Data.Instigator, Data.Target)
                
            Sleep(Data.Duration)

    ActivateSpiderSense(Instigator: agent, Target : agent)<suspends>:void=
        Sleep(0.7)
        PP_Spidersense.BlendIn(Instigator)
        PP_Spidersense.BlendIn(Target)
        Sleep(2.8)
        PP_Spidersense.BlendOut(Instigator)
        PP_Spidersense.BlendOut(Target)

    ActivateSpiderman4(Devices_O : ?finisher_effects, Data:finisher_data)<suspends>:void=
        if (Devices := Devices_O?):
            Data.Cinema.Play()
            spawn:
                PlaySound(Data.Sound, Data.MoverTransform, Data.Duration)
            spawn:
                HeroDestructionPath(Data, Devices.Explosive, 0.0, 2, 200.0)
                
            Sleep(Data.Duration)

    ActivateLuffy1(Devices_O : ?finisher_effects, Data:finisher_data)<suspends>:void=
        if (Devices := Devices_O?):
            spawn:
                AddCamera(Data.Instigator, Data.Target, Devices.Luffy1Camera)
            Data.Cinema.Play()
            spawn:
                PlaySound(Data.Sound, Data.MoverTransform, Data.Duration)
            spawn:
                Luffy1Vfx(Data.PlayerTransform, Data.TestOffset)
                
            Sleep(Data.Duration-0.3)
            RemoveCamera(Data.Instigator, Data.Target, Devices.Luffy1Camera)
            Sleep(0.31)
        
    AddCamera(Instigator:agent, Target:agent, Camera: gameplay_camera_device)<suspends>:void=
        Sleep(0.01)
        Camera.AddTo(Instigator)
        if (Instigator <> Target):
            Camera.AddTo(Target)

    RemoveCamera(Instigator:agent, Target:agent, Camera: gameplay_camera_device)<suspends>:void=
        Camera.RemoveFrom(Instigator)
        # if (Instigator <> Target):
        Camera.RemoveFrom(Target)

    Luffy1Vfx(MoverT : transform, TestOffset:vector3)<suspends>:void=
        Forward := MoverT.Rotation.GetLocalForward()
        NewLoc := MoverT.Translation + vector3 {Z := 163.0} + Forward * 113.0
        Sleep(0.7)
        MyVFX:cancelable = SpawnParticleSystem(Asset := _VFX_ToUEFN.N_PhysicalStrength_small_asset, Position := NewLoc + TestOffset, ?Rotation := MoverT.Rotation, ?StartDelay := 0.0)
        MyVFX2:cancelable = SpawnParticleSystem(_VFX_ToUEFN.NS_OnePiece_Hit_Simple_asset, NewLoc + TestOffset, ?Rotation := MoverT.Rotation, ?StartDelay := 0.0)
        Sleep(0.8)
        MyVFX.Cancel()
        MyVFX2.Cancel()

    ActivateLuffy2(Devices_O : ?finisher_effects, Data:finisher_data)<suspends>:void=
        Forward := Data.PlayerTransform.Rotation.GetLocalForward()
        NewLoc := Data.PlayerTransform.Translation + Forward * 520.0
        if (Devices := Devices_O?):
            Data.Cinema.Play()
            spawn:
                PlaySound(Data.Sound, Data.MoverTransform, Data.Duration)
            spawn:
                Luffy2Vfx(Data.PlayerTransform, Data.TestOffset)
            spawn:
                HeroDestructionPath(Data, Devices.Explosive, 0.57, 3, 200.0)
                
            Sleep(Data.Duration)
            IsValid := IsValidLocation(NewLoc)
            if (IsValid = true):
                if (Devices.Teleporter.TeleportTo[NewLoc, DEFAULT_ROTATION]):
                    Sleep(0.0)
                    Devices.Teleporter.Teleport(Data.Target)

    Luffy2Vfx(MoverT : transform, TestOffset:vector3)<suspends>:void=
        Forward := MoverT.Rotation.GetLocalForward()
        NewLoc := MoverT.Translation + vector3 {Z := 165.0} + Forward * 250.0
        Sleep(0.57)
        MyVFX2:cancelable = SpawnParticleSystem(_VFX_ToUEFN.NS_OnePiece_Hit_Simple_asset, NewLoc + TestOffset, ?Rotation := MoverT.Rotation, ?StartDelay := 0.0)
        Sleep(0.8)
        MyVFX2.Cancel()

    ActivateLuffy3(Devices_O : ?finisher_effects, Data:finisher_data)<suspends>:void=
        Forward := Data.PlayerTransform.Rotation.GetLocalForward()
        NewLoc := Data.PlayerTransform.Translation + Forward * 860.0
        if (Devices := Devices_O?):
            Data.Cinema.Play()
            spawn:
                PlaySound(Data.Sound, Data.MoverTransform, Data.Duration)
            spawn:
                Luffy3Vfx(Data.PlayerTransform, Data.TestOffset)
            spawn:
                HeroDestructionPath(Data, Devices.Explosive, 2.0, 4, 220.0)
                
            Sleep(Data.Duration)
            IsValid := IsValidLocation(NewLoc)
            if (IsValid = true):
                if (Devices.Teleporter.TeleportTo[NewLoc, DEFAULT_ROTATION]):
                    Sleep(0.0)
                    Devices.Teleporter.Teleport(Data.Target)

    Luffy3Vfx(MoverT : transform, TestOffset:vector3)<suspends>:void=
        Forward := MoverT.Rotation.GetLocalForward()
        NewLoc := MoverT.Translation + vector3 {Z := 90.0} + Forward * 100.0
        NewRot := MoverT.Rotation.ApplyYaw(DegreesToRadians(-90.0))
        Sleep(0.03)
        GattlingVFX:cancelable = SpawnParticleSystem(_VFX_ToUEFN.NS_OnePiece_Gattling2_asset, NewLoc + TestOffset, ?Rotation := NewRot, ?StartDelay := 0.0)
        Sleep(1.6)
        GattlingVFX.Cancel()
        Sleep(0.34)
        ImpactLoc := MoverT.Translation + vector3 {Z := 120.0} + Forward * 360.0
        ImpactRot := MoverT.Rotation.ApplyPitch(DegreesToRadians(90.0))
        StrengthFX:cancelable = SpawnParticleSystem(_VFX_ToUEFN.N_PhysicalStrength_small_asset, ImpactLoc + TestOffset, ?Rotation := ImpactRot, ?StartDelay := 0.0)
        HitVFX:cancelable = SpawnParticleSystem(_VFX_ToUEFN.NS_OnePiece_Hit_Simple_asset, ImpactLoc + TestOffset, ?Rotation := MoverT.Rotation, ?StartDelay := 0.0)
        Sleep(0.8)
        StrengthFX.Cancel()
        HitVFX.Cancel()

    ActivateLuffy4(Devices_O : ?finisher_effects, Data:finisher_data)<suspends>:void=
        Forward := Data.PlayerTransform.Rotation.GetLocalForward()
        NewLoc := Data.PlayerTransform.Translation + Forward * 780.0
        DiceRoll := GetRandomFloat(0.0, 1.0)
        if (Devices := Devices_O?):
            Data.Cinema.Play()
            spawn:
                PlayLuffy4Sound(Data.Sound, Devices.Luffy4RedHawkSound, Data.MoverTransform, DiceRoll)
            spawn:
                Luffy4Vfx(Data.PlayerTransform, Data.TestOffset, DiceRoll)
            spawn:
                HeroDestructionPath(Data, Devices.Explosive, 1.0, 4, 200.0)
                
            Sleep(Data.Duration)
            IsValid := IsValidLocation(NewLoc)
            if (IsValid = true):
                if (Devices.Teleporter.TeleportTo[NewLoc, DEFAULT_ROTATION]):
                    Sleep(0.0)
                    Devices.Teleporter.Teleport(Data.Target)

    Luffy4Vfx(MoverT : transform, TestOffset:vector3, DiceRoll:float)<suspends>:void= 
        Forward := MoverT.Rotation.GetLocalForward()
        Right := MoverT.Rotation.GetLocalRight()
        if(DiceRoll > 0.75):
            # Red hawk
            NewLoc := MoverT.Translation + vector3 {Z := 120.0} + Forward * 120.0 + Right * 30.0
            Sleep(0.87)
            RedHawkVFX:cancelable = SpawnParticleSystem(_VFX_ToUEFN.NS_OnePiece_RedHawk_asset, NewLoc + TestOffset, ?Rotation := MoverT.Rotation, ?StartDelay := 0.0)
            Sleep(1.0)
            RedHawkVFX.Cancel()
        else:
            Sleep(0.93)
            ImpactLoc := MoverT.Translation + vector3 {Z := 120.0} + Forward * 416.0
            ImpactRot := MoverT.Rotation.ApplyPitch(DegreesToRadians(90.0))
            StrengthFX:cancelable = SpawnParticleSystem(_VFX_ToUEFN.N_PhysicalStrength_small_asset, ImpactLoc + TestOffset, ?Rotation := ImpactRot, ?StartDelay := 0.0)
            HitVFX:cancelable = SpawnParticleSystem(_VFX_ToUEFN.NS_OnePiece_Hit_Simple_asset, ImpactLoc + TestOffset, ?Rotation := MoverT.Rotation, ?StartDelay := 0.0)
            Sleep(0.8)
            StrengthFX.Cancel()
            HitVFX.Cancel()

    PlayLuffy4Sound(SoundBasic: audio_player_device, SoundRedHawk: audio_player_device, MoverT : transform, DiceRoll:float)<suspends>:void=
        Forward := MoverT.Rotation.GetLocalForward()
        NewLoc := MoverT.Translation + vector3 {Z := 100.0} + Forward * 78.0
        if (DiceRoll > 0.75):
            # Red hawk

            if (SoundRedHawk.TeleportTo[NewLoc, DEFAULT_ROTATION]):
                # Print("Sound teleported")
            Sleep(0.0)
            SoundRedHawk.Play()
        else:
            if (SoundBasic.TeleportTo[NewLoc, DEFAULT_ROTATION]):
                # Print("Sound teleported")
            Sleep(0.0)
            SoundBasic.Play()

    ActivateLuffy5(Devices_O : ?finisher_effects, Data:finisher_data)<suspends>:void=
        Forward := Data.PlayerTransform.Rotation.GetLocalForward()
        NewLoc := Data.PlayerTransform.Translation + Forward * 540.0
        if (Devices := Devices_O?):
            spawn:
                AddDualCamera(Data.Instigator, Data.Target, Devices.Luffy5CameraAtk, Devices.Luffy5CameraVic)
            Data.Cinema.Play()
            spawn:
                PlaySound(Data.Sound, Data.MoverTransform, Data.Duration)
            spawn:
                Luffy5Vfx(Data.PlayerTransform, Data.TestOffset)
            spawn:
                HeroDestructionPath(Data, Devices.Explosive, 0.83, 3, 200.0)
                
            Sleep(Data.Duration-0.31)
            RemoveDualCamera(Data.Instigator, Data.Target, Devices.Luffy5CameraAtk, Devices.Luffy5CameraVic)
            Sleep(0.305)

            IsValid := IsValidLocation(NewLoc)
            if (IsValid = true):
                if (Devices.Teleporter.TeleportTo[NewLoc, DEFAULT_ROTATION]):
                    Sleep(0.0)
                    Devices.Teleporter.Teleport(Data.Target)

    AddDualCamera(Instigator:agent, Target:agent, CameraAtk: gameplay_camera_device, CameraVic: gameplay_camera_device)<suspends>:void=
        Sleep(0.01)
        CameraAtk.AddTo(Instigator)
        if (Instigator <> Target):
            CameraVic.AddTo(Target)

    RemoveDualCamera(Instigator:agent, Target:agent, CameraAtk: gameplay_camera_device, CameraVic: gameplay_camera_device)<suspends>:void=
        CameraAtk.RemoveFrom(Instigator)
        CameraVic.RemoveFrom(Target)

    Luffy5Vfx(MoverT : transform, TestOffset:vector3)<suspends>:void=
        Forward := MoverT.Rotation.GetLocalForward()
        NewLoc := MoverT.Translation + vector3 {Z := 60.0} + Forward * 520.0
        Sleep(1.16)
        MyVFX:cancelable = SpawnParticleSystem(_VFX_ToUEFN.N_PhysicalStrength_small_asset, NewLoc + TestOffset, ?Rotation := MoverT.Rotation, ?StartDelay := 0.0)
        Sleep(0.8)
        MyVFX.Cancel()

    ActivateFuga(Devices_O : ?finisher_effects, Data:finisher_data)<suspends>:void=
        Forward := Data.PlayerTransform.Rotation.GetLocalForward()
        NewLoc := Data.PlayerTransform.Translation + Forward * 1370.0
        if (Devices := Devices_O?):
            spawn:
                AddCamera(Data.Instigator, Data.Target, Devices.FugaCamera)
            Data.Cinema.Play()
            spawn:
                PlaySound(Data.Sound, Data.MoverTransform, Data.Duration)
            spawn:
                HeroDestructionPath(Data, Devices.Explosive, 2.45, 4, 350.0)
            spawn:
                FugaVfx(Data.MoverTransform, Data.TestOffset)
                
            Sleep(Data.Duration)
            RemoveCamera(Data.Instigator, Data.Target, Devices.FugaCamera)

            IsValid := IsValidLocation(NewLoc)
            if (IsValid = true):
                if (Devices.Teleporter.TeleportTo[NewLoc, DEFAULT_ROTATION]):
                    Sleep(0.0)
                    Devices.Teleporter.Teleport(Data.Target)

    FugaVfx(MoverT : transform, TestOffset:vector3)<suspends>:void=
        Forward := MoverT.Rotation.GetLocalForward()
        NewLoc := MoverT.Translation + Forward * 1370.0
        Sleep(2.7)
        MyVFX:cancelable = SpawnParticleSystem(_VFX_ToUEFN.NS_Fuga_Explosion_asset, NewLoc + TestOffset, ?Rotation := MoverT.Rotation, ?StartDelay := 0.0)
        Sleep(4.0)
        MyVFX.Cancel()

    ActivateToji1(Devices_O : ?finisher_effects, Data:finisher_data)<suspends>:void=
        Forward := Data.PlayerTransform.Rotation.GetLocalForward()
        NewLoc := Data.PlayerTransform.Translation + Forward * 320.0
        NewLocTarget := Data.PlayerTransform.Translation + Forward * 550.0
        if (Devices := Devices_O?):
            spawn:
                AddCamera(Data.Instigator, Data.Target, Devices.Toji1Camera)
            Data.Cinema.Play()
            spawn:
                PlaySound(Data.Sound, Data.MoverTransform, Data.Duration)
            spawn:
                HeroDestructionPath(Data, Devices.Explosive, 2.0, 3, 200.0)
                
            Sleep(Data.Duration)
            RemoveCamera(Data.Instigator, Data.Target, Devices.Toji1Camera)

            # IsValid := IsValidLocation(NewLoc)
            # if (IsValid = true):
            #     if (Devices.Teleporter.TeleportTo[NewLoc, DEFAULT_ROTATION]):
            #         Sleep(0.0)
            #         Devices.Teleporter.Teleport(Data.Instigator)
            # TeleportTarget(Data.Target, Devices.Teleporter, NewLocTarget)

    ActivateHollowPurple(Devices_O : ?finisher_effects, Data:finisher_data)<suspends>:void=
        Forward := Data.PlayerTransform.Rotation.GetLocalForward()
        NewLoc := Data.PlayerTransform.Translation + Forward * 1400.0
        if (Devices := Devices_O?):
            Data.Cinema.Play()
            spawn:
                PlaySound(Data.Sound, Data.MoverTransform, Data.Duration)
            spawn:
                HeroDestructionPath(Data, Devices.Explosive, 1.7, 7, 200.0)
            Sleep(Data.Duration)
            IsValid := IsValidLocation(NewLoc)
            if (IsValid = true):
                if (Devices.Teleporter.TeleportTo[NewLoc, DEFAULT_ROTATION]):
                    Sleep(0.0)
                    Devices.Teleporter.Teleport(Data.Target)

    ActivateFlash(Devices_O : ?finisher_effects, Data:finisher_data)<suspends>:void=
        if (Devices := Devices_O?):
            Data.Cinema.Play()
            spawn:
                PlaySound(Data.Sound, Data.MoverTransform, Data.Duration)
            spawn:
                HeroDestructionPath(Data, Devices.Explosive, 0.3, 4, 400.0)
            spawn:
                HeroDestructionPath(Data, Devices.Explosive, 1.5, 4, 400.0)
            Sleep(Data.Duration)
            
    ActivateDemonSlayer1(Devices_O : ?finisher_effects, Data:finisher_data)<suspends>:void=
        Forward := Data.PlayerTransform.Rotation.GetLocalForward()
        NewLoc := Data.PlayerTransform.Translation + Forward * 361.0
        if (Devices := Devices_O?):
            Data.Cinema.Play()
            spawn:
                PlaySound(Data.Sound, Data.MoverTransform, Data.Duration)
            spawn:
                HeroDestructionPath(Data, Devices.Explosive, 0.8, 2, 200.0)
            spawn:
                DemonSlayer1Vfx(Data.MoverTransform, Data.TestOffset)
            Sleep(Data.Duration)
            IsValid := IsValidLocation(NewLoc)
            if (IsValid = true):
                if (Devices.Teleporter.TeleportTo[NewLoc, DEFAULT_ROTATION]):
                    Sleep(0.0)
                    Devices.Teleporter.Teleport(Data.Instigator)

    DemonSlayer1Vfx(MoverT : transform, TestOffset:vector3)<suspends>:void=
        Forward := MoverT.Rotation.GetLocalForward()
        NewLoc := MoverT.Translation + Forward * 175.0 + vector3 {Z:= 90.0}
        Sleep(0.6)
        MyVFX:cancelable = SpawnParticleSystem(_VFX_ToUEFN.NS_Slash_Water_asset, NewLoc + TestOffset, ?Rotation := MoverT.Rotation, ?StartDelay := 0.0)
        Sleep(1.0)
        MyVFX.Cancel()

    ActivateDemonSlayer2(Devices_O : ?finisher_effects, Data:finisher_data)<suspends>:void=
        Forward := Data.PlayerTransform.Rotation.GetLocalForward()
        NewLoc := Data.PlayerTransform.Translation + Forward * 330.0
        if (Devices := Devices_O?):
            Data.Cinema.Play()
            spawn:
                PlaySound(Data.Sound, Data.MoverTransform, Data.Duration)
            spawn:
                HeroDestructionPath(Data, Devices.Explosive, 1.8, 3, 230.0)
            spawn:
                DemonSlayer2Vfx(Data.MoverTransform, Data.TestOffset)
            Sleep(Data.Duration)
            IsValid := IsValidLocation(NewLoc)
            if (IsValid = true):
                if (Devices.Teleporter.TeleportTo[NewLoc, DEFAULT_ROTATION]):
                    Sleep(0.0)
                    Devices.Teleporter.Teleport(Data.Instigator)

    DemonSlayer2Vfx(MoverT : transform, TestOffset:vector3)<suspends>:void=
        Forward := MoverT.Rotation.GetLocalForward()
        NewLoc := MoverT.Translation + Forward * 325.0 + vector3 {Z:= 90.0}
        Sleep(1.93)
        MyVFX:cancelable = SpawnParticleSystem(_VFX_ToUEFN.NS_Slash_Fire_asset, NewLoc + TestOffset, ?Rotation := MoverT.Rotation, ?StartDelay := 0.0)
        Sleep(1.0)
        MyVFX.Cancel()

    ActivateDemonSlayer3(Devices_O : ?finisher_effects, Data:finisher_data)<suspends>:void=
        Forward := Data.PlayerTransform.Rotation.GetLocalForward()
        NewLoc := Data.PlayerTransform.Translation + Forward * 550.0
        if (Devices := Devices_O?):
            Data.Cinema.Play()
            spawn:
                PlaySound(Data.Sound, Data.MoverTransform, Data.Duration)
            spawn:
                HeroDestructionPath(Data, Devices.Explosive, 0.35, 2, 300.0)
            spawn:
                DemonSlayer3Vfx(Data.MoverTransform, Data.TestOffset)
            Sleep(Data.Duration)
            IsValid := IsValidLocation(NewLoc)
            if (IsValid = true):
                if (Devices.Teleporter.TeleportTo[NewLoc, DEFAULT_ROTATION]):
                    Sleep(0.0)
                    Devices.Teleporter.Teleport(Data.Instigator)

    DemonSlayer3Vfx(MoverT : transform, TestOffset:vector3)<suspends>:void=
        Forward := MoverT.Rotation.GetLocalForward()
        NewLoc := MoverT.Translation + Forward * 350.0 + vector3 {Z:= 90.0}
        NewLoc2 := MoverT.Translation + Forward * 360.0 + vector3 {Z:= 90.0}
        Sleep(0.47)
        MyVFX:cancelable = SpawnParticleSystem(_VFX_ToUEFN.NS_Piercing_Electric_Blue_asset, NewLoc + TestOffset, ?Rotation := MoverT.Rotation, ?StartDelay := 0.0)
        MyVFX2:cancelable = SpawnParticleSystem(_VFX_ToUEFN.NS_Piercing_Electric_Yellow_asset, NewLoc2 + TestOffset, ?Rotation := MoverT.Rotation, ?StartDelay := 0.0)
        Sleep(1.0)
        MyVFX.Cancel()
        MyVFX2.Cancel()

    ActivateDemonSlayer4(Devices_O : ?finisher_effects, Data:finisher_data)<suspends>:void=
        if (Devices := Devices_O?):
            Data.Cinema.Play()
            spawn:
                PlaySound(Data.Sound, Data.MoverTransform, Data.Duration)
            spawn:
                DemonSlayer4Vfx(Data.MoverTransform, Data.TestOffset)
            Sleep(Data.Duration)

    DemonSlayer4Vfx(MoverT : transform, TestOffset:vector3)<suspends>:void=
        Forward := MoverT.Rotation.GetLocalForward()
        NewLoc := MoverT.Translation + Forward * 145.0 + vector3 {Z:= 190.0}
        NewRot := MoverT.Rotation.ApplyRoll(DegreesToRadians(90.0))
        Sleep(0.47)
        MyVFX:cancelable = SpawnParticleSystem(_VFX_ToUEFN.NS_Slash_Electric_Pink_butterfly_asset, NewLoc + TestOffset, ?Rotation := NewRot, ?StartDelay := 0.0)
        Sleep(1.0)
        MyVFX.Cancel()

    ActivateMK2(Devices_O : ?finisher_effects, Data:finisher_data)<suspends>:void=
        Forward := Data.PlayerTransform.Rotation.GetLocalForward()
        NewLoc := Data.PlayerTransform.Translation + Forward * 400.0
        if (Devices := Devices_O?):
            Data.Cinema.Play()
            spawn:
                PlaySound(Data.Sound, Data.MoverTransform, Data.Duration)
            spawn:
                HeroDestructionPath(Data, Devices.Explosive, 0.8, 3, 250.0)
            Sleep(Data.Duration)
            IsValid := IsValidLocation(NewLoc)
            if (IsValid = true):
                if (Devices.Teleporter.TeleportTo[NewLoc, DEFAULT_ROTATION]):
                    Sleep(0.0)
                    Devices.Teleporter.Teleport(Data.Instigator)

    ActivateMK3(Devices_O : ?finisher_effects, Data:finisher_data)<suspends>:void=
        Forward := Data.PlayerTransform.Rotation.GetLocalForward()
        NewLoc := Data.PlayerTransform.Translation + Forward * 370.0
        if (Devices := Devices_O?):
            Data.Cinema.Play()
            spawn:
                PlaySound(Data.Sound, Data.MoverTransform, Data.Duration)
            spawn:
                HeroDestructionPath(Data, Devices.Explosive, 2.0, 2, 200.0)
            spawn:
                Mk3Vfx(Data.MoverTransform, Data.TestOffset)
            Sleep(Data.Duration)
            IsValid := IsValidLocation(NewLoc)
            if (IsValid = true):
                if (Devices.Teleporter.TeleportTo[NewLoc, DEFAULT_ROTATION]):
                    Sleep(0.0)
                    Devices.Teleporter.Teleport(Data.Target)
        
    Mk3Vfx(MoverT : transform, TestOffset:vector3)<suspends>:void=
        Forward := MoverT.Rotation.GetLocalForward()
        NewLoc := MoverT.Translation + Forward * 140.0 + vector3 {Z:= 100.0}
        Sleep(0.73)
        MyVFX:cancelable = SpawnParticleSystem(_VFX_ToUEFN.NS_Ice_Magic_Muzzle_asset, NewLoc + TestOffset, ?Rotation := MoverT.Rotation, ?StartDelay := 0.0)
        Sleep(0.8)
        MyVFX.Cancel()

    ActivateInvincibleTrain(Devices_O : ?finisher_effects, Data:finisher_data)<suspends>:void=
        if (Devices := Devices_O?):
            Data.Cinema.Play()
            spawn:
                PlaySound(Data.Sound, Data.MoverTransform, Data.Duration)
            spawn:
                HeroDestructionPath(Data, Devices.Explosive, 1.5, 4, 400.0)
            Sleep(Data.Duration)

    ActivateInvincibleThink(Devices_O : ?finisher_effects, Data:finisher_data)<suspends>:void=
        Forward := Data.PlayerTransform.Rotation.GetLocalForward()
        NewLoc := Data.PlayerTransform.Translation + Forward * 1160.0
        if (Devices := Devices_O?):
            Data.Cinema.Play()
            spawn:
                AddCamera(Data.Instigator, Data.Target, Devices.ThinkCamera)
            spawn:
                PlaySound(Data.Sound, Data.MoverTransform, Data.Duration)
            spawn:
                HeroDestructionPath(Data, Devices.Explosive, 2.2, 4, 400.0)
            spawn:
                DomainDestruction(Data.Instigator, Data.MoverTransform.Translation, Devices.GiantExplosive, 0.5, 2)
            Sleep(Data.Duration)
            RemoveCamera(Data.Instigator, Data.Target, Devices.ThinkCamera)

            IsValid := IsValidLocation(NewLoc)
            if (IsValid = true):
                if (Devices.Teleporter.TeleportTo[NewLoc, DEFAULT_ROTATION]):
                    Sleep(0.0)
                    Devices.Teleporter.Teleport(Data.Instigator)

    ActivateZoro1(Devices_O : ?finisher_effects, Data:finisher_data)<suspends>:void=
        Forward := Data.PlayerTransform.Rotation.GetLocalForward()
        NewLoc := Data.PlayerTransform.Translation + Forward * -175.0
        if (Devices := Devices_O?):
            Data.Cinema.Play()
            spawn:
                PlaySound(Data.Sound, Data.MoverTransform, Data.Duration)
            spawn:
                HeroDestructionPath(Data, Devices.Explosive, 0.5, 6, 200.0)
            spawn:
                Zoro1Vfx(Data.MoverTransform, Data.TestOffset)
            Sleep(Data.Duration)
            IsValid := IsValidLocation(NewLoc)
            if (IsValid = true):
                if (Devices.Teleporter.TeleportTo[NewLoc, DEFAULT_ROTATION]):
                    Sleep(0.0)
                    Devices.Teleporter.Teleport(Data.Instigator)

    Zoro1Vfx(MoverT : transform, TestOffset:vector3)<suspends>:void=
        Forward := MoverT.Rotation.GetLocalForward()
        NewLoc := MoverT.Translation  # + Forward * 90.0
        Sleep(0.5)
        MyVFX:cancelable = SpawnParticleSystem(_VFX_ToUEFN.NS_Zoro_RangeSlash_asset, NewLoc + TestOffset, ?Rotation := MoverT.Rotation, ?StartDelay := 0.0)
        Sleep(0.8)
        MyVFX.Cancel()

    ActivateZoro2(Devices_O : ?finisher_effects, Data:finisher_data)<suspends>:void=
        Forward := Data.PlayerTransform.Rotation.GetLocalForward()
        NewLoc := Data.PlayerTransform.Translation + Forward * 365.0
        if (Devices := Devices_O?):
            Data.Cinema.Play()
            spawn:
                PlaySound(Data.Sound, Data.MoverTransform, Data.Duration)
            spawn:
                HeroDestructionPath(Data, Devices.Explosive, 0.7, 2, 200.0)
            spawn:
                Zoro2Vfx(Data.MoverTransform, Data.TestOffset)
            Sleep(Data.Duration)
            IsValid := IsValidLocation(NewLoc)
            if (IsValid = true):
                if (Devices.Teleporter.TeleportTo[NewLoc, DEFAULT_ROTATION]):
                    Sleep(0.0)
                    Devices.Teleporter.Teleport(Data.Instigator)

    Zoro2Vfx(MoverT : transform, TestOffset:vector3)<suspends>:void=
        Forward := MoverT.Rotation.GetLocalForward()
        NewLoc := MoverT.Translation + Forward * 190.0 + vector3 {Z:= 100.0}
        Sleep(0.8)
        MyVFX:cancelable = SpawnParticleSystem(_VFX_ToUEFN.NS_Zoro_SlashCluster_asset, NewLoc + TestOffset, ?Rotation := MoverT.Rotation, ?StartDelay := 0.0)
        Sleep(1.0)
        MyVFX.Cancel()

    ActivateSanji1(Devices_O : ?finisher_effects, Data:finisher_data)<suspends>:void=
        Forward := Data.PlayerTransform.Rotation.GetLocalForward()
        NewLoc := Data.PlayerTransform.Translation + Forward * 365.0
        if (Devices := Devices_O?):
            Data.Cinema.Play()
            spawn:
                PlaySound(Data.Sound, Data.MoverTransform, Data.Duration)
            spawn:
                HeroDestructionPath(Data, Devices.Explosive, 1.35, 5, 200.0)
            Sleep(Data.Duration)
            IsValid := IsValidLocation(NewLoc)
            if (IsValid = true):
                if (Devices.Teleporter.TeleportTo[NewLoc, DEFAULT_ROTATION]):
                    Sleep(0.0)
                    Devices.Teleporter.Teleport(Data.Instigator)

    ActivateSanji2(Devices_O : ?finisher_effects, Data:finisher_data)<suspends>:void=
        Forward := Data.PlayerTransform.Rotation.GetLocalForward()
        NewLoc := Data.PlayerTransform.Translation + Forward * 535.0
        if (Devices := Devices_O?):
            Data.Cinema.Play()
            spawn:
                PlaySound(Data.Sound, Data.MoverTransform, Data.Duration)
            spawn:
                HeroDestructionPath(Data, Devices.Explosive, 2.75, 4, 200.0)
            spawn:
                Sanji2Vfx(Data.MoverTransform, Data.TestOffset)
            Sleep(Data.Duration)
            IsValid := IsValidLocation(NewLoc)
            if (IsValid = true):
                if (Devices.Teleporter.TeleportTo[NewLoc, DEFAULT_ROTATION]):
                    Sleep(0.0)
                    Devices.Teleporter.Teleport(Data.Target)

    Sanji2Vfx(MoverT : transform, TestOffset:vector3)<suspends>:void=
        Forward := MoverT.Rotation.GetLocalForward()
        NewLoc := MoverT.Translation + vector3 {Z := 155.0} + Forward * 145.0
        ImpactRot := MoverT.Rotation.ApplyPitch(DegreesToRadians(90.0))
        Sleep(2.6)
        MyVFX:cancelable = SpawnParticleSystem(Asset := _VFX_ToUEFN.N_PhysicalStrength_small_asset, Position := NewLoc + TestOffset, ?Rotation := ImpactRot, ?StartDelay := 0.0)
        MyVFX2:cancelable = SpawnParticleSystem(_VFX_ToUEFN.NS_OnePiece_Hit_Simple_asset, NewLoc + TestOffset, ?Rotation := MoverT.Rotation, ?StartDelay := 0.0)
        Sleep(0.8)
        MyVFX.Cancel()
        MyVFX2.Cancel()

    ActivateRockLee(Devices_O : ?finisher_effects, Data:finisher_data)<suspends>:void=
        Forward := Data.PlayerTransform.Rotation.GetLocalForward()
        NewLoc := Data.PlayerTransform.Translation + Forward * 530.0
        if (Devices := Devices_O?):
            Data.Cinema.Play()
            spawn:
                PlaySound(Data.Sound, Data.MoverTransform, Data.Duration)
            spawn:
                DomainDestruction(Data.Instigator, Data.MoverTransform.Translation, Devices.DomainExplosive, 0.7, 1)
            Sleep(Data.Duration)
            IsValid := IsValidLocation(NewLoc)
            if (IsValid = true):
                if (Devices.Teleporter.TeleportTo[NewLoc, DEFAULT_ROTATION]):
                    Sleep(0.0)
                    Devices.Teleporter.Teleport(Data.Instigator)

    ActivateSharingan(Devices_O : ?finisher_effects, Data:finisher_data)<suspends>:void=
        if (Devices := Devices_O?):
            Data.Cinema.Play()
            spawn:
                PlaySound(Data.Sound, Data.MoverTransform, Data.Duration)
            spawn:
                AddSharinganWidget(GetRandomSharinganHud(), Data.Instigator, Data.Target, 0.3)
            Sleep(Data.Duration)

    GetRandomSharinganHud():hud_message_device=
        DiceRoll := GetRandomFloat(0.0, 1.0)
        if (DiceRoll <= 0.2):
            return SharinganHud
        else if (DiceRoll <= 0.4):
            return SharinganItachiHud
        else if (DiceRoll <= 0.6):
            return SharinganKakashiHud
        else if (DiceRoll <= 0.8):
            return SharinganSasukeHud
        else:
            return SharinganShusuiHud

    AddSharinganWidget(Hud: hud_message_device, Instigator:agent, Target:agent, WaitTime:float)<suspends>:void=
        Sleep(WaitTime)
        Hud.Show(Instigator)
        Hud.Show(Target)
    
    ActivateRasenshuriken(Devices_O : ?finisher_effects, Data:finisher_data)<suspends>:void=
        if (Devices := Devices_O?):
            Data.Cinema.Play()
            spawn:
                PlaySound(Data.Sound, Data.MoverTransform, Data.Duration)
            spawn:
                HeroDestructionPath(Data, Devices.Explosive, 2.6, 4, 200.0)
            spawn:
                RasenshurikenExplosionVfx(Data.MoverTransform, Data.TestOffset)
            Sleep(Data.Duration)

    RasenshurikenExplosionVfx(MoverT : transform, TestOffset:vector3)<suspends>:void=
        Forward := MoverT.Rotation.GetLocalForward()
        NewLoc := MoverT.Translation + vector3 {Z := 110.0} + Forward * 580.0
        Sleep(2.7)
        MyVFX:cancelable = SpawnParticleSystem(Asset := _VFX_ToUEFN.NS_Rasenshuriken_explosion_asset, Position := NewLoc + TestOffset, ?Rotation := MoverT.Rotation, ?StartDelay := 0.0)
        Sleep(3.0)
        MyVFX.Cancel()

    ActivateBijuu(Devices_O : ?finisher_effects, Data:finisher_data)<suspends>:void=
        if (Devices := Devices_O?):
            Data.Cinema.Play()
            spawn:
                PlaySound(Data.Sound, Data.MoverTransform, Data.Duration)
            spawn:
                HeroDestructionPath(Data, Devices.Explosive, 4.5, 4, 200.0)
            spawn:
                BijuuExplosionVfx(Data.MoverTransform, Data.TestOffset)
            Sleep(Data.Duration)

    BijuuExplosionVfx(MoverT : transform, TestOffset:vector3)<suspends>:void=
        Forward := MoverT.Rotation.GetLocalForward()
        NewLoc := MoverT.Translation + vector3 {Z := 50.0} + Forward * 590.0
        Sleep(4.4)
        MyVFX:cancelable = SpawnParticleSystem(Asset := _VFX_ToUEFN.NS_Bijuu_explosion_asset, Position := NewLoc + TestOffset, ?Rotation := MoverT.Rotation, ?StartDelay := 0.0)
        Sleep(3.0)
        MyVFX.Cancel()

    ActivateAoT1(Devices_O : ?finisher_effects, Data:finisher_data)<suspends>:void=
        Forward := Data.PlayerTransform.Rotation.GetLocalForward()
        NewLoc := Data.PlayerTransform.Translation + Forward * 560.0
        if (Devices := Devices_O?):
            Data.Cinema.Play()
            spawn:
                PlaySound(Data.Sound, Data.MoverTransform, Data.Duration)
            spawn:
                HeroDestructionPath(Data, Devices.Explosive, 0.7, 4, 200.0)
            spawn:
                AotSpikesVfx(Data.MoverTransform, Data.TestOffset)
            Sleep(Data.Duration)
            IsValid := IsValidLocation(NewLoc)
            if (IsValid = true):
                if (Devices.Teleporter.TeleportTo[NewLoc, DEFAULT_ROTATION]):
                    Sleep(0.0)
                    Devices.Teleporter.Teleport(Data.Target)

    AotSpikesVfx(MoverT : transform, TestOffset:vector3)<suspends>:void=
        Forward := MoverT.Rotation.GetLocalForward()
        Loc := MoverT.Translation + Forward * 560.0
        Sleep(0.3)
        StartVFX:cancelable = SpawnParticleSystem(_VFX_ToUEFN.NS_AoT_Spikes_asset, Loc, ?Rotation := MoverT.Rotation, ?StartDelay := 0.0)
        Sleep(3.0)
        StartVFX.Cancel()

    ActivateAoT2(Devices_O : ?finisher_effects, Data:finisher_data)<suspends>:void=
        Forward := Data.PlayerTransform.Rotation.GetLocalForward()
        NewLoc := Data.PlayerTransform.Translation + Forward * 200.0
        ExplosionLocation := NewLoc + vector3 {Z:= 400.0}
        if (Devices := Devices_O?):
            Data.Cinema.Play()
            spawn:
                AddCamera(Data.Instigator, Data.Target, Devices.ThinkCamera)
            spawn:
                PlaySound(Data.Sound, Data.MoverTransform, Data.Duration)
            spawn:
                DomainDestruction(Data.Instigator, ExplosionLocation, Devices.GiantExplosive, 0.6, 6)
            spawn:
                AotTransformVfx(Data.MoverTransform, Data.TestOffset)
            Sleep(Data.Duration)
            RemoveCamera(Data.Instigator, Data.Target, Devices.ThinkCamera)

            IsValid := IsValidLocation(NewLoc)
            if (IsValid = true):
                if (Devices.Teleporter.TeleportTo[NewLoc, DEFAULT_ROTATION]):
                    Sleep(0.0)
                    Devices.Teleporter.Teleport(Data.Instigator)

    AotTransformVfx(MoverT : transform, TestOffset:vector3)<suspends>:void=
        Forward := MoverT.Rotation.GetLocalForward()
        Loc := MoverT.Translation + Forward * 350.0 + vector3 {Z:=-1350.0} # 620
        Sleep(1.2)
        StartVFX:cancelable = SpawnParticleSystem(_VFX_ToUEFN.NS_AoT_Transform_asset, Loc, ?Rotation := MoverT.Rotation, ?StartDelay := 0.0)
        Sleep(1.5)
        StartVFX.Cancel()
        Sleep(0.1)
        EndLoc := MoverT.Translation + Forward * 380.0 + vector3 {Z:=170.0} # F := 410
        EndVFX:cancelable = SpawnParticleSystem(_VFX_ToUEFN.NS_Missile_Smoke_asset, EndLoc, ?Rotation := MoverT.Rotation, ?StartDelay := 0.0)
        Sleep(0.5)
        EndVFX.Cancel()
    
    ActivateAoT3(Devices_O : ?finisher_effects, Data:finisher_data)<suspends>:void=
        Forward := Data.PlayerTransform.Rotation.GetLocalForward()
        Right := Data.PlayerTransform.Rotation.GetLocalRight()
        NewLoc := Data.PlayerTransform.Translation + Forward * -135.0 + Right * -380.0
        ExplosionLocation1 := Data.PlayerTransform.Translation + Forward * 770.0 + vector3 {Z:= 400.0}
        ExplosionLocation2 := Data.PlayerTransform.Translation + Forward * 1100.0 + vector3 {Z:= 600.0}
        if (Devices := Devices_O?):
            Data.Cinema.Play()
            spawn:
                AddCamera(Data.Instigator, Data.Target, Devices.ThinkCamera)
            spawn:
                PlaySound(Data.Sound, Data.MoverTransform, Data.Duration)
            spawn:
                DomainDestruction(Data.Instigator, ExplosionLocation1, Devices.GiantExplosive, 0.0, 2)
            spawn:
                DomainDestruction(Data.Instigator, ExplosionLocation2, Devices.GiantExplosive, 1.5, 4)
            spawn:
                AotSmokeVfx(Data.MoverTransform, Data.TestOffset)
            spawn:
                AoTBoostVfx(Data.MoverTransform, Data.TestOffset)
            Sleep(Data.Duration)
            RemoveCamera(Data.Instigator, Data.Target, Devices.ThinkCamera)

            IsValid := IsValidLocation(NewLoc)
            if (IsValid = true):
                if (Devices.Teleporter.TeleportTo[NewLoc, DEFAULT_ROTATION]):
                    Sleep(0.0)
                    Devices.Teleporter.Teleport(Data.Instigator)

    AotSmokeVfx(MoverT : transform, TestOffset:vector3)<suspends>:void=
        Forward := MoverT.Rotation.GetLocalForward()
        Right := MoverT.Rotation.GetLocalRight()
        Loc := MoverT.Translation + Forward * 500.0 + Right * 420.0 + vector3 {Z:= 130.0}
        Sleep(5.1)
        StartVFX:cancelable = SpawnParticleSystem(_VFX_ToUEFN.NS_Missile_Smoke_asset, Loc, ?Rotation := MoverT.Rotation, ?StartDelay := 0.0)
        Sleep(0.5)
        StartVFX.Cancel()

    AoTBoostVfx(MoverT : transform, TestOffset:vector3)<suspends>:void=
        Forward := MoverT.Rotation.GetLocalForward()
        Right := MoverT.Rotation.GetLocalRight()
        Loc := MoverT.Translation + Forward * 1330.0 + Right * 260.0 + vector3 {Z:= 780.0}
        Sleep(2.5)
        StartVFX:cancelable = SpawnParticleSystem(_VFX_ToUEFN.NS_AirDodge_mod_asset, Loc, ?Rotation := MoverT.Rotation, ?StartDelay := 0.0)
        Sleep(1.0)
        StartVFX.Cancel()

    ActivateDeku1(Devices_O : ?finisher_effects, Data:finisher_data)<suspends>:void=
        Forward := Data.PlayerTransform.Rotation.GetLocalForward()
        NewLoc := Data.PlayerTransform.Translation + Forward * 530.0
        if (Devices := Devices_O?):
            Data.Cinema.Play()
            spawn:
                PlaySound(Data.Sound, Data.MoverTransform, Data.Duration)
            spawn:
                HeroDestructionPath(Data, Devices.Explosive, 1.8, 7, 200.0)
            Sleep(Data.Duration)
            IsValid := IsValidLocation(NewLoc)
            if (IsValid = true):
                if (Devices.Teleporter.TeleportTo[NewLoc, DEFAULT_ROTATION]):
                    Sleep(0.0)
                    Devices.Teleporter.Teleport(Data.Instigator)

    ActivateTodoroki1(Devices_O : ?finisher_effects, Data:finisher_data)<suspends>:void=
        Forward := Data.PlayerTransform.Rotation.GetLocalForward()
        NewLoc := Data.PlayerTransform.Translation + Forward * 500.0
        if (Devices := Devices_O?):
            Data.Cinema.Play()
            spawn:
                PlaySound(Data.Sound, Data.MoverTransform, Data.Duration)
            spawn:
                HeroDestructionPath(Data, Devices.Explosive, 2.4, 3, 200.0)
            spawn:
                TodorokiVfx(Data.MoverTransform, Data.TestOffset)
            Sleep(Data.Duration)
            IsValid := IsValidLocation(NewLoc)
            if (IsValid = true):
                if (Devices.Teleporter.TeleportTo[NewLoc, DEFAULT_ROTATION]):
                    Sleep(0.0)
                    Devices.Teleporter.Teleport(Data.Target)

    TodorokiVfx(MoverT : transform, TestOffset:vector3)<suspends>:void=
        Forward := MoverT.Rotation.GetLocalForward()
        Right := MoverT.Rotation.GetLocalRight()
        Loc := MoverT.Translation + Forward * 260.0
        NewLoc := MoverT.Translation + Forward * 90.0 + vector3{Z:= 115.0}
        Sleep(0.7)
        StartVFX:cancelable = SpawnParticleSystem(_VFX_ToUEFN.NS_SleepInFreeze_asset, Loc, ?Rotation := MoverT.Rotation, ?StartDelay := 0.0)
        Sleep(1.3)
        RedHawkVFX:cancelable = SpawnParticleSystem(_VFX_ToUEFN.NS_OnePiece_RedHawk_asset, NewLoc + TestOffset, ?Rotation := MoverT.Rotation, ?StartDelay := 0.0)
        Sleep(0.4)
        StartVFX.Cancel()
        Sleep(0.6)
        RedHawkVFX.Cancel()

    ActivateTodoroki2(Devices_O : ?finisher_effects, Data:finisher_data)<suspends>:void=
        Forward := Data.PlayerTransform.Rotation.GetLocalForward()
        NewLoc := Data.PlayerTransform.Translation + Forward * 500.0
        if (Devices := Devices_O?):
            Data.Cinema.Play()
            spawn:
                PlaySound(Data.Sound, Data.MoverTransform, Data.Duration)
            spawn:
                Todoroki2Vfx(Data.MoverTransform, Data.TestOffset)
            Sleep(Data.Duration)

    Todoroki2Vfx(MoverT : transform, TestOffset:vector3)<suspends>:void=
        Forward := MoverT.Rotation.GetLocalForward()
        Right := MoverT.Rotation.GetLocalRight()
        Loc := MoverT.Translation + Forward * 80.0
        Rot := MoverT.Rotation.ApplyPitch(DegreesToRadians(30.0))
        Sleep(0.4)
        StartVFX:cancelable = SpawnParticleSystem(_VFX_ToUEFN.NS_GlacierPoke_asset, Loc, ?Rotation := Rot, ?StartDelay := 0.0)
        Sleep(0.8)
        StartVFX.Cancel()

    ActivateBakugo1(Devices_O : ?finisher_effects, Data:finisher_data)<suspends>:void=
        Forward := Data.PlayerTransform.Rotation.GetLocalForward()
        NewLoc := Data.PlayerTransform.Translation + Forward * 900.0
        if (Devices := Devices_O?):
            Data.Cinema.Play()
            spawn:
                PlaySound(Data.Sound, Data.MoverTransform, Data.Duration)
            spawn:
                HeroDestructionPath(Data, Devices.Explosive, 2.1, 5, 200.0)
            spawn:
                Bakugo1Vfx(Data.MoverTransform, Data.TestOffset)
            spawn:
                AddCamera(Data.Instigator, Data.Target, Devices.ThinkCamera)
            Sleep(Data.Duration)
            RemoveCamera(Data.Instigator, Data.Target, Devices.ThinkCamera)
            IsValid := IsValidLocation(NewLoc)
            if (IsValid = true):
                if (Devices.Teleporter.TeleportTo[NewLoc, DEFAULT_ROTATION]):
                    Sleep(0.0)
                    Devices.Teleporter.Teleport(Data.Target)

    Bakugo1Vfx(MoverT : transform, TestOffset:vector3)<suspends>:void=
        Forward := MoverT.Rotation.GetLocalForward()
        Right := MoverT.Rotation.GetLocalRight()
        Loc := MoverT.Translation + Forward * 415.0 + vector3{Z:= 200.0}
        Sleep(0.5)
        Vfx1:cancelable = SpawnParticleSystem(_VFX_ToUEFN.NS_Hit_Fire_01_asset, Loc, ?Rotation := MoverT.Rotation, ?StartDelay := 0.0)
        Sleep(0.3)
        Vfx1.Cancel()
        Sleep(0.3)
        Vfx2:cancelable = SpawnParticleSystem(_VFX_ToUEFN.NS_Hit_Fire_01_asset, Loc, ?Rotation := MoverT.Rotation, ?StartDelay := 0.0)
        Sleep(0.3)
        Vfx2.Cancel()
        Vfx3:cancelable = SpawnParticleSystem(_VFX_ToUEFN.NS_Hit_Fire_01_asset, Loc, ?Rotation := MoverT.Rotation, ?StartDelay := 0.0)
        Sleep(0.3)
        Vfx3.Cancel()
        Sleep(0.3)
        Vfx4:cancelable = SpawnParticleSystem(_VFX_ToUEFN.NS_Hit_Fire_01_asset, Loc, ?Rotation := MoverT.Rotation, ?StartDelay := 0.0)
        Sleep(0.3)
        Vfx4.Cancel()

    ActivateBakugo2(Devices_O : ?finisher_effects, Data:finisher_data)<suspends>:void=
        Forward := Data.PlayerTransform.Rotation.GetLocalForward()
        NewLoc := Data.PlayerTransform.Translation + Forward * 500.0
        if (Devices := Devices_O?):
            Data.Cinema.Play()
            spawn:
                PlaySound(Data.Sound, Data.MoverTransform, Data.Duration)
            spawn:
                HeroDestructionPath(Data, Devices.Explosive, 1.0, 3, 200.0)
            spawn:
                Bakugo2Vfx(Data.MoverTransform, Data.TestOffset)
            Sleep(Data.Duration)
            IsValid := IsValidLocation(NewLoc)
            if (IsValid = true):
                if (Devices.Teleporter.TeleportTo[NewLoc, DEFAULT_ROTATION]):
                    Sleep(0.0)
                    Devices.Teleporter.Teleport(Data.Instigator)

    Bakugo2Vfx(MoverT : transform, TestOffset:vector3)<suspends>:void=
        Forward := MoverT.Rotation.GetLocalForward()
        Right := MoverT.Rotation.GetLocalRight()
        Loc := MoverT.Translation + Forward * 330.0 + vector3{Z:= 140.0}
        Sleep(0.8)
        Vfx1:cancelable = SpawnParticleSystem(_VFX_ToUEFN.NS_Hit_Fire_01_asset, Loc, ?Rotation := MoverT.Rotation, ?StartDelay := 0.0)
        Sleep(0.3)
        Vfx1.Cancel()

    ActivateGuns1(Devices_O : ?finisher_effects, Data:finisher_data)<suspends>:void=
        Forward := Data.PlayerTransform.Rotation.GetLocalForward()
        NewLoc := Data.PlayerTransform.Translation + Forward * 235.0
        if (Devices := Devices_O?):
            Data.Cinema.Play()
            spawn:
                PlaySound(Data.Sound, Data.MoverTransform, Data.Duration)
            spawn:
                HeroDestructionPath(Data, Devices.Explosive, 2.5, 5, 200.0)
            spawn:
                Guns1Vfx(Data.MoverTransform, Data.TestOffset)
            Sleep(Data.Duration)
            IsValid := IsValidLocation(NewLoc)
            if (IsValid = true):
                if (Devices.Teleporter.TeleportTo[NewLoc, DEFAULT_ROTATION]):
                    Sleep(0.0)
                    Devices.Teleporter.Teleport(Data.Instigator)

    Guns1Vfx(MoverT : transform, TestOffset:vector3)<suspends>:void=
        Forward := MoverT.Rotation.GetLocalForward()
        Right := MoverT.Rotation.GetLocalRight()
        Loc := MoverT.Translation + Forward * 230.0 + vector3{Z:= 85.0}
        Loc2 := MoverT.Translation + Forward * 230.0 + vector3{Z:= 128.0}
        Loc3 := MoverT.Translation + Forward * 370.0 + vector3{Z:= 128.0}
        Sleep(0.5)
        Vfx1:cancelable = SpawnParticleSystem(_VFX_ToUEFN.NS_Fire_Muzzle_asset, Loc, ?Rotation := MoverT.Rotation, ?StartDelay := 0.0)
        Sleep(0.2)
        Vfx2:cancelable = SpawnParticleSystem(_VFX_ToUEFN.NS_Fire_Muzzle_asset, Loc2, ?Rotation := MoverT.Rotation, ?StartDelay := 0.0)
        Sleep(0.3)
        Vfx1.Cancel()
        Sleep(0.2)
        Vfx2.Cancel()
        Sleep(1.5)
        Vfx3:cancelable = SpawnParticleSystem(_VFX_ToUEFN.NS_Fire_Muzzle_asset, Loc3, ?Rotation := MoverT.Rotation, ?StartDelay := 0.0)
        Sleep(0.5)
        Vfx3.Cancel()

    ActivateGuns2(Devices_O : ?finisher_effects, Data:finisher_data)<suspends>:void=
        Forward := Data.PlayerTransform.Rotation.GetLocalForward()
        NewLoc := Data.PlayerTransform.Translation + Forward * 1150.0
        if (Devices := Devices_O?):
            Data.Cinema.Play()
            spawn:
                PlaySound(Data.Sound, Data.MoverTransform, Data.Duration)
            spawn:
                HeroDestructionPath(Data, Devices.Explosive, 0.3, 4, 200.0)
            spawn:
                HeroDestructionPath(Data, Devices.Explosive, 2.8, 6, 200.0)
            spawn:
                Guns2Vfx(Data.MoverTransform, Data.TestOffset)
            Sleep(Data.Duration)
            IsValid := IsValidLocation(NewLoc)
            if (IsValid = true):
                if (Devices.Teleporter.TeleportTo[NewLoc, DEFAULT_ROTATION]):
                    Sleep(0.0)
                    Devices.Teleporter.Teleport(Data.Target)

    Guns2Vfx(MoverT : transform, TestOffset:vector3)<suspends>:void=
        Forward := MoverT.Rotation.GetLocalForward()
        Right := MoverT.Rotation.GetLocalRight()
        Loc := MoverT.Translation + Forward * 300.0 + vector3{Z:= 95.0}
        Loc2 := MoverT.Translation + Forward * 218.0 + vector3{Z:= 104.0}
        Sleep(1.1)
        Vfx1:cancelable = SpawnParticleSystem(_VFX_ToUEFN.NS_Fire_Muzzle_asset, Loc, ?Rotation := MoverT.Rotation, ?StartDelay := 0.0)
        Sleep(0.5)
        Vfx1.Cancel()
        Sleep(1.1)
        Vfx2:cancelable = SpawnParticleSystem(_VFX_ToUEFN.NS_Fire_Muzzle_asset, Loc2, ?Rotation := MoverT.Rotation, ?StartDelay := 0.0)
        Sleep(0.5)
        Vfx2.Cancel()

    ActivateGuns3(Devices_O : ?finisher_effects, Data:finisher_data)<suspends>:void=
        Forward := Data.PlayerTransform.Rotation.GetLocalForward()
        NewLoc := Data.PlayerTransform.Translation + Forward * 660.0
        if (Devices := Devices_O?):
            Data.Cinema.Play()
            spawn:
                PlaySound(Data.Sound, Data.MoverTransform, Data.Duration)
            spawn:
                HeroDestructionPath(Data, Devices.Explosive, 2.3, 3, 230.0)
            spawn:
                Guns3Vfx(Data.MoverTransform, Data.TestOffset)
            Sleep(Data.Duration)
            IsValid := IsValidLocation(NewLoc)
            if (IsValid = true):
                if (Devices.Teleporter.TeleportTo[NewLoc, DEFAULT_ROTATION]):
                    Sleep(0.0)
                    Devices.Teleporter.Teleport(Data.Target)

    Guns3Vfx(MoverT : transform, TestOffset:vector3)<suspends>:void=
        Forward := MoverT.Rotation.GetLocalForward()
        Right := MoverT.Rotation.GetLocalRight()
        Loc := MoverT.Translation + Forward * 225.0 + vector3{Z:= 138.0}
        Loc3 := MoverT.Translation + Forward * 260.0 + vector3{Z:= 82.0}
        Sleep(0.8)
        Vfx1:cancelable = SpawnParticleSystem(_VFX_ToUEFN.NS_Fire_Muzzle_asset, Loc, ?Rotation := MoverT.Rotation, ?StartDelay := 0.0)
        Sleep(0.3)
        Vfx2:cancelable = SpawnParticleSystem(_VFX_ToUEFN.NS_Fire_Muzzle_asset, Loc, ?Rotation := MoverT.Rotation, ?StartDelay := 0.0)
        Sleep(0.2)
        Vfx1.Cancel()
        Sleep(0.3)
        Vfx2.Cancel()
        Sleep(0.7)
        Vfx3:cancelable = SpawnParticleSystem(_VFX_ToUEFN.NS_Fire_Muzzle_asset, Loc3, ?Rotation := MoverT.Rotation, ?StartDelay := 0.0)
        Sleep(0.5)
        Vfx3.Cancel()

    ActivateGuns4(Devices_O : ?finisher_effects, Data:finisher_data)<suspends>:void=
        Forward := Data.PlayerTransform.Rotation.GetLocalForward()
        NewLoc := Data.PlayerTransform.Translation + Forward * 320.0
        if (Devices := Devices_O?):
            Data.Cinema.Play()
            spawn:
                PlaySound(Data.Sound, Data.MoverTransform, Data.Duration)
            spawn:
                HeroDestructionPath(Data, Devices.Explosive, 1.4, 4, 200.0)
            spawn:
                Guns4Vfx(Data.MoverTransform, Data.TestOffset)
            Sleep(Data.Duration)
            IsValid := IsValidLocation(NewLoc)
            if (IsValid = true):
                if (Devices.Teleporter.TeleportTo[NewLoc, DEFAULT_ROTATION]):
                    Sleep(0.0)
                    Devices.Teleporter.Teleport(Data.Instigator)

    Guns4Vfx(MoverT : transform, TestOffset:vector3)<suspends>:void=
        Forward := MoverT.Rotation.GetLocalForward()
        Right := MoverT.Rotation.GetLocalRight()
        Loc := MoverT.Translation + Forward * 350.0 + vector3{Z:= 108.0}
        Sleep(1.4)
        Vfx1:cancelable = SpawnParticleSystem(_VFX_ToUEFN.NS_Fire_Muzzle_asset, Loc, ?Rotation := MoverT.Rotation, ?StartDelay := 0.0)
        Sleep(0.5)
        Vfx1.Cancel()

    ActivateGuns5(Devices_O : ?finisher_effects, Data:finisher_data)<suspends>:void=
        Forward := Data.PlayerTransform.Rotation.GetLocalForward()
        NewLoc := Data.PlayerTransform.Translation + Forward * 700.0
        if (Devices := Devices_O?):
            Data.Cinema.Play()
            spawn:
                PlaySound(Data.Sound, Data.MoverTransform, Data.Duration)
            spawn:
                HeroDestructionPath(Data, Devices.Explosive, 0.5, 4, 200.0)
            spawn:
                Guns5Vfx(Data.MoverTransform, Data.TestOffset)
            Sleep(Data.Duration)
            IsValid := IsValidLocation(NewLoc)
            if (IsValid = true):
                if (Devices.Teleporter.TeleportTo[NewLoc, DEFAULT_ROTATION]):
                    Sleep(0.0)
                    Devices.Teleporter.Teleport(Data.Instigator)

    Guns5Vfx(MoverT : transform, TestOffset:vector3)<suspends>:void=
        Forward := MoverT.Rotation.GetLocalForward()
        Right := MoverT.Rotation.GetLocalRight()
        Loc := MoverT.Translation + Forward * 375.0 + vector3{Z:= 82.0}
        Sleep(0.5)
        Vfx1:cancelable = SpawnParticleSystem(_VFX_ToUEFN.NS_Fire_Muzzle_asset, Loc, ?Rotation := MoverT.Rotation, ?StartDelay := 0.0)
        Sleep(0.5)
        Vfx1.Cancel()

    ActivateGuns6(Devices_O : ?finisher_effects, Data:finisher_data)<suspends>:void=
        Forward := Data.PlayerTransform.Rotation.GetLocalForward()
        NewLoc := Data.PlayerTransform.Translation + Forward * 700.0
        if (Devices := Devices_O?):
            Data.Cinema.Play()
            spawn:
                PlaySound(Data.Sound, Data.MoverTransform, Data.Duration)
            spawn:
                HeroDestructionPath(Data, Devices.Explosive, 0.7, 4, 200.0)
            spawn:
                Guns6Vfx(Data.MoverTransform, Data.TestOffset)
            Sleep(Data.Duration)
            IsValid := IsValidLocation(NewLoc)
            if (IsValid = true):
                if (Devices.Teleporter.TeleportTo[NewLoc, DEFAULT_ROTATION]):
                    Sleep(0.0)
                    Devices.Teleporter.Teleport(Data.Target)

    Guns6Vfx(MoverT : transform, TestOffset:vector3)<suspends>:void=
        Forward := MoverT.Rotation.GetLocalForward()
        Right := MoverT.Rotation.GetLocalRight()
        Loc := MoverT.Translation + Forward * 185.0 + vector3{Z:= 82.0}
        Sleep(0.6)
        Vfx1:cancelable = SpawnParticleSystem(_VFX_ToUEFN.NS_BlasterProjectile_asset, Loc, ?Rotation := MoverT.Rotation, ?StartDelay := 0.0)
        Sleep(0.5)
        Vfx1.Cancel()

    ActivateGuns7(Devices_O : ?finisher_effects, Data:finisher_data)<suspends>:void=
        Forward := Data.PlayerTransform.Rotation.GetLocalForward()
        NewLoc := Data.PlayerTransform.Translation + Forward * 600.0
        if (Devices := Devices_O?):
            Data.Cinema.Play()
            spawn:
                PlaySound(Data.Sound, Data.MoverTransform, Data.Duration)
            spawn:
                HeroDestructionPath(Data, Devices.Explosive, 0.6, 3, 200.0)
            spawn:
                Guns7Vfx(Data.MoverTransform, Data.TestOffset)
            Sleep(Data.Duration)
            IsValid := IsValidLocation(NewLoc)
            if (IsValid = true):
                if (Devices.Teleporter.TeleportTo[NewLoc, DEFAULT_ROTATION]):
                    Sleep(0.0)
                    Devices.Teleporter.Teleport(Data.Target)

    Guns7Vfx(MoverT : transform, TestOffset:vector3)<suspends>:void=
        Forward := MoverT.Rotation.GetLocalForward()
        Right := MoverT.Rotation.GetLocalRight()
        Loc := MoverT.Translation + Forward * 112.0 + vector3{Z:= 108.0}
        Sleep(0.6)
        Vfx1:cancelable = SpawnParticleSystem(_VFX_ToUEFN.NS_BlasterProjectile_asset, Loc, ?Rotation := MoverT.Rotation, ?StartDelay := 0.0)
        Sleep(0.5)
        Vfx1.Cancel()

    ActivateGuns8(Devices_O : ?finisher_effects, Data:finisher_data)<suspends>:void=
        Forward := Data.PlayerTransform.Rotation.GetLocalForward()
        NewLoc := Data.PlayerTransform.Translation + Forward * 600.0
        if (Devices := Devices_O?):
            Data.Cinema.Play()
            spawn:
                PlaySound(Data.Sound, Data.MoverTransform, Data.Duration)
            spawn:
                HeroDestructionPath(Data, Devices.Explosive, 1.0, 3, 200.0)
            spawn:
                Guns8Vfx(Data.MoverTransform, Data.TestOffset)
            Sleep(Data.Duration)
            IsValid := IsValidLocation(NewLoc)
            if (IsValid = true):
                if (Devices.Teleporter.TeleportTo[NewLoc, DEFAULT_ROTATION]):
                    Sleep(0.0)
                    Devices.Teleporter.Teleport(Data.Target)

    Guns8Vfx(MoverT : transform, TestOffset:vector3)<suspends>:void=
        Forward := MoverT.Rotation.GetLocalForward()
        Right := MoverT.Rotation.GetLocalRight()
        Loc := MoverT.Translation + Forward * 50.0 + vector3{Z:= 55.0}
        Sleep(0.9)
        Vfx1:cancelable = SpawnParticleSystem(_VFX_ToUEFN.NS_BlasterProjectile_asset, Loc, ?Rotation := MoverT.Rotation, ?StartDelay := 0.0)
        Sleep(0.5)
        Vfx1.Cancel()

    ActivateGuns9(Devices_O : ?finisher_effects, Data:finisher_data)<suspends>:void=
        Forward := Data.PlayerTransform.Rotation.GetLocalForward()
        NewLoc := Data.PlayerTransform.Translation + Forward * 280.0
        if (Devices := Devices_O?):
            Data.Cinema.Play()
            spawn:
                PlaySound(Data.Sound, Data.MoverTransform, Data.Duration)
            spawn:
                HeroDestructionPath(Data, Devices.Explosive, 1.8, 4, 200.0)
            spawn:
                Guns9Vfx(Data.MoverTransform, Data.TestOffset)
            Sleep(Data.Duration)
            IsValid := IsValidLocation(NewLoc)
            if (IsValid = true):
                if (Devices.Teleporter.TeleportTo[NewLoc, DEFAULT_ROTATION]):
                    Sleep(0.0)
                    Devices.Teleporter.Teleport(Data.Instigator)

    Guns9Vfx(MoverT : transform, TestOffset:vector3)<suspends>:void=
        Forward := MoverT.Rotation.GetLocalForward()
        Right := MoverT.Rotation.GetLocalRight()
        Loc := MoverT.Translation + Forward * 215.0 + vector3{Z:= 135.0}
        Loc2 := MoverT.Translation + Forward * 380.0 + vector3{Z:= 128.0}
        Sleep(0.5)
        Vfx1:cancelable = SpawnParticleSystem(_VFX_ToUEFN.NS_BlasterProjectile_asset, Loc, ?Rotation := MoverT.Rotation, ?StartDelay := 0.0)
        Sleep(0.5)
        Vfx1.Cancel()
        Sleep(0.3)
        Vfx2:cancelable = SpawnParticleSystem(_VFX_ToUEFN.NS_BlasterProjectile_asset, Loc2, ?Rotation := MoverT.Rotation, ?StartDelay := 0.0)
        Sleep(0.2)
        Vfx3:cancelable = SpawnParticleSystem(_VFX_ToUEFN.NS_BlasterProjectile_asset, Loc2, ?Rotation := MoverT.Rotation, ?StartDelay := 0.0)
        Sleep(0.2)
        Vfx4:cancelable = SpawnParticleSystem(_VFX_ToUEFN.NS_BlasterProjectile_asset, Loc2, ?Rotation := MoverT.Rotation, ?StartDelay := 0.0)
        Sleep(0.1)
        Vfx2.Cancel()
        Sleep(0.2)
        Vfx3.Cancel()
        Sleep(0.2)
        Vfx4.Cancel()

    ActivateHalo1(Devices_O : ?finisher_effects, Data:finisher_data)<suspends>:void=
        Forward := Data.PlayerTransform.Rotation.GetLocalForward()
        NewLoc := Data.PlayerTransform.Translation + Forward * 315.0
        if (Devices := Devices_O?):
            Data.Cinema.Play()
            spawn:
                PlaySound(Data.Sound, Data.MoverTransform, Data.Duration)
            spawn:
                HeroDestructionPath(Data, Devices.Explosive, 0.3, 3, 200.0)
            Sleep(Data.Duration)
            IsValid := IsValidLocation(NewLoc)
            if (IsValid = true):
                if (Devices.Teleporter.TeleportTo[NewLoc, DEFAULT_ROTATION]):
                    Sleep(0.0)
                    Devices.Teleporter.Teleport(Data.Instigator)

    ActivateHalo2(Devices_O : ?finisher_effects, Data:finisher_data)<suspends>:void=
        Forward := Data.PlayerTransform.Rotation.GetLocalForward()
        NewLoc := Data.PlayerTransform.Translation + Forward * 540.0
        if (Devices := Devices_O?):
            Data.Cinema.Play()
            spawn:
                PlaySound(Data.Sound, Data.MoverTransform, Data.Duration)
            spawn:
                HeroDestructionPath(Data, Devices.Explosive, 0.7, 3, 200.0)
            Sleep(Data.Duration)
            IsValid := IsValidLocation(NewLoc)
            if (IsValid = true):
                if (Devices.Teleporter.TeleportTo[NewLoc, DEFAULT_ROTATION]):
                    Sleep(0.0)
                    Devices.Teleporter.Teleport(Data.Target)

    ActivateHalo3(Devices_O : ?finisher_effects, Data:finisher_data)<suspends>:void=
        Forward := Data.PlayerTransform.Rotation.GetLocalForward()
        NewLoc := Data.PlayerTransform.Translation + Forward * 215.0
        if (Devices := Devices_O?):
            Data.Cinema.Play()
            spawn:
                PlaySound(Data.Sound, Data.MoverTransform, Data.Duration)
            spawn:
                HeroDestructionPath(Data, Devices.Explosive, 0.6, 3, 200.0)
            Sleep(Data.Duration)
            IsValid := IsValidLocation(NewLoc)
            if (IsValid = true):
                if (Devices.Teleporter.TeleportTo[NewLoc, DEFAULT_ROTATION]):
                    Sleep(0.0)
                    Devices.Teleporter.Teleport(Data.Instigator)

    ActivateHalo4(Devices_O : ?finisher_effects, Data:finisher_data)<suspends>:void=
        Forward := Data.PlayerTransform.Rotation.GetLocalForward()
        NewLoc := Data.PlayerTransform.Translation + Forward * 545.0
        if (Devices := Devices_O?):
            Data.Cinema.Play()
            spawn:
                PlaySound(Data.Sound, Data.MoverTransform, Data.Duration)
            spawn:
                HeroDestructionPath(Data, Devices.Explosive, 0.8, 3, 200.0)
            Sleep(Data.Duration)
            IsValid := IsValidLocation(NewLoc)
            if (IsValid = true):
                if (Devices.Teleporter.TeleportTo[NewLoc, DEFAULT_ROTATION]):
                    Sleep(0.0)
                    Devices.Teleporter.Teleport(Data.Target)

    ActivateCap1(Devices_O : ?finisher_effects, Data:finisher_data)<suspends>:void=
        Forward := Data.PlayerTransform.Rotation.GetLocalForward()
        NewLoc := Data.PlayerTransform.Translation + Forward * 340.0
        if (Devices := Devices_O?):
            Data.Cinema.Play()
            spawn:
                PlaySound(Data.Sound, Data.MoverTransform, Data.Duration)
            spawn:
                HeroDestructionPath(Data, Devices.Explosive, 1.0, 3, 200.0)
            Sleep(Data.Duration)
            IsValid := IsValidLocation(NewLoc)
            if (IsValid = true):
                if (Devices.Teleporter.TeleportTo[NewLoc, DEFAULT_ROTATION]):
                    Sleep(0.0)
                    Devices.Teleporter.Teleport(Data.Instigator)

    ActivateCap2(Devices_O : ?finisher_effects, Data:finisher_data)<suspends>:void=
        Forward := Data.PlayerTransform.Rotation.GetLocalForward()
        NewLoc := Data.PlayerTransform.Translation + Forward * 378.0
        if (Devices := Devices_O?):
            Data.Cinema.Play()
            spawn:
                PlaySound(Data.Sound, Data.MoverTransform, Data.Duration)
            spawn:
                HeroDestructionPath(Data, Devices.Explosive, 1.3, 4, 200.0)
            Sleep(Data.Duration)
            IsValid := IsValidLocation(NewLoc)
            if (IsValid = true):
                if (Devices.Teleporter.TeleportTo[NewLoc, DEFAULT_ROTATION]):
                    Sleep(0.0)
                    Devices.Teleporter.Teleport(Data.Instigator)

    ActivateThor1(Devices_O : ?finisher_effects, Data:finisher_data)<suspends>:void=
        Forward := Data.PlayerTransform.Rotation.GetLocalForward()
        NewLoc := Data.PlayerTransform.Translation + Forward * 830.0
        if (Devices := Devices_O?):
            Data.Cinema.Play()
            spawn:
                PlaySound(Data.Sound, Data.MoverTransform, Data.Duration)
            spawn:
                HeroDestructionPath(Data, Devices.Explosive, 1.6, 4, 200.0)
            spawn:
                Thor1Vfx(Data.MoverTransform, Data.TestOffset)
            Sleep(Data.Duration)
            IsValid := IsValidLocation(NewLoc)
            if (IsValid = true):
                if (Devices.Teleporter.TeleportTo[NewLoc, DEFAULT_ROTATION]):
                    Sleep(0.0)
                    Devices.Teleporter.Teleport(Data.Target)

    Thor1Vfx(MoverT : transform, TestOffset:vector3)<suspends>:void=
        Forward := MoverT.Rotation.GetLocalForward()
        Right := MoverT.Rotation.GetLocalRight()
        Loc := MoverT.Translation + Forward * 156.0 + vector3{Z:= 100.0}
        Loc2 := MoverT.Translation + Forward * 240.0 + vector3{Z:= 100.0}
        Loc3 := MoverT.Translation + Forward * 300.0 + vector3{Z:= 100.0}
        Loc4 := MoverT.Translation + Forward * 350.0 + vector3{Z:= 100.0}
        Sleep(0.4)
        Vfx1:cancelable = SpawnParticleSystem(Electric_VFX.Niagara.Electric.NS_Electric_Shock_asset, Loc, ?Rotation := MoverT.Rotation, ?StartDelay := 0.0)
        spawn:
            CancelAfterTime(Vfx1, 1.0)
        Sleep(0.7)
        Vfx2:cancelable = SpawnParticleSystem(Electric_VFX.Niagara.Electric.NS_Electric_Shock_asset, Loc2, ?Rotation := MoverT.Rotation, ?StartDelay := 0.0)
        spawn:
            CancelAfterTime(Vfx2, 1.0)
        Sleep(0.23)
        Vfx3:cancelable = SpawnParticleSystem(Electric_VFX.Niagara.Electric.NS_Electric_Shock_asset, Loc3, ?Rotation := MoverT.Rotation, ?StartDelay := 0.0)
        spawn:
            CancelAfterTime(Vfx3, 1.0)
        Sleep(0.27)
        Vfx4:cancelable = SpawnParticleSystem(Electric_VFX.Niagara.Electric.NS_Electric_Shock_asset, Loc4, ?Rotation := MoverT.Rotation, ?StartDelay := 0.0)
        Sleep(1.0)
        Vfx4.Cancel()
        # 0.4 , 1.1,  1.33, 1.6

    CancelAfterTime(Vfx:cancelable, Duration:float)<suspends>:void=
        Sleep(Duration)
        Vfx.Cancel()

    ActivateThor2(Devices_O : ?finisher_effects, Data:finisher_data)<suspends>:void=
        Forward := Data.PlayerTransform.Rotation.GetLocalForward()
        NewLoc := Data.PlayerTransform.Translation + Forward * 1050.0
        if (Devices := Devices_O?):
            Data.Cinema.Play()
            spawn:
                PlaySound(Data.Sound, Data.MoverTransform, Data.Duration)
            spawn:
                HeroDestructionPath(Data, Devices.Explosive, 2.9, 6, 200.0)
            spawn:
                Thor2Vfx(Data.MoverTransform, Data.TestOffset)
            spawn:
                AddCamera(Data.Instigator, Data.Target, Devices.ThinkCamera)
            Sleep(Data.Duration)
            RemoveCamera(Data.Instigator, Data.Target, Devices.ThinkCamera)
            IsValid := IsValidLocation(NewLoc)
            if (IsValid = true):
                if (Devices.Teleporter.TeleportTo[NewLoc, DEFAULT_ROTATION]):
                    Sleep(0.0)
                    Devices.Teleporter.Teleport(Data.Target)

    Thor2Vfx(MoverT : transform, TestOffset:vector3)<suspends>:void=
        Forward := MoverT.Rotation.GetLocalForward()
        Right := MoverT.Rotation.GetLocalRight()
        Loc := MoverT.Translation
        Loc2 := MoverT.Translation + Forward * 620.0 + vector3{Z:= 80.0}
        Sleep(0.13)
        Vfx1:cancelable = SpawnParticleSystem(Electric_VFX.Niagara.Electric.NS_Electric_Random2_asset, Loc, ?Rotation := MoverT.Rotation, ?StartDelay := 0.0)
        Sleep(2.5)
        Vfx1.Cancel()
        Sleep(0.4)
        Vfx2:cancelable = SpawnParticleSystem(Electric_VFX.Niagara.Electric.NS_Electric_Shock_asset, Loc2, ?Rotation := MoverT.Rotation, ?StartDelay := 0.0)
        Sleep(1.0)
        Vfx2.Cancel()

    ActivateBoxing1(Devices_O : ?finisher_effects, Data:finisher_data)<suspends>:void=
        Forward := Data.PlayerTransform.Rotation.GetLocalForward()
        NewLoc := Data.PlayerTransform.Translation + Forward * 540.0
        if (Devices := Devices_O?):
            Data.Cinema.Play()
            spawn:
                PlaySound(Data.Sound, Data.MoverTransform, Data.Duration)
            spawn:
                HeroDestructionPath(Data, Devices.Explosive, 1.1, 3, 200.0)
            Sleep(Data.Duration)
            IsValid := IsValidLocation(NewLoc)
            if (IsValid = true):
                if (Devices.Teleporter.TeleportTo[NewLoc, DEFAULT_ROTATION]):
                    Sleep(0.0)
                    Devices.Teleporter.Teleport(Data.Target)

    ActivateBoxing2(Devices_O : ?finisher_effects, Data:finisher_data)<suspends>:void=
        Forward := Data.PlayerTransform.Rotation.GetLocalForward()
        NewLoc := Data.PlayerTransform.Translation + Forward * -200.0
        if (Devices := Devices_O?):
            Data.Cinema.Play()
            spawn:
                PlaySound(Data.Sound, Data.MoverTransform, Data.Duration)
            Sleep(Data.Duration)
            IsValid := IsValidLocation(NewLoc)
            if (IsValid = true):
                if (Devices.Teleporter.TeleportTo[NewLoc, DEFAULT_ROTATION]):
                    Sleep(0.0)
                    Devices.Teleporter.Teleport(Data.Target)

    ActivateBoxing4(Devices_O : ?finisher_effects, Data:finisher_data)<suspends>:void=
        Forward := Data.PlayerTransform.Rotation.GetLocalForward()
        NewLoc := Data.PlayerTransform.Translation + Forward * 670.0
        if (Devices := Devices_O?):
            Data.Cinema.Play()
            spawn:
                PlaySound(Data.Sound, Data.MoverTransform, Data.Duration)
            spawn:
                HeroDestructionPath(Data, Devices.Explosive, 1.7, 4, 200.0)
            Sleep(Data.Duration)
            IsValid := IsValidLocation(NewLoc)
            if (IsValid = true):
                if (Devices.Teleporter.TeleportTo[NewLoc, DEFAULT_ROTATION]):
                    Sleep(0.0)
                    Devices.Teleporter.Teleport(Data.Target)

    ActivateSamurai4(Devices_O : ?finisher_effects, Data:finisher_data)<suspends>:void=
        Forward := Data.PlayerTransform.Rotation.GetLocalForward()
        NewLoc := Data.PlayerTransform.Translation + Forward * 690.0
        if (Devices := Devices_O?):
            Data.Cinema.Play()
            spawn:
                PlaySound(Data.Sound, Data.MoverTransform, Data.Duration)
            spawn:
                HeroDestructionPath(Data, Devices.Explosive, 0.4, 4, 200.0)
            spawn:
                Samurai4Vfx(Data.MoverTransform, Data.TestOffset)
            Sleep(Data.Duration)
            IsValid := IsValidLocation(NewLoc)
            if (IsValid = true):
                if (Devices.Teleporter.TeleportTo[NewLoc, DEFAULT_ROTATION]):
                    Sleep(0.0)
                    Devices.Teleporter.Teleport(Data.Instigator)

    Samurai4Vfx(MoverT : transform, TestOffset:vector3)<suspends>:void=
        Forward := MoverT.Rotation.GetLocalForward()
        Right := MoverT.Rotation.GetLocalRight()
        Loc := MoverT.Translation + Forward * 300.0 + vector3{Z:= 80.0}
        Sleep(0.4)
        Vfx1:cancelable = SpawnParticleSystem(_VFX_ToUEFN.NS_SlashRing_02_asset, Loc, ?Rotation := MoverT.Rotation, ?StartDelay := 0.0)
        Sleep(1.5)
        Vfx1.Cancel()

    ActivateSamurai5(Devices_O : ?finisher_effects, Data:finisher_data)<suspends>:void=
        Forward := Data.PlayerTransform.Rotation.GetLocalForward()
        NewLoc := Data.PlayerTransform.Translation + Forward * 260.0
        if (Devices := Devices_O?):
            Data.Cinema.Play()
            spawn:
                PlaySound(Data.Sound, Data.MoverTransform, Data.Duration)
            spawn:
                HeroDestructionPath(Data, Devices.Explosive, 0.4, 2, 200.0)
            spawn:
                Samurai5Vfx(Data.MoverTransform, Data.TestOffset)
            Sleep(Data.Duration)
            IsValid := IsValidLocation(NewLoc)
            if (IsValid = true):
                if (Devices.Teleporter.TeleportTo[NewLoc, DEFAULT_ROTATION]):
                    Sleep(0.0)
                    Devices.Teleporter.Teleport(Data.Instigator)

    Samurai5Vfx(MoverT : transform, TestOffset:vector3)<suspends>:void=
        Forward := MoverT.Rotation.GetLocalForward()
        Right := MoverT.Rotation.GetLocalRight()
        Loc := MoverT.Translation + Forward * 335.0 + vector3{Z:= 100.0}
        Rot := MoverT.Rotation.ApplyYaw(DegreesToRadians(-90.0))
        Sleep(0.6)
        Vfx1:cancelable = SpawnParticleSystem(_VFX_ToUEFN.NS_Stab_Sl_02_asset, Loc, ?Rotation := Rot, ?StartDelay := 0.0)
        Sleep(0.7)
        Vfx1.Cancel()

    ActivateSamurai6(Devices_O : ?finisher_effects, Data:finisher_data)<suspends>:void=
        Forward := Data.PlayerTransform.Rotation.GetLocalForward()
        NewLoc := Data.PlayerTransform.Translation + Forward * 370.0
        if (Devices := Devices_O?):
            Data.Cinema.Play()
            spawn:
                PlaySound(Data.Sound, Data.MoverTransform, Data.Duration)
            spawn:
                HeroDestructionPath(Data, Devices.Explosive, 0.2, 2, 200.0)
            spawn:
                Samurai6Vfx(Data.MoverTransform, Data.TestOffset)
            Sleep(Data.Duration)
            IsValid := IsValidLocation(NewLoc)
            if (IsValid = true):
                if (Devices.Teleporter.TeleportTo[NewLoc, DEFAULT_ROTATION]):
                    Sleep(0.0)
                    Devices.Teleporter.Teleport(Data.Target)

    Samurai6Vfx(MoverT : transform, TestOffset:vector3)<suspends>:void=
        Forward := MoverT.Rotation.GetLocalForward()
        Right := MoverT.Rotation.GetLocalRight()
        Loc := MoverT.Translation + Forward * 370.0 + vector3{Z:= 90.0}
        Sleep(0.2)
        Vfx1:cancelable = SpawnParticleSystem(_VFX_ToUEFN.NS_SlashCi2Ab_03_asset, Loc, ?Rotation := MoverT.Rotation, ?StartDelay := 0.0)
        Sleep(1.0)
        Vfx1.Cancel()

    ActivateItadori1(Devices_O : ?finisher_effects, Data:finisher_data)<suspends>:void=
        Forward := Data.PlayerTransform.Rotation.GetLocalForward()
        NewLoc := Data.PlayerTransform.Translation + Forward * 490.0
        if (Devices := Devices_O?):
            Data.Cinema.Play()
            spawn:
                PlaySound(Data.Sound, Data.MoverTransform, Data.Duration)
            spawn:
                HeroDestructionPath(Data, Devices.Explosive, 1.1, 3, 200.0)
            Sleep(Data.Duration)
            IsValid := IsValidLocation(NewLoc)
            if (IsValid = true):
                if (Devices.Teleporter.TeleportTo[NewLoc, DEFAULT_ROTATION]):
                    Sleep(0.0)
                    Devices.Teleporter.Teleport(Data.Target)

    ActivateItadori2(Devices_O : ?finisher_effects, Data:finisher_data)<suspends>:void=
        Forward := Data.PlayerTransform.Rotation.GetLocalForward()
        NewLoc := Data.PlayerTransform.Translation + Forward * 490.0
        if (Devices := Devices_O?):
            Data.Cinema.Play()
            spawn:
                PlaySound(Data.Sound, Data.MoverTransform, Data.Duration)
            spawn:
                HeroDestructionPath(Data, Devices.Explosive, 1.7, 3, 200.0)
            Sleep(Data.Duration)
            IsValid := IsValidLocation(NewLoc)
            if (IsValid = true):
                if (Devices.Teleporter.TeleportTo[NewLoc, DEFAULT_ROTATION]):
                    Sleep(0.0)
                    Devices.Teleporter.Teleport(Data.Target)

    ActivateItadori3(Devices_O : ?finisher_effects, Data:finisher_data)<suspends>:void=
        Forward := Data.PlayerTransform.Rotation.GetLocalForward()
        NewLoc := Data.PlayerTransform.Translation + Forward * 490.0
        if (Devices := Devices_O?):
            Data.Cinema.Play()
            spawn:
                PlaySound(Data.Sound, Data.MoverTransform, Data.Duration)
            spawn:
                HeroDestructionPath(Data, Devices.Explosive, 1.5, 3, 200.0)
            Sleep(Data.Duration)
            IsValid := IsValidLocation(NewLoc)
            if (IsValid = true):
                if (Devices.Teleporter.TeleportTo[NewLoc, DEFAULT_ROTATION]):
                    Sleep(0.0)
                    Devices.Teleporter.Teleport(Data.Target)

    ActivateItadori4(Devices_O : ?finisher_effects, Data:finisher_data)<suspends>:void=
        Forward := Data.PlayerTransform.Rotation.GetLocalForward()
        NewLoc := Data.PlayerTransform.Translation + Forward * 490.0
        if (Devices := Devices_O?):
            Data.Cinema.Play()
            spawn:
                PlaySound(Data.Sound, Data.MoverTransform, Data.Duration)
            spawn:
                HeroDestructionPath(Data, Devices.Explosive, 4.0, 3, 200.0)
            spawn:
                ActivateBlackFlashPP(Data.Instigator, Data.Target)
            spawn:
                BlackFlashVfx(Data.MoverTransform, Data.TestOffset)

            Sleep(Data.Duration)
            IsValid := IsValidLocation(NewLoc)
            if (IsValid = true):
                if (Devices.Teleporter.TeleportTo[NewLoc, DEFAULT_ROTATION]):
                    Sleep(0.0)
                    Devices.Teleporter.Teleport(Data.Target)

    ActivateBlackFlashPP(Instigator: agent, Target : agent)<suspends>:void=
        Sleep(3.8)
        PP_BlackFlash3.Enable(Instigator)
        PP_BlackFlash3.Enable(Target)
        PP_BlackFlash3.BlendIn(Instigator)
        PP_BlackFlash3.BlendIn(Target)
  
        Sleep(0.2)
        PP_BlackFlash3.BlendOut(Instigator)
        PP_BlackFlash3.BlendOut(Target)
        PP_BlackFlash3.Disable(Instigator)
        PP_BlackFlash3.Disable(Target)


    BlackFlashVfx(MoverT : transform, TestOffset:vector3)<suspends>:void=
        Forward := MoverT.Rotation.GetLocalForward()
        Right := MoverT.Rotation.GetLocalRight()
        Loc := MoverT.Translation + Forward * 145.0 + vector3{Z:= 90.0}
        Sleep(4.1)
        Vfx1:cancelable = SpawnParticleSystem(_VFX_ToUEFN.NS_BlackFlash_explosion_asset, Loc, ?Rotation := MoverT.Rotation, ?StartDelay := 0.0)
        Sleep(1.1)
        Vfx1.Cancel()

    SelectPlasmaColor(ColorCinema : cinematic_sequence_device, ColorId: int):void=
        ColorCinema.Stop()
        ColorCinema.Play()
        ColorCinema.Pause()
        ColorCinema.SetPlaybackFrame(ColorId)
    
    PlayCamera(CameraCinema: cinematic_sequence_device, Agent: agent)<suspends>:void=
        CameraCinema.Play(Agent)

    PlaySound(Sound: audio_player_device, MoverT : transform, Duration:float)<suspends>:void=
        Forward := MoverT.Rotation.GetLocalForward()
        NewLoc := MoverT.Translation + vector3 {Z := 100.0} + Forward * 78.0
        if (Sound.TeleportTo[NewLoc, DEFAULT_ROTATION]):
            Sleep(0.0)
            Sound.Play()
            # do nothing

    ARENA_MIN_X : float = -12991.0
    ARENA_MIN_Y : float = -13001.0
    ARENA_MAX_X : float = 13509.0
    ARENA_MAX_Y : float = 13524.0
    IsValidLocation(Loc:vector3):logic=
        if (Loc.X < ARENA_MIN_X or Loc.Y < ARENA_MIN_Y):
            return false
        if (Loc.X > ARENA_MAX_X or Loc.Y > ARENA_MAX_Y):
            return false
        # Inside of the cell arena
        if (Loc.Z >= 0.0 and Loc.Z <= 350.0):
            if (Loc.X > 250.0 and Loc.Y < -250.0):
                return false
        # WWE area:
        if (Loc.Z >= 0.0 and Loc.Z <= 203.0):
            if (Loc.X < -3300.0 and Loc.Y > 3655.0):
                return false
        return true