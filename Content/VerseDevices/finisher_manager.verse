
using { /Fortnite.com/Devices }
using { /Fortnite.com/Game }
using { /Verse.org/Simulation }
using { /Fortnite.com/Characters }
using { /UnrealEngine.com/Temporary/Diagnostics }
using { /UnrealEngine.com/Temporary/SpatialMath }
using { /Verse.org/Random }
using { Array }

finisher_devices := class <concrete>():

    HIDE_LOCATION<private>: vector3 = vector3 {Z := -700.0}
    HIDE_LOCATION_SOUND<private>: vector3 = vector3 {Z := -900.0}
    DAMAGE_DISTANCE: float = 1500.0
    DAMAGE:float = 300.0
    PARRY_TARGET_INVULNERABILITY_TIME: float = 1.1
    INVULNERABILITY_TIME:float = 1.1
    var BeingUsed : logic = false
    var _Instigator : ?agent = false
    var _Target : ?agent = false

    DEFAULT_SOUND_ROTATION : rotation = rotation {}

    @editable
    InstigatorReferenceTrigger: trigger_device = trigger_device {}
    @editable
    TargetReferenceTrigger: trigger_device = trigger_device {}
    @editable
    TargetClearTrigger: trigger_device = trigger_device {}
    @editable
    InstigatorClearTrigger: trigger_device = trigger_device {}
    @editable
    ReferenceMover: creative_prop = creative_prop {}
    @editable
    Explosive : explosive_device = explosive_device {}
    @editable
    FinisherCinemas: []cinematic_sequence_device = array {}
    @editable
    AfterEffectDevicesId : int = 0
    @editable
    PlasmaColorCinema : cinematic_sequence_device = cinematic_sequence_device {}
    @editable
    ParryCinemas: []cinematic_sequence_device = array {}

    FINISHER_DURATION: []float = array {
        2.2, # 2
        1.9, # 4
        2.7, # 7
        2.4, # 8
        1.66, # 9 (reduced)
        3.4, # 10
        2.8, # 11
        3.2, # 13
        1.5, # 1
        2.0, # 3
        3.5, # 5
        2.6, # 6
        3.0, # 14
        2.46, # Ninja 1
        1.63, # Ninja 2
        3.70, # Ninja 3
        4.43, # Ninja 4
        2.90, # Ninja 5
        1.50, # Ninja 6
        1.43, # Hero 1
        2.17, # Hero 2
        3.10, # Hero 3
        4.20, # Hero 4
        1.70, # Hero 5
        3.93, # Hero 6
        2.93, # Ki 1 (Kamehameha)
        1.90, # Samurai 1
        1.9,  # Samurai 2
        4.23, # Ki 2 (Rasengan)
        6.00, # Ultimate 1 (Infinite void)
        3.10, # WWE 1
        5.56, # WWE 2
        2.46, # WWE 3
        4.23, # WWE 4
        5.80, # WWE 5
        5.30, # WWE 6
        4.0,  # WWE 7
        2.4,  # Samurai 3 (SW short)
        6.00, # Ultimate 2 (Sukuna domain)
        6.00, # Ultimate 3 (Goku genkidama)
        4.50, # Avatar 1 (water 1)
        1.67, # Avatar 2 (Fire 1)
        1.40, # Avatar 3 (Earth 1)
        2.10, # Avatar 4 (Air 1)
        3.36, # Wolverine 1
        3.46, # Wolverine 2
        2.03, # Wolverine 3
        1.96, # Wolverine 3
        3.43, # Ki 3
        4.40, # Ki 4
        4.9, # Ki 5
        3.43, # Ki 6
        6.0, # Ultimate 4 (Giant 1)
        6.0, # Ultimate 5 (Giant 2)
        1.86, # Spiderman 1
        3.56, # Spiderman 2
        2.60, # Spiderman 3
        3.33, # Spiderman 4
        2.83, # 12
        3.30, # SW 1
        3.80, # SW 2
        2.10, # SW 3
        2.10, # SW 4
        1.93, # Luffy 1
        2.20, # Luffy 2
        2.90, # Luffy 3
        2.53, # Luffy 4
        2.63, # Luffy 5
        2.8, # Hollow Purple
        4.66,  # WWE 8
        3.5,  # WWE 9
        3.8,  # WWE 10
        6.3,  # WWE 11
        3.6,  # WWE 12
        3.9, # Fuga
        2.86,  # Toji 1
        2.46,  # Toji 2
        3.3,  # Toji 3
        3.3, # Hero 7 
        2.43, # Demon Slayer 1
        3.93, # Demon Slayer 2
        2.9, # Demon Slayer 3
        2.3, # Demon Slayer 4
        3.6, # MK 1
        2.5, # MK 2
        2.83, # MK 3
        5.1, # MK 4
        4.9, # Invincible - train
        5.2, # Invincible - think
        1.9, # Zoro 1
        2.1, # Zoro 2
        2.7, # Sanji 1
        3.6, # Sanji 2
        3.0, # Naruto 1
        1.7, # Naruto 2
        4.5, # Naruto 3
        6.1, # Naruto 4
        1.7, # AoT 1
        3.5, # AoT 2
        5.33, # AoT 3 (100)
        2.4, # Hulk 1
        4.0, # Hulk 2
        3.5, # My hero 1
        2.8, # My hero 2
        2.3, # My hero 3
        3.5, # My hero 4
        2.0, # My hero 5
        4.0, # Guns 1
        3.9, # Guns
        3.2, # Guns
        2.4, # Guns
        1.8, # Guns
        1.8, # Guns
        1.8, # Guns
        2.3, # Guns
        3.0, # Guns 9
        2.0, # Halo 1
        1.5, # Halo 2
        1.6, # Halo 3
        2.0, # Halo 4
        2.0, # Cap 1
        2.2, # Thor 1
        2.0, # Cap 2
        5.0, # Thor 2
        2.0, # Boxing 1
        3.5, # Boxing 2
        3.0, # Boxing 3
        2.5, # Boxing 4
        1.1, # Samurai 4
        1.3, # Samurai 5
        1.2, # Samurai 6
        2.2, # Itadori 1
        2.6, # Itadori 2
        2.4, # Itadori 3
        5.0 # Itadori 4
    }

    PARRY_DURATION: []float = array {
        2.0,
        2.0,
        2.0,
        2.0,
        3.5, # 2nd wave
        2.3, # 2nd wave
        2.6 # 2nd wave
    }

    Activate(Instigator:agent, Target:agent, FinisherId:int, Offset:vector3, AfterEffectHandler : finisher_after_effects_handler, InstigatorCP:CustomPlayer, FinisherSound: audio_player_device)<suspends>:void=
        set BeingUsed = true
        set _Instigator = option {Instigator}
        set _Target = option {Target}

        InstigatorReferenceTrigger.Trigger(Instigator)
        TargetReferenceTrigger.Trigger(Target)
        if:
            FC := Instigator.GetFortCharacter[],
            TargetFC := Target.GetFortCharacter[],
            FinisherCinema := FinisherCinemas[FinisherId],
            FinisherDuration := FINISHER_DURATION[FinisherId]
        then:
            T := FC.GetTransform()
            var Loc :vector3 = T.Translation - vector3 {Z := 76.0}

            if (FC.IsCrouching[]):
                set Loc = Loc + vector3 {Z := 25.0}

            set Loc = Loc + Offset

            Rot := T.Rotation
            Sleep(0.0)
            HideFortCharacter(FC, InstigatorCP)
            HideFortCharacter(TargetFC)
            PlasmaColorId := InstigatorCP.GetPlasmaColor() 

            HasEffect := AfterEffectHandler.HasEffect(FinisherId)
            MoverTransform := transform {Translation := Loc, Rotation := Rot}
            
            if (ReferenceMover.TeleportTo[Loc, Rot]):
                Explosive.Reset()
                Explosive.Explode(Instigator)
                if (HasEffect = true):
                    FinisherData := finisher_data {
                        Cinema := FinisherCinema,
                        Sound := FinisherSound,
                        Duration := FinisherDuration,
                        PlayerTransform := T,
                        Instigator := Instigator,
                        Target := Target,
                        MoverTransform := MoverTransform,
                        DevicesId := AfterEffectDevicesId,
                        PlasmaColorCinema := PlasmaColorCinema,
                        PlasmaColorId := PlasmaColorId,
                        TestOffset := Offset}
                    AfterEffectHandler.HandleFinisher(FinisherId, FinisherData)
                else:
                    spawn:
                        PlaySound(FinisherSound, MoverTransform, FinisherDuration)
                    FinisherCinema.Play()
                    Sleep(FinisherDuration)
            
            TargetClearTrigger.Trigger(Target)
            InstigatorClearTrigger.Trigger(Instigator)
            Sleep(0.0)
            if (ReferenceMover.TeleportTo[HIDE_LOCATION, Rot]):
                # Do nothing
            spawn:
                ReferenceMover.MoveTo(HIDE_LOCATION, Rot, 0.1)
            
            Sleep(0.0)
            ShowFortCharacter(FC, InstigatorCP, INVULNERABILITY_TIME)
            DmgArgs : damage_args = damage_args {Instigator := option{FC}, 
                                                            Amount:=1000.0,
                                                            Source:=option{FC}}
            ShowFortCharacter(TargetFC)
            TargetFC.Damage(DmgArgs)

        set BeingUsed = false
        set _Instigator = false
        set _Target = false

    ActivateParry(Instigator:agent, Target:agent, ParryId:int, Offset:vector3, InstigatorCP:CustomPlayer, TargetCP:CustomPlayer, ParrySound: audio_player_device)<suspends>:void=
        set BeingUsed = true
        set _Instigator = option {Instigator}
        set _Target = option {Target}

        InstigatorReferenceTrigger.Trigger(Instigator)
        TargetReferenceTrigger.Trigger(Target)
        if:
            FC := Instigator.GetFortCharacter[],
            TargetFC := Target.GetFortCharacter[],
            ParryCinema := ParryCinemas[ParryId],
            ParryDuration := PARRY_DURATION[ParryId]
        then:
            T := FC.GetTransform()
            var Loc :vector3 = T.Translation - vector3 {Z := 76.0}

            if (FC.IsCrouching[]):
                set Loc = Loc + vector3 {Z := 25.0}

            set Loc = Loc + Offset

            Rot := T.Rotation
            Sleep(0.0)
            HideFortCharacter(FC, InstigatorCP)
            HideFortCharacter(TargetFC, TargetCP)
            MoverTransform := transform {Translation := Loc, Rotation := Rot}
            
            if (ReferenceMover.TeleportTo[Loc, Rot]):
                Explosive.Reset()
                Explosive.Explode(Instigator)
                spawn:
                    PlaySound(ParrySound, MoverTransform, ParryDuration)
                ParryCinema.Play()
                Sleep(ParryDuration)

            TargetClearTrigger.Trigger(Target)
            InstigatorClearTrigger.Trigger(Instigator)
            Sleep(0.0)
            if (ReferenceMover.TeleportTo[HIDE_LOCATION, Rot]):
                # Do nothing
            spawn:
                ReferenceMover.MoveTo(HIDE_LOCATION, Rot, 0.05)

            ShowFortCharacter(FC, InstigatorCP, PARRY_TARGET_INVULNERABILITY_TIME)
            ShowFortCharacter(TargetFC, TargetCP, INVULNERABILITY_TIME)

        set BeingUsed = false
        set _Instigator = false
        set _Target = false


    HideFortCharacter(FC: fort_character):void=
        FC.Hide()
        FC.PutInStasis(stasis_args { AllowTurning:= false, AllowFalling := false, AllowEmotes := false})
        FC.SetVulnerability(false)

    ShowFortCharacter(FC: fort_character):void=
        FC.ReleaseFromStasis()
        FC.Show()
        FC.SetVulnerability(true)

    HideFortCharacter(FC: fort_character, InstigatorCP:CustomPlayer):void=
        FC.Hide()
        FC.PutInStasis(stasis_args { AllowTurning:= false, AllowFalling := false, AllowEmotes := false})
        InstigatorCP.AddInvulnerable()
        FC.SetVulnerability(false)

    ShowFortCharacter(FC: fort_character, InstigatorCP:CustomPlayer, Time_: float):void=
        FC.ReleaseFromStasis()
        if (InstigatorCP.InvisibleOn = false):
            FC.Show()
        if (InstigatorCP.InvulnerableOn = false):
            spawn:
                DelayedInvulnerability(FC, InstigatorCP, Time_)


    DelayedInvulnerability(FC: fort_character, InstigatorCP:CustomPlayer, Time_: float)<suspends>:void=
        Sleep(Time_)
        InstigatorCP.ReduceInvulnerable()
        CheckChangeInvulnerable(FC, InstigatorCP.GetInvulnerableToken())

    CheckChangeInvulnerable(FC : fort_character, Token:int):void=
        if (Token = 0):
            FC.SetVulnerability(true)

    PlaySound(Sound: audio_player_device, MoverT : transform, Duration:float)<suspends>:void=
        Forward := MoverT.Rotation.GetLocalForward()
        NewLoc := MoverT.Translation + vector3 {Z := 100.0} + Forward * 78.0
        if (Sound.TeleportTo[NewLoc, DEFAULT_SOUND_ROTATION]):
            # Print("Sound teleported")
            Sleep(0.0)
            Sound.Play()
            

finisher_manager := class(creative_device):

    NO_OFFSET:vector3 = vector3 {}

    @editable var FinishersList : []finisher_devices = array {  }

    var RunningFinishers : []agent = array {}

    var GoldGranterId<private> : int = 0

    # ----- Accolades -----
    @editable
    DefaultAcc : accolades_device = accolades_device {}
    @editable
    FinisherUsedAcc : []accolades_device = array {}
    var FinisherAccId : int = 0

    @editable
    GoldGranter : item_granter_device = item_granter_device {}

    # ----- Device after effects -----
    @editable
    AfterEffectHandler : finisher_after_effects_handler = finisher_after_effects_handler {}

    @editable
    FinisherSoundManager: pool_sound_manager = pool_sound_manager {}
    @editable
    ParrySoundManager: pool_sound_manager = pool_sound_manager {}
        
    RunFinisher(Instigator:agent, Target:agent, FinisherId:int, InstigatorCP: CustomPlayer)<suspends>:void=
        if (RunningFinishers.ContainsValue(Instigator) = true or RunningFinishers.ContainsValue(Target) = true):
            return
        set RunningFinishers = RunningFinishers + array {Instigator, Target}
        Finisher_O := GetFreeFinisher()
        if (FinisherToUse := Finisher_O?):
            spawn:
                Use(FinisherToUse, Instigator, Target, FinisherId, InstigatorCP)
            Accolade := GetNextAccolade()
            Accolade.Award(Instigator)
            GoldGranter.GrantItemIndex(Instigator, GoldGranterId)

    RunParry(Instigator:agent, Target:agent, ParryId:int, InstigatorCP: CustomPlayer, TargetCP: CustomPlayer)<suspends>:void=
        if (RunningFinishers.ContainsValue(Instigator) = true or RunningFinishers.ContainsValue(Target) = true):
            return
        set RunningFinishers = RunningFinishers + array {Instigator, Target}
        Finisher_O := GetFreeFinisher()
        if (FinisherToUse := Finisher_O?):
            spawn:
                UseParry(FinisherToUse, Instigator, Target, ParryId, InstigatorCP, TargetCP)
            Accolade := GetNextAccolade()
            Accolade.Award(Instigator)
            GoldGranter.GrantItemIndex(Instigator, GoldGranterId)

    Use(FinisherToUse:finisher_devices, Instigator:agent, Target:agent, FinisherId:int, InstigatorCP:CustomPlayer)<suspends>:void=
        Sound := FinisherSoundManager.GetSound(FinisherId)
        FinisherToUse.Activate(Instigator, Target, FinisherId, NO_OFFSET, AfterEffectHandler, InstigatorCP, Sound)
        Sleep(0.3)
        set RunningFinishers = RunningFinishers.RemoveValue(Instigator)
        set RunningFinishers = RunningFinishers.RemoveValue(Target)

    UseParry(FinisherToUse:finisher_devices, Instigator:agent, Target:agent, ParryId:int, InstigatorCP:CustomPlayer, TargetCP:CustomPlayer)<suspends>:void=
        Sound := ParrySoundManager.GetSound(ParryId)
        FinisherToUse.ActivateParry(Instigator, Target, ParryId, NO_OFFSET, InstigatorCP, TargetCP, Sound)
        Sleep(0.3)
        set RunningFinishers = RunningFinishers.RemoveValue(Instigator)
        set RunningFinishers = RunningFinishers.RemoveValue(Target)
    
    GetFreeFinisher():?finisher_devices =
        for (Finisher:FinishersList):
            if (Finisher.BeingUsed = false):
                return option{Finisher}
        return false

    GetNextAccolade():accolades_device=
        Id := FinisherAccId
        set FinisherAccId += 1
        if(Mod[FinisherAccId, FinisherUsedAcc.Length] = 0):
                set FinisherAccId = 0
        if (Next := FinisherUsedAcc[Id]):
            return Next
        return DefaultAcc

    ClearUsed():void=
        for (Finisher:FinishersList):
            set Finisher.BeingUsed = false

    TestFinishers(Agent:agent, FinisherId:int, InstigatorCP:CustomPlayer)<suspends>:void=
        for (Index -> Finisher:FinishersList):
            Offset := vector3 {X := Index * 100.0}
            Sound := FinisherSoundManager.GetSound(FinisherId)
            spawn:
                Finisher.Activate(Agent, Agent, FinisherId, Offset, AfterEffectHandler, InstigatorCP, Sound)

    TestParries(Agent:agent, FinisherId:int, InstigatorCP:CustomPlayer)<suspends>:void=
        for (Index -> Finisher:FinishersList):
            Offset := vector3 {X := Index * 100.0}
            Sound := ParrySoundManager.GetSound(FinisherId)
            spawn:
                Finisher.ActivateParry(Agent, Agent, FinisherId, Offset, InstigatorCP, InstigatorCP, Sound)

    # ----- Setters -----
    SetGoldGranterId(Id:int):void=
        set GoldGranterId = Id
