using { /UnrealEngine.com/Temporary/UI }
using { /UnrealEngine.com/Temporary/SpatialMath}
using { /Verse.org/Simulation }
using { /Fortnite.com/Devices }

finisher_selector_player_widgets := struct:
    Canvas<public> : canvas
    FinisherSelector<public> : finisher_selector_widget

finisher_selector_ui_manager := class ():

    var PlayerWidgetsMap<private> : [player]finisher_selector_player_widgets = map {}

    Init(Player: player, PrevSelected : []int, SaveTrigger : trigger_device):void=
        if (Exists := PlayerWidgetsMap[Player]):
            return
        Canvas:canvas = canvas {}
        FinisherSelector:finisher_selector_widget = finisher_selector_widget{Selected:= PrevSelected, SaveTrigger:=SaveTrigger}
        FinisherSelector.Init()

        CanvasSlot := canvas_slot:
            Widget := FinisherSelector.GetRootWidget()
            Anchors := anchors {Minimum := vector2{X:=0.5, Y:=0.5}, Maximum := vector2{X:=0.5, Y:=0.5}}
            Alignment := vector2{X:=0.5, Y:=0.5}
            SizeToContent := true
        Canvas.AddWidget(CanvasSlot)
        PlayerWidgets := finisher_selector_player_widgets:
            Canvas := Canvas
            FinisherSelector := FinisherSelector
        if (set PlayerWidgetsMap[Player] = PlayerWidgets, PlayerUI := GetPlayerUI[Player]):
            PlayerUI.AddWidget(Canvas, player_ui_slot{InputMode:=ui_input_mode.All})
      
    RemoveFromPlayer(Player:player):void=
        if (Widgets := PlayerWidgetsMap[Player], PlayerUI := GetPlayerUI[Player]):
            PlayerUI.RemoveWidget(Widgets.Canvas)
            set PlayerWidgetsMap = RemoveKeyFromMap(PlayerWidgetsMap, Player)

    OnPlayerLeave(Player:player):void=
        if (Widgets := PlayerWidgetsMap[Player], PlayerUI := GetPlayerUI[Player]):
            PlayerUI.RemoveWidget(Widgets.Canvas)
        set PlayerWidgetsMap = RemoveKeyFromMap(PlayerWidgetsMap, Player)

    GetSelectedFinishers(Player:player):[]int=
        if (Widgets := PlayerWidgetsMap[Player]):
            return Widgets.FinisherSelector.Selected
        return array {}
            
    # From Epic's documentation
    RemoveKeyFromMap(ExampleMap:[player]finisher_selector_player_widgets, ElementToRemove:player):[player]finisher_selector_player_widgets=
        var NewMap:[player]finisher_selector_player_widgets = map{}
        # Concatenate Keys from ExampleMap into NewMap, excluding ElementToRemove
        for (Key -> Value : ExampleMap, Key <> ElementToRemove):
            set NewMap = ConcatenateMaps(NewMap, map{Key => Value})
        return NewMap

    

finisher_selector_device := class (creative_device):

    FinisherSelectorUIManager:finisher_selector_ui_manager = finisher_selector_ui_manager {}

    @editable
    SaveTrigger : trigger_device = trigger_device {}

    @editable
    Button : button_device = button_device {}

    @editable
    CustomPlayerManager : custom_player_manager = custom_player_manager {}

    OnBegin<override>()<suspends>:void=
        Sleep(0.0)

        Button.InteractedWithEvent.Subscribe(ShowPinPad)

        SaveTrigger.TriggeredEvent.Subscribe(OnSaved)

        spawn:
            TrackPlayerLeave()

    TrackPlayerLeave()<suspends>:void=
        loop:
            Agent := CustomPlayerManager.PlayerLeaveEvent.Await()
            if (Player := player[Agent]):
                FinisherSelectorUIManager.OnPlayerLeave(Player)

    ShowPinPad(Agent:agent):void=
        if (Player := player[Agent]):
            PrevSelected := GetPrevSelected(Agent)
            FinisherSelectorUIManager.Init(Player, PrevSelected, SaveTrigger)
    
    OnSaved(Agent_O : ?agent):void=
        if (Agent := Agent_O?, Player := player[Agent]):
            SelectedFinishers := FinisherSelectorUIManager.GetSelectedFinishers(Player)
            FinisherSelectorUIManager.RemoveFromPlayer(Player)
            CustomPlayerManager.StoreSelectedFinishers(Player, SelectedFinishers)

    GetPrevSelected(Agent:agent):[]int=
        CP_O := CustomPlayerManager.GetCustomPlayer(Agent)
        if (CP := CP_O?):
            return CP.GetSelected()
        return array {}