using { /UnrealEngine.com/Temporary/UI }
using { /UnrealEngine.com/Temporary/SpatialMath}
using { /Verse.org/Simulation }
using { /Verse.org/Assets }
using { /Verse.org/Colors }

finisher_ui_device := class ():

    var PlayerWidgetsMap : [player]player_widgets = map {}

    Init(Player: player, Hour_:int, Min_:int):void=
        if (Exists := PlayerWidgetsMap[Player]):
            return
        Canvas:canvas = canvas {}
        ProgressBar:levelup_progress_bar = levelup_progress_bar:
            Width := 360.0
            Height := 40.0
        ProgressBar.Init()
        CanvasSlot := canvas_slot:
            Widget := ProgressBar.GetRootWidget()
            Anchors := anchors {Minimum := vector2{X:= 0.0, Y := 1.0}, Maximum := vector2{X:= 0.0, Y:= 1.0}}
            Offsets := margin {Left := 50.0, Top := -300.0}
        Canvas.AddWidget(CanvasSlot)

        # Add Streak UI
        StreakUI: streak_ui_widget = streak_ui_widget {}
        StreakUI.Init()
        StreakSlot := canvas_slot:
            Widget := StreakUI.GetRootWidget()
            Anchors := anchors {Minimum := vector2{X:= 0.0, Y := 0.5}, Maximum := vector2{X:= 0.0, Y:= 0.5}}
            Offsets := margin {Left := 50.0, Top := -200.0}
            SizeToContent := true
        Canvas.AddWidget(StreakSlot)

        # Add Life server UI
        ServerLife: server_life_widget = server_life_widget {}
        ServerLife.Init(Hour_, Min_)
        ServerLifeSlot := canvas_slot:
            Widget := ServerLife.GetRootWidget()
            Anchors := anchors {Minimum := vector2{X:= 1.0, Y := 1.0}, Maximum := vector2{X:= 1.0, Y:= 1.0}}
            Offsets := margin {Bottom := 10.0, Right := 10.0}
            Alignment := vector2 {X := 1.0, Y := 1.0}
            SizeToContent := true
        Canvas.AddWidget(ServerLifeSlot)

        PlayerWidgets := player_widgets:
            Canvas := Canvas
            ProgressBar := ProgressBar
            StreakUI := StreakUI
            ServerLife := ServerLife
        if (set PlayerWidgetsMap[Player] = PlayerWidgets, PlayerUI := GetPlayerUI[Player]):
            PlayerUI.AddWidget(Canvas)
   
    UpdateLevel(Player:player, Stats : finisher_stats):void=
        if (Widgets := PlayerWidgetsMap[Player]):
            UltimateN := ConvertFinisherId(Stats.Level)
            if (Stats.Level = UltimateN):
                Widgets.ProgressBar.SetColorFG(NamedColors.DarkTurquoise)
                Widgets.ProgressBar.SetLevelText("Finisher Equipped: {Stats.Level}")
            else:
                Widgets.ProgressBar.SetColorFG(NamedColors.DarkRed)
                Widgets.ProgressBar.SetLevelText("Ultimate #{UltimateN}")

    UpdateParryCount(Player:player, Amount:int):void=
        if (Widgets := PlayerWidgetsMap[Player]):
            Widgets.ProgressBar.SetParryText("Parry Count: {Amount}")

    UpdateServerTime(Player:player, Hour_:int, Min_:int):void=
        if (Widgets := PlayerWidgetsMap[Player]):
            Widgets.ServerLife.SetTimeText(Hour_, Min_)

    Hide(Player:player):void=
        if (Widgets := PlayerWidgetsMap[Player], PlayerUI := GetPlayerUI[Player]):

    RemoveFromPlayer(Player:player):void=
        if (Widgets := PlayerWidgetsMap[Player], PlayerUI := GetPlayerUI[Player]):
            PlayerUI.RemoveWidget(Widgets.Canvas)
            set PlayerWidgetsMap = RemoveKeyFromMap(PlayerWidgetsMap, Player)

    OnPlayerLeave(Player:player):void=
        if (Widgets := PlayerWidgetsMap[Player], PlayerUI := GetPlayerUI[Player]):
            PlayerUI.RemoveWidget(Widgets.Canvas)
        set PlayerWidgetsMap = RemoveKeyFromMap(PlayerWidgetsMap, Player)

    RemoveKeyFromMap(ExampleMap:[player]player_widgets, ElementToRemove:player):[player]player_widgets=
        var NewMap:[player]player_widgets = map{}
        for (Key -> Value : ExampleMap, Key <> ElementToRemove):
            set NewMap = ConcatenateMaps(NewMap, map{Key => Value})
        return NewMap

    # Also change the function in custom_player_manager
    # {param} Id is backend finisher number (same as persistence) +1 or public finisher number. 
    ConvertFinisherId(Id : int):int=
        case(Id):
            30 => 1
            39 => 2
            40 => 3
            53 => 4
            54 => 5
            68 => 6
            75 => 7
            76 => 8
            88 => 9
            89 => 10
            96 => 11
            97 => 12
            99 => 13
            100 => 14
            106 => 15
            124 => 16
            _ => Id

    UpdateStreak(Player:player, NewStreak: int):void=
        if (Widgets := PlayerWidgetsMap[Player]):
            Widgets.StreakUI.SetStreak(NewStreak)

player_widgets := struct:
    Canvas<public> : canvas
    ProgressBar<public> : levelup_progress_bar
    StreakUI<public> : streak_ui_widget
    ServerLife<public> : server_life_widget

finisher_stats := struct:
    Level<public> : int
    Count<public> : int = 0