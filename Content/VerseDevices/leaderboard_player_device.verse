using { /Fortnite.com/Devices }
using { /Verse.org/Simulation }
using { /Fortnite.com/Characters }

leaderboard_player_device := class(creative_device):

    @editable
    PlayerReference : player_reference_device = player_reference_device {}

    @editable
    CustomPlayerManager : custom_player_manager = custom_player_manager {}

    # ----- Accolades -----
    @editable
    TimePlayedAcc : accolades_device = accolades_device{}
    @editable
    TimePlayed5Acc : accolades_device = accolades_device{}
    @editable
    TimePlayed10Acc : accolades_device = accolades_device{}
    @editable
    TimePlayed20Acc : accolades_device = accolades_device{}

    OnBegin<override>()<suspends>:void=
        Sleep(0.0)

        spawn:
            UpdateAlways()
        spawn:
            GrantXpForTimePlayed(60.0, TimePlayedAcc)
        spawn:
            GrantXpForTimePlayed(300.0, TimePlayed5Acc)
        spawn:
            GrantXpForTimePlayed(600.0, TimePlayed10Acc)
        spawn:
            GrantXpForTimePlayed(1200.0, TimePlayed20Acc)
        
    UpdateAlways()<suspends>:void=
        loop:
            Players := CustomPlayerManager.GetAllPlayers()
            for (CP : Players):
                if:
                    FC := CP.MyAgentObj.GetFortCharacter[],
                    FC.IsActive[]
                then:
                    PlayerReference.Register(CP.MyAgentObj)
            Sleep(10.0)

    GrantXpForTimePlayed(WaitTime : float, Acc: accolades_device)<suspends>:void=
        loop:
            Sleep(WaitTime)
            Players := CustomPlayerManager.GetAllPlayersAlive()
            for (Player : Players):
                Acc.Award(Player)
            