using { /UnrealEngine.com/Temporary/UI }
using { /Fortnite.com/UI }
using { /Verse.org/Colors }
using { /Verse.org/Assets }
using { /UnrealEngine.com/Temporary/SpatialMath}


levelup_progress_bar := class():
    Width : float
    Height : float

    ColorFG : color_block = color_block {DefaultColor:= NamedColors.DarkTurquoise}
    LevelText : text_block = text_block {DefaultTextColor := NamedColors.White}

    ParryColorFG : color_block = color_block {DefaultColor:= NamedColors.Orange}
    ParryText : text_block = text_block {DefaultTextColor := NamedColors.White}

    var RootWidget : widget = canvas {}

    Init():void=
        ColorFG.SetDesiredSize(vector2{X:= Width, Y := Height})
        ParryColorFG.SetDesiredSize(vector2{X:= Width, Y := Height})

        SetParryText("Parry Count: 0")
        FinisherEquipOverlay := overlay:
            Slots := array:
                overlay_slot:
                    Widget := ColorFG
                    HorizontalAlignment := horizontal_alignment.Left
                overlay_slot:
                    Widget := LevelText
                    HorizontalAlignment := horizontal_alignment.Center
        
        ParryOverlay := overlay:
            Slots := array:
                overlay_slot:
                    Widget := ParryColorFG
                    HorizontalAlignment := horizontal_alignment.Left
                overlay_slot:
                    Widget := ParryText
                    HorizontalAlignment := horizontal_alignment.Center

        set RootWidget = overlay:
            Slots := array:
                overlay_slot:
                    Widget := stack_box:
                        Orientation := orientation.Vertical
                        Slots := array:
                            stack_box_slot:
                                Widget := ParryOverlay
                                HorizontalAlignment := horizontal_alignment.Fill
                                VerticalAlignment := vertical_alignment.Fill
                                Padding := margin{Bottom := 10.0}
                            stack_box_slot:
                                Widget := FinisherEquipOverlay
                                HorizontalAlignment := horizontal_alignment.Fill
                                VerticalAlignment := vertical_alignment.Fill
                            


    GetRootWidget():widget=
        return RootWidget

    SetColorFG(Color : color):void=
        ColorFG.SetColor(Color)
    
    SetLevelText(Str:string):void=
        LevelText.SetText(S2M(Str))

    SetParryText(Str:string):void=
        ParryText.SetText(S2M(Str))

    S2M<localizes>(S:string) : message = "{S}"
