using { /Fortnite.com/Devices }
using { /Verse.org/Simulation }
using { /Fortnite.com/Characters }

nuke_finisher_device := class(creative_device):

    @editable
    SusanoStartTrigger: trigger_device = trigger_device {}
    @editable
    RunNukeInput : input_trigger_device = input_trigger_device {}
    @editable
    PlayerManager : custom_player_manager = custom_player_manager {}
    @editable
    BadgeGrantedPopUp : popup_dialog_device = popup_dialog_device {}
    @editable
    SusanoClass : class_and_team_selector_device = class_and_team_selector_device {}
    @editable
    NormalClass : class_and_team_selector_device = class_and_team_selector_device {}


    AOT_NUKE_DURATION: float = 12.80
    AOT_BADGE_ID : int = 0
    SUSANO_BADGE_ID : int = 1

    var InUse : logic = false

    OnBegin<override>()<suspends>:void=
        Sleep(0.0)

        RunNukeInput.PressedEvent.Subscribe(OnNukeExecuted)

    OnNukeExecuted(Agent: agent):void=
        spawn:
            RunNuke(Agent)

    RunNuke(Agent:agent)<suspends>:void=
        if (InUse = true):
            return
        set InUse = true
        SusanoClass.ChangeClass(Agent)
        SusanoStartTrigger.Trigger(Agent)
        RunNukeInput.Unregister(Agent)
        PlayerManager.RemoveUI(Agent)
        CP_O := PlayerManager.GetCustomPlayer(Agent)
        NumPlayers := PlayerManager.GetNumPlayers()
        Sleep(91.0)
        set InUse = false
        NormalClass.ChangeClass(Agent)
        PlayerManager.AddUI(Agent)
        if (CP := CP_O?):
            if (NumPlayers > 3):
                CP.GetBadge(SUSANO_BADGE_ID)
                BadgeGrantedPopUp.Show(Agent)







    

    