using { /Fortnite.com/Devices }
using { /Verse.org/Simulation }
using { /Verse.org/Simulation/Tags }

only_finishers_arena_device := class(creative_device):

    var ArenaActive : logic = false

    var TIME_ARENA_ENABLED : float = 7 * 60.0

    @editable
    Arena : mutator_zone_device = mutator_zone_device {}

    @editable
    FinishersOnlyButton : button_device = button_device {}

    @editable 
    TimerDevice : timer_device = timer_device {}

    @editable 
    ReleaseTimer : timer_device = timer_device {}

    @editable
    CustomPlayerManager : custom_player_manager = custom_player_manager {}

    @editable
    WeaponRemover : item_remover_device = item_remover_device {}

    @editable
    CoinRemoverDevice : coin_remover_device = coin_remover_device{}

    OnBegin<override>()<suspends>:void=
        Sleep(0.0)

        Arena.Disable()
        FinishersOnlyButton.InteractedWithEvent.Subscribe(OnEnabledZonePressed)
        Arena.AgentEntersEvent.Subscribe(OnAgentEntersZone)

        set TIME_ARENA_ENABLED = TimerDevice.GetMaxDuration()

    OnAgentEntersZone(Agent:agent):void=
        if (PlayerObj := player[Agent]):
            spawn:
                CustomPlayerManager.CheckRegister(Agent)

    OnEnabledZonePressed(Agent:agent):void=
        if (ArenaActive = true):
            return
        CanActivate := CoinRemoverDevice.ActivateWithCost(Agent, 50)
        if (CanActivate = true):
            set ArenaActive = true
            Arena.Enable()
            TimerDevice.Reset()
            TimerDevice.Start()
            ReleaseTimer.Reset()
            ReleaseTimer.Start()
            spawn:
                DisableZone()
            spawn:
                GunsRemover()

    DisableZone()<suspends>:void=
        Sleep(TIME_ARENA_ENABLED)
        Arena.Disable()
        set ArenaActive = false

    GunsRemover()<suspends>:void=
        race:
            Sleep(TIME_ARENA_ENABLED)
            RunRemover()

    RunRemover()<suspends>:void=
        loop:
            if (ArenaActive = false):
                return
            Players := CustomPlayerManager.GetAllPlayers()
            for(CP : Players):
                WeaponRemover.Remove(CP.MyAgentObj)
            Sleep(5.0)
        
            