using { /UnrealEngine.com/Temporary/UI }
using { /UnrealEngine.com/Temporary/SpatialMath}
using { /Verse.org/Simulation }
using { /Fortnite.com/Devices }

pass_codeplayer_widgets := struct:
    Canvas<public> : canvas
    Passcode<public> : passcode_widget

passcode_ui_manager := class ():

    Pins_ : []string

    var PlayerWidgetsMap<private> : [player]pass_codeplayer_widgets = map {}

    Init(Player: player, SuccessTrigger : trigger_device, FailTrigger : trigger_device):void=
        if (Exists := PlayerWidgetsMap[Player]):
            return
        Canvas:canvas = canvas {}
        Passcode:passcode_widget = passcode_widget { Pins := Pins_, SuccessTrigger:= SuccessTrigger, FailTrigger := FailTrigger}
        Passcode.Init()

        CanvasSlot := canvas_slot:
            Widget := Passcode.GetRootWidget()
            Anchors := anchors {Minimum := vector2{X:=0.5, Y:=0.5}, Maximum := vector2{X:=0.5, Y:=0.5}}
            Alignment := vector2{X:=0.5, Y:=0.5}
            SizeToContent := true
        Canvas.AddWidget(CanvasSlot)
        PlayerWidgets := pass_codeplayer_widgets:
            Canvas := Canvas
            Passcode := Passcode
        if (set PlayerWidgetsMap[Player] = PlayerWidgets, PlayerUI := GetPlayerUI[Player]):
            PlayerUI.AddWidget(Canvas, player_ui_slot{InputMode:=ui_input_mode.All})
      
    RemoveFromPlayer(Player:player):void=
        if (Widgets := PlayerWidgetsMap[Player], PlayerUI := GetPlayerUI[Player]):
            PlayerUI.RemoveWidget(Widgets.Canvas)
            set PlayerWidgetsMap = RemoveKeyFromMap(PlayerWidgetsMap, Player)

    OnPlayerLeave(Player:player):void=
        if (Widgets := PlayerWidgetsMap[Player], PlayerUI := GetPlayerUI[Player]):
            PlayerUI.RemoveWidget(Widgets.Canvas)
        set PlayerWidgetsMap = RemoveKeyFromMap(PlayerWidgetsMap, Player)

    GetSuccessPin(Player:player):string=
        if (Widgets := PlayerWidgetsMap[Player]):
            return Widgets.Passcode.LastSuccessPin
        return ""
            
    # From Epic's documentation
    RemoveKeyFromMap(ExampleMap:[player]pass_codeplayer_widgets, ElementToRemove:player):[player]pass_codeplayer_widgets=
        var NewMap:[player]pass_codeplayer_widgets = map{}
        # Concatenate Keys from ExampleMap into NewMap, excluding ElementToRemove
        for (Key -> Value : ExampleMap, Key <> ElementToRemove):
            set NewMap = ConcatenateMaps(NewMap, map{Key => Value})
        return NewMap

    

passcode_device := class (creative_device):

    PasscodeUiManager:passcode_ui_manager = passcode_ui_manager { Pins_ := array {"959418", "6784", "0504", "000", "108", "853", "859", "245", "193", "845", "032", "437", "1019", "3384", "3218", "983507", "4068", "9031325", "578888"} }

    @editable
    SuccessTrigger : trigger_device = trigger_device {}

    @editable
    FailTrigger : trigger_device = trigger_device {}

    @editable
    Button : button_device = button_device {}

    @editable
    ErrorMessage : hud_message_device = hud_message_device {}

    @editable
    ClassSelector : class_and_team_selector_device = class_and_team_selector_device {}

    @editable
    AdminInput : input_trigger_device = input_trigger_device {}

    @editable
    AdminTrigger : trigger_device = trigger_device {}

    @editable
    AdminUsageMetric : analytics_device = analytics_device {}

    @editable
    LargeGoldGranter : item_granter_device = item_granter_device {}

    @editable
    GoldUsageMetric : analytics_device = analytics_device {}

    @editable
    CustomPlayerManager : custom_player_manager = custom_player_manager {}

    @editable
    InfinityUnlockMetric : analytics_device = analytics_device {}

    @editable
    ChangedColorMessage : hud_message_device = hud_message_device {}

    @editable
    SWAcc : accolades_device = accolades_device{}

    @editable
    FlashAcc : accolades_device = accolades_device{}

    @editable
    ExtraLargeGoldGranter : item_granter_device = item_granter_device {}

    OnBegin<override>()<suspends>:void=
        Sleep(0.0)

        Button.InteractedWithEvent.Subscribe(ShowPinPad)

        SuccessTrigger.TriggeredEvent.Subscribe(OnSuccess)
        FailTrigger.TriggeredEvent.Subscribe(OnFail)

        spawn:
            TrackPlayerLeave()

    TrackPlayerLeave()<suspends>:void=
        loop:
            Agent := CustomPlayerManager.PlayerLeaveEvent.Await()
            if (Player := player[Agent]):
                PasscodeUiManager.OnPlayerLeave(Player)

    ShowPinPad(Agent:agent):void=
        if (Player := player[Agent]):
            PasscodeUiManager.Init(Player, SuccessTrigger, FailTrigger)
    
    OnSuccess(Agent_O : ?agent):void=
        if (Agent := Agent_O?, Player := player[Agent]):
            SuccessPin := PasscodeUiManager.GetSuccessPin(Player)
            PasscodeUiManager.RemoveFromPlayer(Player)
            # if (SuccessPin = "0771467"):
            #     AdminUsageMetric.Submit(Agent)
            #     GrantAdmin(Agent)
            if (SuccessPin = "959418"):
                LargeGoldGranter.GrantItem(Agent)
                GoldUsageMetric.Submit(Agent)
            if (SuccessPin = "6784"):
                CustomPlayerManager.UnlockFinisher(Agent, (28, true))  # Unlock ultimate 1 (infinite)
                InfinityUnlockMetric.Submit(Agent)
            if (SuccessPin = "0504"):
                CustomPlayerManager.UnlockFinisher(Agent, (36, true))  # Star Wars finisher
                SWAcc.Award(Agent)
            if (SuccessPin = "3384"):
                CustomPlayerManager.UnlockFinisher(Agent, (74, true))  # Toji 1 finisher (76)
                SWAcc.Award(Agent)
            if (SuccessPin = "3218"):
                CustomPlayerManager.UnlockFinisher(Agent, (76, true))  # Toji 2 finisher (78)
                SWAcc.Award(Agent)
            if (SuccessPin = "1019"):
                CustomPlayerManager.UnlockFinisher(Agent, (77, true))  # Flash finisher (79)
                FlashAcc.Award(Agent)
            # if (SuccessPin = "613002"):
            #     CustomPlayerManager.LoadFakeStats(Agent)
            if (SuccessPin = "983507"):
                CustomPlayerManager.UnlockAllFinishers(Agent)
            if (SuccessPin = "4068"):
                CustomPlayerManager.UnlockFinisher(Agent, (84, true))  # MK finisher (86)
                FlashAcc.Award(Agent)
            if (SuccessPin = "9031325"):
                CustomPlayerManager.GrantNuke(Agent)
            if (SuccessPin = "578888"):
                CustomPlayerManager.ClearData(Agent)
            ColorId := CheckPlasmaCustomColor(SuccessPin)
            if (ColorId > -2):
                SWAcc.Award(Agent)
                CustomPlayerManager.ChooseCustomColor(Player, ColorId)
                ChangedColorMessage.Show(Agent)


    OnFail(Agent_O : ?agent):void=
        if (Agent := Agent_O?, Player := player[Agent]):
            PasscodeUiManager.RemoveFromPlayer(Player)
            ErrorMessage.Show(Agent)
            
    GrantAdmin(Agent:agent):void=
        AdminTrigger.Trigger(Agent)
        AdminInput.Register(Agent)

    CheckPlasmaCustomColor(Pin : string):int=
        case(Pin):
            "000" => -1
            "108" => 0
            "853" => 1
            "859" => 2
            "245" => 3
            "193" => 4
            "845" => 5
            "032" => 6
            "437" => 7
            _ => -2