using { /UnrealEngine.com/Temporary/UI }
using { /Fortnite.com/UI }
using { /Verse.org/Colors }
using { /Verse.org/Colors/NamedColors }
using { /Verse.org/Assets }
using { /UnrealEngine.com/Temporary/SpatialMath}
using { /Fortnite.com/Devices }
using { /Verse.org/Simulation }

passcode_widget := class():
    
    Pins : []string
    SuccessTrigger : trigger_device
    FailTrigger : trigger_device

    var RootWidget<private> : widget = canvas {}
    var ShownText<private> : text_block = text_block {DefaultTextColor := NamedColors.White}
    var CurrentCode : string = ""
    var LastSuccessPin : string = ""

    Init():void=

        Button0 := button_loud{DefaultText:=S2M("0")}
        Button0.OnClick().SubscribeWidgetMessage(Add, 0)
        Button1 := button_loud{DefaultText:=S2M("1")}
        Button1.OnClick().SubscribeWidgetMessage(Add, 1)
        Button2 := button_loud{DefaultText:=S2M("2")}
        Button2.OnClick().SubscribeWidgetMessage(Add, 2)
        Button3 := button_loud{DefaultText:=S2M("3")}
        Button3.OnClick().SubscribeWidgetMessage(Add, 3)
        Button4 := button_loud{DefaultText:=S2M("4")}
        Button4.OnClick().SubscribeWidgetMessage(Add, 4)
        Button5 := button_loud{DefaultText:=S2M("5")}
        Button5.OnClick().SubscribeWidgetMessage(Add, 5)
        Button6 := button_loud{DefaultText:=S2M("6")}
        Button6.OnClick().SubscribeWidgetMessage(Add, 6)
        Button7 := button_loud{DefaultText:=S2M("7")}
        Button7.OnClick().SubscribeWidgetMessage(Add, 7)
        Button8 := button_loud{DefaultText:=S2M("8")}
        Button8.OnClick().SubscribeWidgetMessage(Add, 8)
        Button9 := button_loud{DefaultText:=S2M("9")}
        Button9.OnClick().SubscribeWidgetMessage(Add, 9)
        Delete := button_regular{DefaultText:=S2M("Delete")}
        Delete.OnClick().Subscribe(DeletePressed)
        EnterButton := button_regular{DefaultText:=S2M("Enter")}
        EnterButton.OnClick().Subscribe(EnterPressed)

        set RootWidget = overlay:
            Slots := array:
                overlay_slot:
                    HorizontalAlignment:=horizontal_alignment.Fill
                    VerticalAlignment:=vertical_alignment.Fill
                    Widget:=color_block{DefaultColor:=MakeColorFromSRGB(0.05,0.05,0.05),DefaultDesiredSize:=vector2{X:=655.0,Y:=660.0}}
                overlay_slot:  
                    HorizontalAlignment:=horizontal_alignment.Center
                    VerticalAlignment:=vertical_alignment.Bottom
                    Padding:=margin{Bottom:=20.0}
                    Widget:=overlay:
                        Slots:=array:
                            #PinPad ColorBlock
                            overlay_slot:
                                HorizontalAlignment:=horizontal_alignment.Fill
                                VerticalAlignment:=vertical_alignment.Fill
                                Widget:=color_block{DefaultColor:=MakeColorFromSRGB(0.0,0.01,0.50)}
                            #PinPad
                            overlay_slot:
                                HorizontalAlignment:=horizontal_alignment.Center
                                VerticalAlignment:=vertical_alignment.Center
                                Padding:=margin{Top:=20.0,Left:=20.0,Bottom:=20.0,Right:=20.0}
                                Widget:=stack_box:
                                    Orientation:=orientation.Vertical
                                    Slots:=array:
                                        #1-3 Stack
                                        stack_box_slot:
                                            HorizontalAlignment:=horizontal_alignment.Fill
                                            VerticalAlignment:=vertical_alignment.Fill
                                            Padding:=margin{Top:=25.0,Bottom:=25.0}
                                            Widget:=stack_box:
                                                Orientation:=orientation.Horizontal
                                                Slots:=array:
                                                    stack_box_slot:
                                                        Distribution:=option{1.0}
                                                        HorizontalAlignment:=horizontal_alignment.Center
                                                        VerticalAlignment:=vertical_alignment.Center
                                                        Widget:=Button1
                                                    stack_box_slot:
                                                        Widget:=Button2
                                                        Distribution:=option{1.0}
                                                        HorizontalAlignment:=horizontal_alignment.Center
                                                        VerticalAlignment:=vertical_alignment.Center
                                                        Padding:=margin{Left:=20.0,Right:=20.0}
                                                    stack_box_slot:
                                                        Widget:=Button3
                                                        Distribution:=option{1.0}
                                                        HorizontalAlignment:=horizontal_alignment.Center
                                                        VerticalAlignment:=vertical_alignment.Center
                                        #4-6
                                        stack_box_slot:
                                            HorizontalAlignment:=horizontal_alignment.Fill
                                            VerticalAlignment:=vertical_alignment.Fill
                                            Padding:=margin{Top:=25.0,Bottom:=25.0}
                                            Widget:=stack_box:
                                                Orientation:=orientation.Horizontal
                                                Slots:=array:
                                                    stack_box_slot:
                                                        Distribution:=option{1.0}
                                                        HorizontalAlignment:=horizontal_alignment.Center
                                                        VerticalAlignment:=vertical_alignment.Center
                                                        Widget:=Button4
                                                    stack_box_slot:
                                                        Distribution:=option{1.0}
                                                        Widget:=Button5
                                                        HorizontalAlignment:=horizontal_alignment.Center
                                                        VerticalAlignment:=vertical_alignment.Center
                                                        Padding:=margin{Left:=20.0,Right:=20.0}
                                                    stack_box_slot:
                                                        Distribution:=option{1.0}
                                                        Widget:=Button6
                                                        HorizontalAlignment:=horizontal_alignment.Center
                                                        VerticalAlignment:=vertical_alignment.Center
                                        #7-9
                                        stack_box_slot:
                                            HorizontalAlignment:=horizontal_alignment.Fill
                                            VerticalAlignment:=vertical_alignment.Fill
                                            Padding:=margin{Top:=25.0,Bottom:=25.0}
                                            Widget:=stack_box:
                                                Orientation:=orientation.Horizontal
                                                Slots:=array:
                                                    stack_box_slot:
                                                        Distribution:=option{1.0}
                                                        HorizontalAlignment:=horizontal_alignment.Center
                                                        VerticalAlignment:=vertical_alignment.Center
                                                        Widget:=Button7
                                                    stack_box_slot:
                                                        Widget:=Button8
                                                        Distribution:=option{1.0}
                                                        HorizontalAlignment:=horizontal_alignment.Center
                                                        VerticalAlignment:=vertical_alignment.Center
                                                        Padding:=margin{Left:=20.0,Right:=20.0}
                                                    stack_box_slot:
                                                        Widget:=Button9
                                                        Distribution:=option{1.0}
                                                        HorizontalAlignment:=horizontal_alignment.Center
                                                        VerticalAlignment:=vertical_alignment.Center
                                        #Delete, 0 and Enter
                                        stack_box_slot:
                                            HorizontalAlignment:=horizontal_alignment.Fill
                                            VerticalAlignment:=vertical_alignment.Fill
                                            Padding:=margin{Top:=25.0,Bottom:=25.0}
                                            Widget:=stack_box:
                                                Orientation:=orientation.Horizontal
                                                Slots:=array:
                                                    stack_box_slot:
                                                        Distribution:=option{1.0}
                                                        HorizontalAlignment:=horizontal_alignment.Center
                                                        VerticalAlignment:=vertical_alignment.Center
                                                        Widget:=Delete
                                                    stack_box_slot:
                                                        Widget:=Button0
                                                        Distribution:=option{1.0}
                                                        HorizontalAlignment:=horizontal_alignment.Center
                                                        VerticalAlignment:=vertical_alignment.Center
                                                        Padding:=margin{Left:=20.0,Right:=20.0}
                                                    stack_box_slot:
                                                        Widget:=EnterButton
                                                        Distribution:=option{1.0}
                                                        HorizontalAlignment:=horizontal_alignment.Center
                                                        VerticalAlignment:=vertical_alignment.Center
                #Text Block Stuff
                overlay_slot:
                    HorizontalAlignment:=horizontal_alignment.Center
                    VerticalAlignment:=vertical_alignment.Top
                    Padding:=margin{Top:=20.0}
                    Widget:=overlay:
                        Slots:=array:
                            overlay_slot:
                                Widget:=color_block{DefaultColor:=Black,DefaultDesiredSize:=vector2{X:=400.0,Y:=32.0}}
                                HorizontalAlignment:=horizontal_alignment.Fill
                                VerticalAlignment:=vertical_alignment.Fill
                            overlay_slot:
                                Widget:=ShownText
                                HorizontalAlignment:=horizontal_alignment.Center
                                VerticalAlignment:=vertical_alignment.Center


    Add(WM : widget_message, Value:int):void=
        set CurrentCode += "{Value}"
        ShownText.SetText(S2M("{CurrentCode}"))

    DeletePressed(WM : widget_message):void=
        var UpdatedString : string = ""
        for(Index->S:CurrentCode):
            if(Index = CurrentCode.Length-1):
                # Ignore last
            else:
                set UpdatedString += "{S}"
        set CurrentCode = UpdatedString
        ShownText.SetText(S2M("{CurrentCode}"))

    EnterPressed(WM :widget_message):void=
        for (Pin : Pins):
            if (Pin = CurrentCode):
                set LastSuccessPin = Pin
                SuccessTrigger.Trigger(WM.Player)
                return
        FailTrigger.Trigger(WM.Player)

    GetRootWidget():widget=
        return RootWidget

    S2M<localizes>(S:string) : message = "{S}"
