using { /Fortnite.com/Devices }
using { /Verse.org/Simulation }
using { Array }

player_profile_data := class<final><persistable>:

    Version:int = 5
    FinishersDone_:int = 0
    CurrentFinisherEquip_:int = 0
    UnlockedFinishers_:[]int = array {1, 0, 0, 0, 0, 0, 0, 0} # deprecated
    AllUnlockedFinishers_:[int]int = map {0 => 1}
    MaxStreak_:int = 0 # deprecated
    StreakMax_:int = 0
    WasAdmin_:logic = false
    CustomOrder_:[]int = array {}
    PlasmaColorId_:int = -1
    AllBadges_:[int]int = map {}
    LastUsedDailySpin:float = 0.0

MakePlayerProfileData<constructor>(Src:player_profile_data)<transacts> := player_profile_data:
    Version := Src.Version
    FinishersDone_ := Src.FinishersDone_
    CurrentFinisherEquip_ := Src.CurrentFinisherEquip_
    UnlockedFinishers_ := Src.UnlockedFinishers_
    AllUnlockedFinishers_ := Src.AllUnlockedFinishers_
    WasAdmin_ := Src.WasAdmin_
    StreakMax_ := Src.StreakMax_
    CustomOrder_ := Src.CustomOrder_
    PlasmaColorId_ := Src.PlasmaColorId_
    AllBadges_ := Src.AllBadges_
    LastUsedDailySpin := Src.LastUsedDailySpin

var PlayerProfileDataMap:weak_map(player, player_profile_data) = map{}

CheckSaveDataForPlayer(Agent:agent, ShouldCreate:logic)<decides><transacts>:void=
    if(Player := player[Agent], Player.IsActive[]):
        if(PlayerProfileDataMap[Player]):
            # no-op
        else:
            if (ShouldCreate = true):
                set PlayerProfileDataMap[Player] = player_profile_data{}

AddToFinishersDone(Agent:agent):void=
    spawn:
        AddToFinishersDone_(Agent)

AddToFinishersDone_(Agent:agent)<suspends>:void=
    if:
        CheckSaveDataForPlayer[Agent, false]
        Player := player[Agent]
        Player.IsActive[]
        SourceData := PlayerProfileDataMap[Player]

        set PlayerProfileDataMap[Player] = player_profile_data:
            MakePlayerProfileData<constructor>(SourceData)
            FinishersDone_ := SourceData.FinishersDone_ + 1

SetFinishersDone(Agent:agent, V:int):void=
    spawn:
        SetFinishersDone_(Agent, V)

SetFinishersDone_(Agent:agent, V:int)<suspends>:void=
    if:
        CheckSaveDataForPlayer[Agent, false]
        Player := player[Agent]
        Player.IsActive[]
        SourceData := PlayerProfileDataMap[Player]

        set PlayerProfileDataMap[Player] = player_profile_data:
            MakePlayerProfileData<constructor>(SourceData)
            FinishersDone_ := V

SetLastUsedDailySpin(Agent:agent, Time:float):void=
    spawn:
        SetLastUsedDailySpin_(Agent, Time)

SetLastUsedDailySpin_(Agent:agent, Time:float)<suspends>:void=
    if:
        CheckSaveDataForPlayer[Agent, false]
        Player := player[Agent]
        Player.IsActive[]
        SourceData := PlayerProfileDataMap[Player]

        set PlayerProfileDataMap[Player] = player_profile_data:
            MakePlayerProfileData<constructor>(SourceData)
            LastUsedDailySpin := Time

MaxStreakUpdate(Agent:agent, CurrentStreak:int):void=
    spawn:
        MaxStreakUpdate_(Agent, CurrentStreak)

MaxStreakUpdate_(Agent:agent, CurrentStreak:int)<suspends>:void=
    if:
        CheckSaveDataForPlayer[Agent, false]
        Player := player[Agent]
        Player.IsActive[]
        SourceData := PlayerProfileDataMap[Player]

        if (CurrentStreak > SourceData.StreakMax_):
            set PlayerProfileDataMap[Player] = player_profile_data:
                MakePlayerProfileData<constructor>(SourceData)
                StreakMax_ := CurrentStreak

ChangeFinisherEquip(Agent:agent, Id:int):void=
    spawn:
        ChangeFinisherEquip_(Agent, Id)

ChangeFinisherEquip_(Agent:agent, Id:int)<suspends>:void=
    if:
        CheckSaveDataForPlayer[Agent, false]
        Player := player[Agent]
        Player.IsActive[]
        SourceData := PlayerProfileDataMap[Player]

        set PlayerProfileDataMap[Player] = player_profile_data:
            MakePlayerProfileData<constructor>(SourceData)
            CurrentFinisherEquip_ := Id

SetSelectedPlasmaColor(Agent:agent, ColorId:int):void=
    spawn:
        SetSelectedPlasmaColor_(Agent, ColorId)

SetSelectedPlasmaColor_(Agent:agent, ColorId:int)<suspends>:void=
    if:
        CheckSaveDataForPlayer[Agent, false]
        Player := player[Agent]
        Player.IsActive[]
        SourceData := PlayerProfileDataMap[Player]

        set PlayerProfileDataMap[Player] = player_profile_data:
            MakePlayerProfileData<constructor>(SourceData)
            PlasmaColorId_ := ColorId

StoreUnlockFinisher(Agent:agent, NewUnlocks:[int]int):void=
    spawn:
        StoreUnlockFinisher_(Agent, NewUnlocks)

StoreUnlockFinisher_(Agent:agent, NewUnlocks:[int]int)<suspends>:void=
    if:
        CheckSaveDataForPlayer[Agent, false]
        Player := player[Agent]
        Player.IsActive[]
        SourceData := PlayerProfileDataMap[Player]

        set PlayerProfileDataMap[Player] = player_profile_data:
            MakePlayerProfileData<constructor>(SourceData)
            AllUnlockedFinishers_ := NewUnlocks

StoreBadges(Agent:agent, NewBadges:[int]int):void=
    spawn:
        StoreBadges_(Agent, NewBadges)

StoreBadges_(Agent:agent, NewBadges:[int]int)<suspends>:void=
    if:
        CheckSaveDataForPlayer[Agent, false]
        Player := player[Agent]
        Player.IsActive[]
        SourceData := PlayerProfileDataMap[Player]

        set PlayerProfileDataMap[Player] = player_profile_data:
            MakePlayerProfileData<constructor>(SourceData)
            AllBadges_ := NewBadges

StoreWasAdmin(Agent:agent):void=
    spawn:
        StoreWasAdmin_(Agent)

StoreWasAdmin_(Agent:agent)<suspends>:void=
    if:
        CheckSaveDataForPlayer[Agent, false]
        Player := player[Agent]
        Player.IsActive[]
        SourceData := PlayerProfileDataMap[Player]

        set PlayerProfileDataMap[Player] = player_profile_data:
            MakePlayerProfileData<constructor>(SourceData)
            WasAdmin_ := true

UpdateCustomOrder(Agent:agent, NewCustomOrder:[]int):void=
    spawn:
        UpdateCustomOrder_(Agent, NewCustomOrder)

UpdateCustomOrder_(Agent:agent, NewCustomOrder:[]int)<suspends>:void=
    if:
        CheckSaveDataForPlayer[Agent, false]
        Player := player[Agent]
        Player.IsActive[]
        SourceData := PlayerProfileDataMap[Player]

        set PlayerProfileDataMap[Player] = player_profile_data:
            MakePlayerProfileData<constructor>(SourceData)
            CustomOrder_ := NewCustomOrder

LoadPersistenceData(Agent:agent):player_profile_data=
    if:
        CheckSaveDataForPlayer[Agent, true]
        Player := player[Agent]
        Player.IsActive[]
        SourceData := PlayerProfileDataMap[Player]
    then:
        PrintDebugData(SourceData) # TODO: remove
        AllUnlocks := GetAllUnlocksV0(SourceData)
        if:
            set PlayerProfileDataMap[Player] = player_profile_data:
                MakePlayerProfileData<constructor>(SourceData)
                Version := 4
                AllUnlockedFinishers_ := AllUnlocks
        then:
            if:
                NewData := PlayerProfileDataMap[Player]
            then:
                return NewData
    return player_profile_data{}

GetPersistenceData(Agent:agent):player_profile_data=
    if:
        CheckSaveDataForPlayer[Agent, false]
        Player := player[Agent]
        Player.IsActive[]
        SourceData := PlayerProfileDataMap[Player]
    then:
        return SourceData
    return player_profile_data {}

GetAllUnlocksV0(SavedData : player_profile_data):[int]int=
    var AllUnlocks : [int]int = map {}
    for (Index -> Unlocked : SavedData.UnlockedFinishers_):
        if (Unlocked > 0):
            if:
                set AllUnlocks[Index] = Unlocked
    for (Index -> Unlocked : SavedData.AllUnlockedFinishers_):
        if (Unlocked > 0):
            if:
                set AllUnlocks[Index] = Unlocked
    return AllUnlocks

PrintDebugData(SourceData:player_profile_data):void=
    Print("LOADED {SourceData.Version}")
    var prev :string = ""
    for (i -> x : SourceData.UnlockedFinishers_):
            set prev += "{x}, "
    Print("Prev: {prev}")
    var new :string = ""
    for (i -> x : SourceData.AllUnlockedFinishers_):
            set new += "{i} = {x}, "
    Print("New: {new}")

    var badges :string = ""
    for (i -> x : SourceData.AllBadges_):
            set badges += "{i} = {x}, "
    Print("Badges: {badges}")

    converted := GetAllUnlocksV0(SourceData)
    var convert :string = ""
    for (i -> x : converted):
        set convert += "{i} = {x}, "
    Print("Converted : {convert}")

    Print("Streak : {SourceData.StreakMax_}")
    if (SourceData.WasAdmin_ = true):
        Print("Admin")

    var CustomOrderMsg :string = ""
    for (id: SourceData.CustomOrder_):
        set CustomOrderMsg += "{id}, "
    Print ("Custom Order : {CustomOrderMsg}")
    Print("Plasma Color : {SourceData.PlasmaColorId_}")
