using { /Fortnite.com/Devices }
using { /Verse.org/Simulation }

pool_sound_manager := class(creative_device):

    @editable
    FinisherSounds1: []audio_player_device = array {}
    @editable
    FinisherSounds2: []audio_player_device = array {}
    @editable
    FinisherSounds3: []audio_player_device = array {}

    var NextId : [int]int = map {}

    OnBegin<override>()<suspends>:void=
        Sleep(0.0)
        L := FinisherSounds1.Length
        for (i := 0..L-1):
            if (set NextId[i] = 0):
                # Do nothing

    GetSound(FinisherId:int):audio_player_device=
        if (Cur := NextId[FinisherId]):
            var N :int= Cur + 1
            if (N = 3):
                set N = 0
            if (set NextId[FinisherId] = N):
                # Do nothing
            return FindSound(FinisherId, Cur)
        return FindSound(FinisherId, 0)

    FindSound(FinisherId:int, PoolId:int):audio_player_device=
        if (PoolId = 0):
            if (Audio := FinisherSounds1[FinisherId]):
                return Audio
        if (PoolId = 1):
            if (Audio := FinisherSounds2[FinisherId]):
                return Audio
        if (PoolId = 2):
            if (Audio := FinisherSounds3[FinisherId]):
                return Audio
        return audio_player_device{}