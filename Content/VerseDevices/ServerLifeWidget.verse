using { /UnrealEngine.com/Temporary/UI }
using { /Fortnite.com/UI }
using { /Verse.org/Colors }
using { /Verse.org/Assets }
using { /UnrealEngine.com/Temporary/SpatialMath}


server_life_widget := class():
    MainText : text_block = text_block {DefaultTextColor := NamedColors.White}

    var RootWidget : widget = canvas {}

    Init(Hour_:int, Min_:int):void=
        MainText.SetText(S2M("Server Time: 00:00"))
        set RootWidget = overlay:
            Slots := array:
                # overlay_slot:
                #     Widget := color_block:
                #         DefaultColor := NamedColors.Black
                #         DefaultOpacity := 0.3
                #     HorizontalAlignment := horizontal_alignment.Fill
                #     VerticalAlignment := vertical_alignment.Fill
                overlay_slot:
                    Widget := MainText
                    HorizontalAlignment := horizontal_alignment.Left
                    Padding := margin {Right := 10.0, Bottom := 5.0}


    GetRootWidget():widget=
        return RootWidget

    SetTimeText(Hour_:int, Min_:int):void=
        Str := "Server Time: {MakeTimerText(Hour_, Min_)}"
        MainText.SetText(S2M( + Str))

    MakeTimerText(Hour_:int, Min_:int):string=
        H := if (Hour_ < 10) then "0{Hour_}" else "{Hour_}"
        M := if (Min_< 10) then "0{Min_}" else "{Min_}"
        return "{H}:{M}"

    S2M<localizes>(S:string) : message = "{S}"


