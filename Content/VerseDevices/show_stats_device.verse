using { /Fortnite.com/Devices }
using { /Verse.org/Simulation }

show_stats_device := class(creative_device):

    @editable
    Text : billboard_device = billboard_device {}
    @editable
    UnlockedText : billboard_device = billboard_device {}
    @editable
    PlayerText : billboard_device = billboard_device {}

    @editable
    RefreshButton: button_device = button_device {}

    @editable
    ShowPlayer : player_reference_device = player_reference_device {}

    @editable
    BadgeProps : []creative_prop = array {}

    OnBegin<override>()<suspends>:void=
        Sleep(0.0)

        RefreshButton.InteractedWithEvent.Subscribe(OnShowStatsPressed)

    OnShowStatsPressed(Agent:agent):void=
        ShowPlayer.Register(Agent)
        ShowStats(Agent)

    ShowStats(Agent:agent):void=
        Data := GetPersistenceData(Agent)
        All := GetUnlockedFinishers(Data.AllUnlockedFinishers_)
        Count := GetUnlockedNumber(Data.AllUnlockedFinishers_)
        StatsMessage := ToMessage(Data.FinishersDone_, Data.StreakMax_, Count, NUM_FINISHERS, All)

        Text.SetText(StatsMessage)
        UnlockedText.SetText(S2M(All))
        PlayerText.SetText(PlayerToMessage(Agent))
        ShowBadges(Data.AllBadges_)

    ShowBadges(Badges: [int]int):void=
        for (Prop : BadgeProps):
            Prop.Hide()
        for (i -> v : Badges):
            if (Prop := BadgeProps[i]):
                if (v > 0):
                    Prop.Show()
                    


    ToMessage<localizes>(Done: int, Streak:int, A: int, B:int, All:string):message=
        "Finishers Done: {Done}\nMax Streak: {Streak}\nUnlocked: {A}/{B}"

    S2M<localizes>(S:string) : message = "{S}"

    PlayerToMessage<localizes>(Agent:agent):message=
        "{Agent}"
    
    GetUnlockedNumber(UnlockedFinishers: [int]int):int=
        var Count : int = 0
        for (i -> x : UnlockedFinishers):
            if (x > 0):
                set Count += 1
        return Count

    GetUnlockedFinishers(UnlockedFinishers: [int]int):string=
        var UnlockedStr :string = ""

        Tens := Int[NUM_FINISHERS*1.0/10.0] or 0
        Print("Tens = {Tens}")
        for (i := 0..Tens-1):
            var AllUnlock : int = 0
            for (j := 0..9):
                Id := i * 10 + j
                if (Unlocked := UnlockedFinishers[Id]):
                    if (Unlocked > 0):
                        set AllUnlock += 1
            if (AllUnlock = 10):
                if (UnlockedStr.Length > 0):
                    set UnlockedStr += ", "
                set UnlockedStr += "{i*10+1}-{i*10+10}"
            else:
                for (j := 0..9):
                    Id := i * 10 + j
                    if (Unlocked := UnlockedFinishers[Id]):
                        if (Unlocked > 0):
                            if (UnlockedStr.Length > 0):
                                set UnlockedStr += ", "
                            set UnlockedStr += "{Id+1}"
        Remaining := Mod[NUM_FINISHERS, 10] or 0
        for (j := 0..Remaining-1):
            Id := Tens * 10 + j
            if (Unlocked := UnlockedFinishers[Id]):
                if (Unlocked > 0):
                    if (UnlockedStr.Length > 0):
                        set UnlockedStr += ", "
                    set UnlockedStr += "{Id+1}"

        return UnlockedStr