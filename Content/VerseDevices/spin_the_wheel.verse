using { /Fortnite.com/Devices }
using { /UnrealEngine.com/Temporary/SpatialMath }
using { /Verse.org/Random }
using { /Verse.org/Simulation }

wheel_reward := class():
    RewardTitle : string = ""
    GiveReward(Agent:agent, CPM:custom_player_manager):void=
        Print("Here you go")
    GiveGold(Agent:agent, Device:item_granter_device):void=
        Print("Giving Gold")

Cash300G := class(wheel_reward):
    RewardTitle<override> : string = "Cash300"
    GiveGold<override>(Agent:agent, Device:item_granter_device):void=
        Print(RewardTitle)
        Device.GrantItemIndex(Agent, 3)

Cash500G := class(wheel_reward):
    RewardTitle<override> : string = "Cash500"
    GiveGold<override>(Agent:agent, Device:item_granter_device):void=
        Print(RewardTitle)
        Device.GrantItemIndex(Agent, 2)

Cash1000G := class(wheel_reward):
    RewardTitle<override> : string = "Cash1000"
    GiveGold<override>(Agent:agent, Device:item_granter_device):void=
        Print(RewardTitle)
        Device.GrantItemIndex(Agent, 1)

Cash2000G := class(wheel_reward):
    RewardTitle<override> : string = "Cash2000"
    GiveGold<override>(Agent:agent, Device:item_granter_device):void=
        Print(RewardTitle)
        Device.GrantItemIndex(Agent, 0)

UnlockFinisherOne := class(wheel_reward):
    RewardTitle<override> : string = "Unlocked Finisher 132"
    GiveReward<override>(Agent:agent, CPM:custom_player_manager):void=
        Print(RewardTitle)
        CPM.UnlockFinisher(Agent, (130, true))

UnlockFinisherTwo := class(wheel_reward):
    RewardTitle<override> : string = "Unlocked Finisher 133"
    GiveReward<override>(Agent:agent, CPM:custom_player_manager):void=
        Print(RewardTitle)
        CPM.UnlockFinisher(Agent, (131, true))

UnlockFinisherThree := class(wheel_reward):
    RewardTitle<override> : string = "Unlocked Finisher 134"
    GiveReward<override>(Agent:agent, CPM:custom_player_manager):void=
        Print(RewardTitle)
        CPM.UnlockFinisher(Agent, (132, true))

UnlockFinisherFour := class(wheel_reward):
    RewardTitle<override> : string = "Unlocked Finisher 135"
    GiveReward<override>(Agent:agent, CPM:custom_player_manager):void=
        Print(RewardTitle)
        CPM.UnlockFinisher(Agent, (133, true))

spin_the_wheel := class(creative_device):
    var Rewards : []wheel_reward = array{}
    @editable SpinButton : button_device = button_device{}
    @editable WheelProp : creative_prop = creative_prop{}
    @editable ErrorHudMessage : hud_message_device = hud_message_device {}
    @editable CustomPlayerManager : custom_player_manager = custom_player_manager {}

    @editable GoldItemGranter : item_granter_device = item_granter_device {}

    var InitialTransform : transform = transform{} 
    var SpinInProgress : logic = false

    SpinAmount : int = 12       # How many times the wheel spins before it lands
    SpinDuration : float = 3.0  # How long the spin duration lasts before landing

    OnBegin<override>()<suspends>:void=
        Sleep(0.0)

        set InitialTransform = WheelProp.GetTransform()
        SetRewards()
        SpinButton.InteractedWithEvent.Subscribe(AttemptSpin)

    SetRewards()<transacts>:void=        
        set Rewards = array:
            Cash2000G{}
            UnlockFinisherThree{}
            Cash1000G{}
            UnlockFinisherOne{}
            Cash500G{}
            UnlockFinisherTwo{}
            Cash300G{}
            UnlockFinisherFour{}

    CanSpin(Agent:agent):logic=
        CustomPlayerManager.CheckAndRegister(Agent)
        LastSpinEpoch: float = CustomPlayerManager.GetDailySpin(Agent)

        if (LastSpinEpoch = 0.0):
            return true

        LastSpinDate := date_data{}.ParseEpoch(LastSpinEpoch)
        CurrentDate := date_data{}.ParseEpoch(GetSecondsSinceEpoch())

        DaysPassedOption := CurrentDate.DaysBetween(LastSpinDate)

        if (DaysPassed := DaysPassedOption?):
            if (DaysPassed >= 1):
                return true

        return false

    AttemptSpin(Agent:agent):void=
        if(SpinInProgress = false):
            ReturnStatement := CanSpin(Agent)
            if (ReturnStatement = true):
                CustomPlayerManager.SetDailySpin(Agent, GetSecondsSinceEpoch())
                spawn { Spin(Agent) }
            else:
                LastSpinEpoch := CustomPlayerManager.GetDailySpin(Agent)
                CurrentEpoch := GetSecondsSinceEpoch()

                LastSpinDate := date_data{}.ParseEpoch(LastSpinEpoch)
                NextSpinDate := LastSpinDate.AddHours(24)

                NextSpinString := ToPrettyTime(NextSpinDate)

                Msg := S2M("You don't have any spins available.\nNext spin {NextSpinString}")
                ErrorHudMessage.Show(Agent, Msg)
        else:
            ErrorHudMessage.Show(Agent, S2M("Wheel is already spinning"))

    Spin(Agent:agent)<suspends>:void=
        set SpinInProgress = true
        PreSelectedRewardIndex := GetRandomInt(0, Rewards.Length-1)
        Print("{PreSelectedRewardIndex}")
        
        NewRotation1 := InitialTransform.Rotation.ApplyPitch(DegreesToRadians((360.0/(Rewards.Length*1.0))*PreSelectedRewardIndex))
        WheelProp.MoveTo(transform{Rotation:=NewRotation1, Translation:=WheelProp.GetTransform().Translation}, 0.1)

        for(Index := 0..SpinAmount):
            NewRotation := WheelProp.GetTransform().Rotation.ApplyPitch(DegreesToRadians(179.99))
            WheelProp.MoveTo(transform{Rotation:=NewRotation, Translation:=WheelProp.GetTransform().Translation}, SpinDuration/(SpinAmount*1.0))

        # Not ideal, but for the system that was made before this is the best way to do it due to editables not being avaliable above the class
        if(RewardClass := Rewards[PreSelectedRewardIndex]):
            if(RewardClass.RewardTitle = "Cash300" or RewardClass.RewardTitle = "Cash500" or RewardClass.RewardTitle = "Cash1000" or RewardClass.RewardTitle = "Cash2000"):
                RewardClass.GiveGold(Agent, GoldItemGranter)
            else:
                RewardClass.GiveReward(Agent, CustomPlayerManager)

        set SpinInProgress = false

    S2M<localizes>(S:string) : message = "{S}"

