using { /UnrealEngine.com/Temporary/UI }
using { /Fortnite.com/UI }
using { /Verse.org/Colors }
using { /Verse.org/Assets }
using { /UnrealEngine.com/Temporary/SpatialMath}


streak_ui_widget := class():
    TextImage : texture_block = texture_block{DefaultImage:=Original.UI.Streak.Img_empty, DefaultDesiredSize := vector2 {X := 128.0, Y := 128.0}}
    Digit1Image : texture_block = texture_block{DefaultImage:=Original.UI.Streak.Img_empty, DefaultDesiredSize := vector2 {X := 128.0, Y := 128.0}}
    Digit2Image : texture_block = texture_block{DefaultImage:=Original.UI.Streak.Img_empty, DefaultDesiredSize := vector2 {X := 128.0, Y := 128.0}}

    DEFAULT_SIZE : vector2 = vector2 {X := 128.0, Y := 128.0}
    DOUBLE_DIGIT_SIZE : vector2 = vector2 {X := 110.0, Y := 110.0}
    EMPTY_IMAGE : texture = Original.UI.Streak.Img_empty
    TEXT_IMAGE : texture = Original.UI.Streak.Streak_text_yellow
    NUMBERS_IMG : []texture = array {
        Original.UI.Streak.Img0,
        Original.UI.Streak.Img1,
        Original.UI.Streak.Img2,
        Original.UI.Streak.Img3,
        Original.UI.Streak.Img4,
        Original.UI.Streak.Img5,
        Original.UI.Streak.Img6,
        Original.UI.Streak.Img7,
        Original.UI.Streak.Img8,
        Original.UI.Streak.Img9
    }

    var CurrentStreak<private> : int = 0
    var RootWidget : widget = canvas {}

    Init():void=
        set RootWidget = overlay:
            
            Slots := array:
                overlay_slot:
                    Padding := margin{Top := 0.0, Left := 12.5, Right := 0.0, Bottom := 0.0}
                    Widget := TextImage
                overlay_slot:
                    Widget := Digit1Image
                overlay_slot:
                    Padding := margin{Top := 0.0, Left := 50.0, Right := 0.0, Bottom := 0.0}
                    Widget := Digit2Image

    GetRootWidget():widget=
        return RootWidget

    SetStreak(NewStreak: int):void=
        if (NewStreak = CurrentStreak):
            return
        set CurrentStreak = NewStreak
        if (NewStreak = 0):
            TextImage.SetImage(EMPTY_IMAGE)
            Digit1Image.SetImage(EMPTY_IMAGE)
            Digit1Image.SetDesiredSize(DEFAULT_SIZE)
            Digit2Image.SetImage(EMPTY_IMAGE)
            Digit2Image.SetDesiredSize(DEFAULT_SIZE)
            return
        TextImage.SetImage(TEXT_IMAGE)
        if (NewStreak < 10):
            Digit1Image.SetImage(GetDigitImage(NewStreak))
            Digit1Image.SetDesiredSize(DEFAULT_SIZE)
            Digit2Image.SetImage(EMPTY_IMAGE)
            Digit2Image.SetDesiredSize(DEFAULT_SIZE)
        else:
            if (Digit2 := Mod[NewStreak, 10], Digit1 := Floor[(NewStreak*1.0)/10.0]):
                # Print("{Digit1} . {Digit2}")
                Digit1Image.SetImage(GetDigitImage(Digit1))
                Digit1Image.SetDesiredSize(DOUBLE_DIGIT_SIZE)
                Digit2Image.SetImage(GetDigitImage(Digit2))
                Digit2Image.SetDesiredSize(DOUBLE_DIGIT_SIZE)


    GetDigitImage(N:int):texture=
        if (Img := NUMBERS_IMG[N]):
            return Img
        return Original.UI.Streak.Img0
