using { /UnrealEngine.com/Assets }
using { /Fortnite.com/Devices }
using { /Verse.org/Simulation }
using { /Fortnite.com/Characters }
using { /UnrealEngine.com/Temporary/SpatialMath }

test_vfx_device := class(creative_device):

   
    @editable
    Button : button_device = button_device{}

    OnBegin<override>()<suspends>:void=
        Sleep(0.0)

        Button.InteractedWithEvent.Subscribe(OnTestVFX)

    OnTestVFX(Agent:agent):void=
        spawn:
            TestVFX(Agent)

    TestVFX(Agent: agent)<suspends>:void=
        if (FC := Agent.GetFortCharacter[]):
            Loc := FC.GetTransform().Translation + vector3 {X := 100.0}
            MyVFX:cancelable = SpawnParticleSystem(_VFX_ToUEFN.NS_SpiderWeb_CaptureExpand_asset, Loc , ?Rotation := rotation {}, ?StartDelay := 0.0)
            Sleep(3.0)
            MyVFX.Cancel()
    