using { /Fortnite.com/Devices }
using { /Verse.org/Simulation }
using { /Verse.org/Simulation/Tags }


timed_event_device := class(creative_device):

    var InEvent : logic = false

    StartEvent : event() = event(){}
    EndEvent : event() = event(){}

    SECONDS_IN_DAY : int = 60*60*24
    SECONDS_IN_HOUR : float = 60*60.0

    EVENT_TIMES : []tuple(int, int) = array {
        (90, 150),  # 1:30 - 2:30 UTC
        (270, 330),  # 4:30 - 5:30 UTC
        (450, 510),  # 7:30 - 8:30 UTC
        (630, 690),  # 10:30 - 11:30 UTC
        (810, 870),  # 13:30 - 14:30 UTC
        (990, 1050),  # 16:30 - 17:30 UTC (+2h MUC Time)
        (1170, 1230),  # 19:30 - 20:30 UTC (+2h MUC Time)
        (1350, 1410)  # 22:30 - 23:30 UTC (+2h MUC Time)
    }

    @editable
    TimeProp : creative_prop = creative_prop {}  

    @editable
    FinisherManager:finisher_manager = finisher_manager {}

    @editable 
    TimerDevice : timer_device = timer_device {}

    @editable
    LabelText : billboard_device = billboard_device {}
    @editable
    LabelTimer : billboard_device = billboard_device {}

    @editable
    StartEventSound : audio_player_device = audio_player_device {}

    @editable
    ContinuousSound : audio_player_device = audio_player_device {}
    @editable
    EndSound : audio_player_device = audio_player_device {}

    @editable
    StartVFX : vfx_spawner_device = vfx_spawner_device {}

    @editable
    EndVFX : vfx_spawner_device = vfx_spawner_device {}

    OnBegin<override>()<suspends>:void=
        Sleep(0.0)

        StartVFX.Disable()
        EndVFX.Disable()
        LabelText.HideText()
        LabelTimer.HideText()

        # spawn:
        #     GetCurrentTime()

    GetCurrentTime()<suspends>:void=
        loop:
            T := TimeProp.GetTransform().Translation
            # Print("{T.X}:{T.Y}:{T.Z}")
            TimeAsMinutes := ConvertToMinutes(T.X, T.Y)
            IsRunningEnd := IsAnyEventRunning(TimeAsMinutes)

            if (IsRunningEnd <> 0):
                # Print("During event")
                ActiveSeconds := (IsRunningEnd - TimeAsMinutes) * 60.0 - T.Z
                RunEvent(ActiveSeconds)
            else:
                # Print("Not event")
                TimeAsSeconds := ConvertToSeconds(T.X, T.Y, T.Z)
                Closest := GetSecondsToClosestEvent(TimeAsSeconds)
                UpdateTimer(Closest)
                CancelEvent()
                
            Sleep(1.0)

    IsAnyEventRunning(TimeAsMinutes:int):int=
        for (E : EVENT_TIMES):
            Start := E(0)
            End := E(1)
            if (TimeAsMinutes >= Start and TimeAsMinutes <= End):
                return End
        return 0

    GetSecondsToClosestEvent(TimeAsSeconds:int):int=
        for (E : EVENT_TIMES):
            Start := E(0)
            if (TimeAsSeconds < Start * 60):
                return Start * 60 - TimeAsSeconds
        if (NextDay := EVENT_TIMES[0], NextDayStart := NextDay(0)):
            return NextDayStart*60 + SECONDS_IN_DAY - TimeAsSeconds
        return SECONDS_IN_DAY


    RunEvent(ActiveSeconds:float):void=
        if (InEvent = true):
            return
        # Print("Start Event")
        set InEvent = true
        FinisherManager.SetGoldGranterId(1)
        ShowTimer(ActiveSeconds)
        LabelText.SetText(S2M("2x Gold event in progress"))
        LabelTimer.HideText()
        StartEventSound.Play()
        ContinuousSound.Play()
        StartVFX.Restart()
        # StartEvent.Signal()

    CancelEvent():void=
        if (InEvent = false):
            return
        # Print("Cancel Event")
        set InEvent = false
        FinisherManager.SetGoldGranterId(0)
        LabelText.SetText(S2M("2x Gold event in:"))
        LabelTimer.ShowText()
        ContinuousSound.Stop()
        StartVFX.Disable()
        EndVFX.Restart()
        EndSound.Play()
        # EndEvent.Signal()

    ShowTimer(ActiveSeconds:float):void=
        TimerDevice.SetActiveDuration(ActiveSeconds)
        TimerDevice.Reset()
        TimerDevice.Start()

    ConvertToMinutes(Hour:float, Minute:float):int=
        if (X := Floor[Hour] * 60 + Floor[Minute]):
            return X
        return 0
    
    ConvertToSeconds(Hour:float, Minute:float, Seconds:float):int=
        if (X := Floor[Hour] * 60 * 60 + Floor[Minute] * 60 + Floor[Seconds]):
            return X
        return 0

    S2M<localizes>(S:string) : message = "{S}"

    UpdateTimer(TimeAsSeconds:int):void = 
        if:
            Hours:= Floor[(TimeAsSeconds*1.0)/SECONDS_IN_HOUR],
            Minutes:= Floor[(TimeAsSeconds*1.0 - Hours * SECONDS_IN_HOUR)/60.0],
            Seconds:= Mod[TimeAsSeconds, 60]
        then:
            LabelTimer.SetText(ToTimerMessage(Hours, Minutes, Seconds))

    ToTimerMessage<localizes>(Hours:int, Minutes: int, Seconds:int):message=
        "{Hours}h : {Minutes}m : {Seconds}s"