using { /Fortnite.com/Devices }
using { /Verse.org/Simulation }
using { /Verse.org/Simulation/Tags }

WeaponConditionButtonTag := class(tag) {}

weapon_holding_device := class(creative_device):

    var ConditionalButtons : []conditional_button_device = array {}

    OnBegin<override>()<suspends>:void=
        Sleep(0.0)

        AllCB := FindCreativeObjectsWithTag(WeaponConditionButtonTag {})

        for (CB_Item : AllCB, WeaponCB := conditional_button_device[CB_Item]):
            set ConditionalButtons = ConditionalButtons + array {WeaponCB}

    IsHoldingWeapon(Agent:agent):logic=
        for (WeaponCB : ConditionalButtons):
            if (WeaponCB.IsHoldingItem[Agent]):
                return true
        return false