using { /Verse.org/Simulation }
using { /UnrealEngine.com/Temporary/UI }
using { /Fortnite.com/Devices }


(Listenable : listenable(agent)).SubscribeAgent(OutputFunc : tuple(agent, t)->void, ExtraData : t where t:type) : cancelable =
    Wrapper := wrapper_agent(t){ExtraData := ExtraData, OutputFunc := OutputFunc}
    Listenable.Subscribe(Wrapper.InputFunc)

wrapper_agent(t : type) := class():
    ExtraData : t;
    OutputFunc : tuple(agent, t) -> void
    InputFunc(Agent : agent):void = OutputFunc(Agent, ExtraData)

(Listenable : listenable(tuple())).SubscribeEmpty(OutputFunc : t -> void, ExtraData : t where t:type) : cancelable =
    Wrapper := wrapper_empty(t) {ExtraData := ExtraData, OutputFunc := OutputFunc}
    Listenable.Subscribe(Wrapper.InputFunc)

wrapper_empty(t : type) := class():
    ExtraData : t;
    OutputFunc : t -> void
    InputFunc():void = OutputFunc(ExtraData)
    
(Listenable : listenable(widget_message)).SubscribeWidgetMessage(OutputFunc : tuple(widget_message, t)->void, ExtraData : t where t:type) : cancelable =
    Wrapper := wrapper_widget_message(t){ExtraData := ExtraData, OutputFunc := OutputFunc}
    Listenable.Subscribe(Wrapper.InputFunc)

wrapper_widget_message(t : type) := class():
    ExtraData : t;
    OutputFunc : tuple(widget_message, t) -> void
    InputFunc(WM : widget_message):void = OutputFunc(WM, ExtraData)

(Listenable : listenable(device_ai_interaction_result)).SubscribeNPC(OutputFunc : tuple(device_ai_interaction_result, t)->void, ExtraData : t where t:type) : cancelable =
    Wrapper := wrapper_ai_interaction_result(t){ExtraData := ExtraData, OutputFunc := OutputFunc}
    Listenable.Subscribe(Wrapper.InputFunc)

wrapper_ai_interaction_result(t : type) := class():
    ExtraData : t;
    OutputFunc : tuple(device_ai_interaction_result, t) -> void
    InputFunc(WM : device_ai_interaction_result):void = OutputFunc(WM, ExtraData)