using { /Fortnite.com/Devices }
using { /Verse.org/Simulation }
using { /Verse.org/Simulation/Tags }

CollectableForXPTag := class(tag) {}

xp_farm_device := class(creative_device):

    var Collectables : []collectible_object_device = array {}

    @editable
    CollectableAcc : accolades_device = accolades_device{}

    @editable
    GoldGranter : item_granter_device = item_granter_device {}

    OnBegin<override>()<suspends>:void=
        Sleep(0.0)

        AllCollectables := FindCreativeObjectsWithTag(CollectableForXPTag {})

        for (CollectableItem : AllCollectables, CollectableDevice := collectible_object_device[CollectableItem]):
            set Collectables = Collectables + array {CollectableDevice}
            CollectableDevice.CollectedEvent.Subscribe(OnCollected)

        spawn:
            ResetXP()

    OnCollected(Agent:agent):void=
        CollectableAcc.Award(Agent)
        GoldGranter.GrantItem(Agent)

    ResetXP()<suspends>:void=
        loop:
            Sleep(60.0 * 7)
            for (Collectable : Collectables):
                Collectable.RespawnForAll()