using { /Verse.org }
using { /Verse.org/Native }
using { /Verse.org/SceneGraph }
using { /Verse.org/Simulation }
using { /Fortnite.com/Devices }
using { /Fortnite.com/Characters }
using { /Verse.org/SpatialMath }
using { /UnrealEngine.com/Temporary/SpatialMath }
using { /UnrealEngine.com/Temporary/Diagnostics }
using { /Verse.org/SceneGraph/KeyframedMovement }
using { /Verse.org/Colors/NamedColors }

raycast_component<public> := class<final_super>(component):

    DebugDraw : debug_draw = debug_draw{Channel := my_debug_draw}

    OnSimulate<override>()<suspends>:void =
        {}

    # Returns hit location and normal
    SearchHit(Player : player):tuple(?VerseVector3, ?VerseVector3)=
        var ReturnValue : tuple(?VerseVector3, ?VerseVector3) = (false, false)
        
        if(Char := Player.GetFortCharacter[]) :
            # Print("Collision search")
            # Calculate player's view direction
            LookRot :[]float= Char.GetViewRotation().GetYawPitchRollDegrees()
            if :
                YawRadian := (/UnrealEngine.com/Temporary/SpatialMath:)DegreesToRadians(LookRot[0])
                PitchRadian := (/UnrealEngine.com/Temporary/SpatialMath:)DegreesToRadians(LookRot[1])
            then :
                LookDirection := UnrealVector3{X := Cos(YawRadian)*Cos(PitchRadian), Y := Sin(YawRadian)*Cos(PitchRadian), Z := Sin(PitchRadian)}
                # Set collision check distance (cm; use very large value for infinite range)
                Range := 200000.0
                DisplacementRange := LookDirection*Range
                # Start position for collision check
                ViewLoc := FromVector3(Char.GetViewLocation())
                # Collision (HitResult occurs when this collides)
                Collision : collision_sphere = collision_sphere :
                    Radius := 1.0
                # Check if sweep collision hits anything
                # HitResult occurs if collision detected
                HitResult := Entity.FindSweepHits(FromVector3(DisplacementRange), VerseTransform{Translation := ViewLoc}, Collision)
                # DebugDraw.DrawLine(Char.GetViewLocation(), Char.GetViewLocation()+DisplacementRange, ?Duration:= 2.0)

                var IsHit : logic = false
                for(Result : HitResult) :
                    ParentEntity := Result.TargetComponent.Entity
                    # Treat as wall hit if no raycast_component
                    # Optionally, differentiate walls by using a specific component
                    if(IsHit = false, not ParentEntity.GetComponent[raycast_component]) :
                        # Perform hit detection only once (when IsHit is false)
                        set IsHit = true
                        HitPos := Result.ContactPosition
                        HitNormal := Result.ContactNormal
                        # Set return value
                        set ReturnValue = (option{HitPos}, option{HitNormal})
                        # Debug
                        # DebugDraw.DrawSphere(FromVector3(HitPos), ?Radius := 10.0, ?Duration:= 2.0)
                        # DebugDraw.DrawLine(FromVector3(HitPos), FromVector3(HitPos + HitNormal*100.0),?Color := Red,  ?Duration:= 2.0)
                        # Print("Hit : {HitPos}")

        return ReturnValue