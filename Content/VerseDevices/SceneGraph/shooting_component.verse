using { /Verse.org }
using { /Verse.org/Native }
using { /Verse.org/SceneGraph }
using { /Verse.org/Simulation }
using { /Fortnite.com/Devices }
using { /Fortnite.com/Characters }
using { /Verse.org/SpatialMath }
using { /UnrealEngine.com/Temporary/SpatialMath }
using { /UnrealEngine.com/Temporary/Diagnostics }
using { /Verse.org/Colors }
using { /Verse.org/Random }


shooting_component<public> := class<final_super>(component):

    DebugDraw : debug_draw = debug_draw{Channel := my_debug_draw}

    var IsAttacking : logic = false
    var NextExplosiveId : int = 0
    HIDE_LOCATION : UnrealVector3 = UnrealVector3 {X := 1000.0, Z := -400.0}

    TIME_MIN: float = 0.7
    TIME_MAX: float = 1.2

    @editable
    AttackInput : input_trigger_device = input_trigger_device {}

    @editable 
    AttackCinema : cinematic_sequence_device = cinematic_sequence_device {}

    @editable
    Explosives : []explosive_device = array {}

    OnSimulate<override>()<suspends>:void =
        AttackInput.PressedEvent.Subscribe(OnAttackPressed)

    OnAttackPressed(Agent:agent):void=
        # Print("Attacking")
        if (IsAttacking = true):
            return
        set IsAttacking = true
        spawn:
            RemoveAttacking()
        spawn:
            Attack(Agent)
        

    RemoveAttacking()<suspends>:void=
        Sleep(2.2)
        set IsAttacking = false

    Attack(Agent:agent)<suspends>:void=
        if :
            Player : player = player[Agent]
            RayComp := Entity.GetComponent[raycast_component]
        then :
            ResultStart := RayComp.SearchHit(Player)
            if:
                InitHitPos :VerseVector3= ResultStart(0)?
            then:
                AttackCinema.Play()
                Sleep(1.85)
                Result := RayComp.SearchHit(Player)
                if :
                    HitPos :VerseVector3= Result(0)?
                then :
                    # Print("Hit Found")
                    StartPos := GetAttackStartLoc(Agent)
                    # DebugDraw.DrawSphere(StartPos, ?DrawDurationPolicy := debug_draw_duration_policy.FiniteDuration, ?Duration := 15.0, ?Radius := 70.0, ?Color := NamedColors.Black)
                    
                    spawn:
                        CreateProjectile(Agent, HitPos, 3.0)
                else:
                    # Print("Init Hit Found")
                    StartPos := GetAttackStartLoc(Agent)
                    # DebugDraw.DrawSphere(StartPos, ?DrawDurationPolicy := debug_draw_duration_policy.FiniteDuration, ?Duration := 15.0, ?Radius := 70.0, ?Color := NamedColors.Black)
                    spawn:
                        CreateProjectile(Agent, InitHitPos, 3.0)
            else :
                # Print("No hit found")


    GetAttackStartLoc(Agent:agent):(/UnrealEngine.com/Temporary/SpatialMath:)vector3=
        if:
            FC := Agent.GetFortCharacter[],
            FC.IsActive[]
        then:
            # return FC.GetViewLocation()
            T := FC.GetTransform()
            # Rot := T.Rotat
            Forward := T.Rotation.GetLocalForward()
            Right := T.Rotation.GetLocalRight()
            Up := T.Rotation.GetLocalUp()
            DeltaLoc := Forward * -188.0 + Right * 131.0 + Up * 389.0
            return T.Translation + DeltaLoc
        return UnrealVector3 {}


    CreateProjectile(Agent:agent, HitPos:(/Verse.org/SpatialMath:)vector3, Duration:float)<suspends>:void=
        Id := GetNextProjectile()
        if:
            FC := Agent.GetFortCharacter[],
            FC.IsActive[]
        then:
            T := FC.GetTransform()
            Rotation := T.Rotation
            StartPos := GetAttackStartLoc(Agent)
            Data := basic_projectile_data:
                Agent:= Agent
                StartLoc := StartPos
                Rotation := Rotation
                EndLoc := FromVector3(HitPos)
                Duration := GetRandomFloat(TIME_MIN, TIME_MAX)
            if(Explosive := Explosives[Id]):
                MoveProjectile(Data, Explosive)
        Sleep(1.0)
        if(Explosive := Explosives[Id]):
            if (Explosive.TeleportTo[HIDE_LOCATION, UnrealRotation{}]):    

    GetNextProjectile():int=
        X := NextExplosiveId
        set NextExplosiveId += 1
        if(NextExplosiveId >= Explosives.Length):
            set NextExplosiveId = 0
        return X

    MoveProjectile(Data:basic_projectile_data, Explosive: explosive_device)<suspends>:void=
        if (Explosive.TeleportTo[Data.StartLoc, Data.Rotation]):
        # Print("Restarting VFX Basic projectile")
        race:
            ContinuousExplosion(Explosive, Data.Agent)
            Explosive.MoveTo(Data.EndLoc, Data.Rotation, Data.Duration)
        # Print("End shoot")

    ContinuousExplosion(Explosive: explosive_device, Agent:agent)<suspends>:void=
        loop:
            Explosive.Reset()
            Sleep(0.2)
            Explosive.Explode(Agent)


basic_projectile_data := struct:
    Agent<public>:agent
    StartLoc<public>:UnrealVector3
    Rotation<public>:UnrealRotation
    EndLoc<public>:UnrealVector3
    Duration<public>:float