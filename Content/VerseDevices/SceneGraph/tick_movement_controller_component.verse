using { /Verse.org }
using { /Verse.org/SceneGraph }
using { /Verse.org/SceneGraph/KeyframedMovement }
using { /Verse.org/Simulation }
using { /Verse.org/SpatialMath }
using { /Fortnite.com/Devices }

using { /Fortnite.com/Characters }
# using { /UnrealEngine.com/Temporary/Diagnostics }
using { /UnrealEngine.com/Temporary/SpatialMath }

tick_movement_controller_component<public> := class<final_super>(component):   
    
    DELTA_X : float = 30.0

    var AgentSeated : ?agent = false
    var TickCancelable : ?cancelable = false

    var InitialT : VerseTransform = VerseTransform {}

    var AccumulatedDeltaLoc: UnrealVector3 = UnrealVector3{}
    var CurrentTargetRotation: VerseRotation = VerseRotation{}

    @editable
    StartButton: button_device = button_device {}

    @editable
    StartTrigger: trigger_device = trigger_device {}

    @editable
    ForwardInput: input_trigger_device = input_trigger_device {}
    @editable
    BackwardInput: input_trigger_device = input_trigger_device {}
    @editable
    LeftInput: input_trigger_device = input_trigger_device {}
    @editable
    RightInput: input_trigger_device = input_trigger_device {}
    @editable
    UpInput: input_trigger_device = input_trigger_device {}
    @editable
    DownInput: input_trigger_device = input_trigger_device {}
    @editable
    FireInput: input_trigger_device = input_trigger_device {}

    @editable
    Camera: gameplay_camera_orbit_device = gameplay_camera_orbit_device {}

    @editable
    Chair : chair_device = chair_device {}

    @editable
    MainExplosive : explosive_device = explosive_device {}

    @editable
    SusanoProp : creative_prop = creative_prop {}

    OnBeginSimulation<override>():void=
        set TickCancelable = option {TickEvents.PostPhysics.Subscribe(OnPostPhysics)}        

    ContinuousExplosion()<suspends>:void=
        loop:
            if (Agent := AgentSeated?):
                MainExplosive.Reset()
                Sleep(0.2)
                MainExplosive.Explode(Agent)
            else:
                Sleep(1.0)

    OnEndSimulation<override>():void=
        if (C := TickCancelable?):
            C.Cancel()
            # Print("Cancelled post")

    OnSimulate<override>()<suspends>:void =
        # Print("Simulate")
        StartButton.InteractedWithEvent.Subscribe(OnStart)
        StartTrigger.TriggeredEvent.Subscribe(OnTriggerStart)
        if:
            TransformComp := Entity.GetComponent[transform_component]
        then:
            set InitialT = TransformComp.GlobalTransform
        SusanoProp.Hide()
        spawn:
            ContinuousExplosion()
        spawn:
            ApplyMovement()
   

    ApplyMovement()<suspends>:void=
        loop:
            if (Agent := AgentSeated?):
                if:
                    TransformComp := Entity.GetComponent[transform_component]
                then:
                    CurT := TransformComp.GlobalTransform
                    Forward := CurT.Rotation.GetForwardAxis()
                    Left := CurT.Rotation.GetLeftAxis()
                    Up := CurT.Rotation.GetUpAxis()
                    DeltaLoc := Forward * AccumulatedDeltaLoc.X + Left * - 1.0 * AccumulatedDeltaLoc.Y + Up * AccumulatedDeltaLoc.Z
                    NewTranslation := CurT.Translation + DeltaLoc
                    set AccumulatedDeltaLoc = UnrealVector3 {}

                    MidRot := Slerp(CurT.Rotation, CurrentTargetRotation, 0.5) #TODO: ajust speed, prev = 0.5

                    NewT := (/Verse.org/SpatialMath:)transform:
                        Translation := NewTranslation
                        Rotation := MidRot
                        Scale := CurT.Scale

                    IsValid := IsValidLocation(FromVector3(NewT.Translation))
                    if (IsValid = true):
                        set TransformComp.GlobalTransform = NewT
                        Chair.MoveTo(FromVector3(NewT.Translation),FromRotation(NewT.Rotation), 0.03)
                    else:
                        Sleep(0.1)
            else:
                Sleep(0.5)


    OnPostPhysics(DeltaTime:float):void=
        if (Agent := AgentSeated?):
            XDelta := GetDelta(Agent, ForwardInput, 1.0) + GetDelta(Agent, BackwardInput, -1.0)
            YDelta := GetDelta(Agent, RightInput, 1.0) + GetDelta(Agent, LeftInput, -1.0)
            ZDelta := GetDelta(Agent, UpInput, 1.0) + GetDelta(Agent, DownInput, -1.0)
            MoveVec := UnrealVector3{X := XDelta, Y := YDelta, Z := ZDelta}

            if (MoveVec.IsAlmostZero[0.001]):
                if:
                    FC := Agent.GetFortCharacter[],
                    FC.IsActive[]
                then:
                    PlayerRot := FC.GetViewRotation()
                    set CurrentTargetRotation = FromRotation(PlayerRot)
            else:
                if:
                    FC := Agent.GetFortCharacter[],
                    FC.IsActive[]
                then:
                    PlayerRot := FC.GetViewRotation()
                    # Normalize direction, apply speed
                    if (DirUnit := MoveVec.MakeUnitVector[]):
                        var Speed : float = 3000.0 # Original: 1000.0
                        set AccumulatedDeltaLoc = AccumulatedDeltaLoc + (DirUnit * Speed * DeltaTime)
                        set CurrentTargetRotation = FromRotation(PlayerRot)

            # if:
            #     FC := Agent.GetFortCharacter[],
            #     FC.IsActive[]
            # then:
            #     PlayerRot := FC.GetViewRotation()
            #     LookRot :[]float= PlayerRot.GetYawPitchRollDegrees()
            #     if :
            #         YawRadian := (/UnrealEngine.com/Temporary/SpatialMath:)DegreesToRadians(LookRot[0])
            #         PitchRadian := (/UnrealEngine.com/Temporary/SpatialMath:)DegreesToRadians(LookRot[1])
            #     then :
            #         LookDirection := UnrealVector3{X := Cos(YawRadian)*Cos(PitchRadian), Y := Sin(YawRadian)*Cos(PitchRadian), Z := Sin(PitchRadian)}
                    
            #         # MoveVec := LookDirection * MoveVec2

            #         # Print("Rot = {YawRadian}, {PitchRadian}")

            #         if (MoveVec.IsAlmostZero[0.001]):

            #             if:
            #                 TransformComp := Entity.GetComponent[transform_component]
            #             then:
            #                 CurT := TransformComp.GlobalTransform
            #                 D := Distance(CurT.Rotation, FromRotation(PlayerRot))
            #                 if (D.IsAlmostZero[0.001]):
            #                     return
            #                 else:
            #                     SetNewRotation(FromRotation(PlayerRot))
            #         else:
            #             # Normalize direction, apply speed
            #             if (DirUnit := MoveVec.MakeUnitVector[]):
            #                 var Speed : float = 3000.0 # Original: 1000.0
            #                 DeltaLoc := DirUnit * Speed * DeltaTime
            #                 TranslateBy(DeltaLoc, FromRotation(PlayerRot))


    SetNewRotation(NewRotation: VerseRotation):void=
        if:
            TransformComp := Entity.GetComponent[transform_component]
        then:

            CurT := TransformComp.GlobalTransform
            MidRot := Slerp(CurT.Rotation, NewRotation, 0.5)
            NewT := (/Verse.org/SpatialMath:)transform:
                Translation := CurT.Translation
                Rotation := MidRot
                Scale := CurT.Scale
            set TransformComp.GlobalTransform = NewT
            # if(Chair.TeleportTo[FromVector3(NewT.Translation),FromRotation(NewT.Rotation)]):
                # Rotate only
            spawn:
                Chair.MoveTo(FromVector3(NewT.Translation),FromRotation(NewT.Rotation), 0.03)

    GetDelta(Agent: agent, Input: input_trigger_device, Amount:float):float=
        if (Input.IsHeld[Agent]):
            # Print("IsHeld = {Amount}")
            return Amount
        return 0.0
                    

    TranslateBy(DeltaLoc:UnrealVector3, NewRotation: VerseRotation):void=
        if:
            TransformComp := Entity.GetComponent[transform_component]
        then:
            CurT := TransformComp.GlobalTransform
            Forward := CurT.Rotation.GetForwardAxis()
            Left := CurT.Rotation.GetLeftAxis()
            Up := CurT.Rotation.GetUpAxis()
            NewDeltaLoc := Forward * DeltaLoc.X + Left * - 1.0 * DeltaLoc.Y + Up * DeltaLoc.Z

            MidRot := Slerp(CurT.Rotation, NewRotation, 0.5)

            NewT := (/Verse.org/SpatialMath:)transform:
                Translation := CurT.Translation + NewDeltaLoc  #+ FromVector3(DeltaLoc)
                Rotation := MidRot
                Scale := CurT.Scale

            IsValid := IsValidLocation(FromVector3(NewT.Translation))
            if (IsValid = true):
                set TransformComp.GlobalTransform = NewT
                # if(Chair.TeleportTo[FromVector3(NewT.Translation),FromRotation(NewT.Rotation)]):
                    # do nothing
                spawn:
                    Chair.MoveTo(FromVector3(NewT.Translation),FromRotation(NewT.Rotation), 0.03)

    OnTriggerStart(Agent_O:?agent):void=
        if (Agent := Agent_O?):
            OnStart(Agent)

    OnStart(Agent:agent):void=
        spawn:
            OnStart_(Agent)

    OnStart_(Agent:agent)<suspends>:void=
        if (A := AgentSeated?):
            return
        SetToInitialPosition()
        SusanoProp.Show()
        # Print("Start clicked")
        set AgentSeated = option{Agent}
        if (FC := Agent.GetFortCharacter[]):
            FC.PutInStasis(stasis_args { AllowTurning:= false, AllowFalling := false, AllowEmotes := false})
            FC.SetVulnerability(false)
            FC.Hide()
        Chair.Seat(Agent)
        Camera.AddTo(Agent)
        
        ForwardInput.Register(Agent)
        BackwardInput.Register(Agent)
        LeftInput.Register(Agent)
        RightInput.Register(Agent)
        UpInput.Register(Agent)
        DownInput.Register(Agent)
        FireInput.Register(Agent)

        spawn:
            RemoveAfterTime(Agent)

    RemoveAfterTime(Agent:agent)<suspends>:void=
        Sleep(90.0)
        set AgentSeated = false
        Chair.Eject()
        Camera.RemoveFrom(Agent)
        if (FC := Agent.GetFortCharacter[]):
            FC.ReleaseFromStasis()
            FC.Show()
            FC.SetVulnerability(true)

        ForwardInput.Unregister(Agent)
        BackwardInput.Unregister(Agent)
        LeftInput.Unregister(Agent)
        RightInput.Unregister(Agent)
        UpInput.Unregister(Agent)
        DownInput.Unregister(Agent)
        FireInput.Unregister(Agent)
        SusanoProp.Hide()
        SetToInitialPosition()

    SetToInitialPosition():void=
        if:
            TransformComp := Entity.GetComponent[transform_component]
        then:
            set TransformComp.GlobalTransform = InitialT
            if(Chair.TeleportTo[FromVector3(InitialT.Translation),FromRotation(InitialT.Rotation)]):
                # do nothing

    ARENA_MIN_X : float = -12991.0
    ARENA_MIN_Y : float = -13001.0
    ARENA_MAX_X : float = 13000.0
    ARENA_MAX_Y : float = 13524.0
    IsValidLocation(Loc:UnrealVector3):logic=
        if (Loc.X < ARENA_MIN_X or Loc.Y < ARENA_MIN_Y):
            return false
        if (Loc.X > ARENA_MAX_X or Loc.Y > ARENA_MAX_Y):
            return false
        # Inside of the cell arena
        if (Loc.Z >= 0.0 and Loc.Z <= 350.0):
            if (Loc.X > 250.0 and Loc.Y < -250.0):
                return false
        # WWE area:
        if (Loc.Z >= 0.0 and Loc.Z <= 203.0):
            if (Loc.X < -3300.0 and Loc.Y > 3655.0):
                return false
        if (Loc.Z < 0.0 or Loc.Z >= 11000.0):
            return false
        return true